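# Translucent UI material for an equipped-state indicator, drawing a rounded rectangle with fill and outline.
# A rounded-mask SDF function takes per-corner radii and outline thickness, then anti-aliasing functions derive the Fill mask and subtract to get the Outline.
# A UseFill scalar blends between shape mask variants, while separate color and opacity parameters feed the blended fill and outline color.
# A pixel-based EdgeOffset is converted to normalized units to inset the shape.
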Material /FortUILibrary/Systems/Items/Materials/M_UIKit_EquippedState
graph {
 properties {
  bUsedWithStaticMesh=false
  blendMode=BLEND_Translucent
  materialDomain=MD_UI
  pixelDepthOffsetMode=PDOM_Legacy
  usedShadingModels=Default Lit
 }
 nodes [
  n1 = MaterialFunctionCall(
   Size=NamedRerouteUsage()
   Corner Radius=MaterialFunctionCall(
    R=ScalarParameter() {defaultValue=20, group=Layout, parameterName=TopLeftCornerRadius, sortPriority=2}
    G=ScalarParameter() {defaultValue=20, group=Layout, parameterName=BottomLeftCornerRadius, sortPriority=3}
    B=ScalarParameter() {defaultValue=20, group=Layout, parameterName=TopRightCornerRadius, sortPriority=4}
    A=ScalarParameter() {defaultValue=20, group=Layout, parameterName=BottomRightCornerRadius, sortPriority=5}
   ) {materialFunction=&AppendMany}.RGBA
   Radius in Pixels=ScalarParameter() {defaultValue=1, group=Layout, parameterName=UseValueInPixels, sortPriority=1}
   Outline Thickness=NamedRerouteUsage()
  ) {materialFunction=&MF_UI_CoreUI_RoundedMask}
  n2 = MaterialFunctionCall(DistanceField=@n1.SDF) {materialFunction=&MF_UI_AntiAliasing}
  NamedRerouteDeclaration(None=Subtract(
    A=NamedRerouteDeclaration(None=@n2.Inner) {name=Fill, nodeColor=<rgba 0 0 0 1>}
    B=MaterialFunctionCall(DistanceField=@n1.Outline SDF) {materialFunction=&MF_UI_AntiAliasing}.Outer
   )) {name=Outline, nodeColor=<rgba 0 0 0 1>}
  n3 = NamedRerouteUsage()
  n4 = NamedRerouteUsage()
  n5 = NamedRerouteUsage()
  n6 = NamedRerouteUsage()
  NamedRerouteDeclaration(None=LinearInterpolate(
    A=LinearInterpolate(A=@n5, B=@n6) {constAlpha=0}
    B=LinearInterpolate(A=@n3, B=@n4) {constAlpha=1}
    Alpha=ScalarParameter() {defaultValue=1, group=Layout, parameterName=UseFill}
   ) {constAlpha=1}) {name=Shape Mask, nodeColor=<rgba 1 0 1 1>}
  n7 = MaterialFunctionCall!(Value PX=Multiply(A=ScalarParameter() {group=Layout, parameterName=EdgeOffset, sortPriority=1}) {constB=2}, UV Pixel Size, X or Y) {materialFunction=/FortUILibrary/MaterialFunctions/Basics/MF_UI_PixelsToNormalized}
  n8 = MaterialFunctionCall(Value PX=Constant()) {materialFunction=&MF_UI_PixelsToNormalized}
  NamedRerouteDeclaration(None=OneMinus(None=Add(A=@n7.Normalized, B=@n8.Normalized))) {name=Edge Offset, nodeColor=<rgba 0 0 0 1>}
  NamedRerouteDeclaration(None=ScalarParameter() {defaultValue=10, group=Outline, parameterName=OutlineThickness, sortPriority=10}) {name=Outline Thickness, nodeColor=<rgba 0 0 0 1>}
  n9 = VectorParameter() {group=Fill, parameterName=FillColor, sortPriority=6, defaultValue=<rgba 1 0 0.6916 1>}
  n10 = ScalarParameter() {defaultValue=1, group=Fill, parameterName=FillOpacity, sortPriority=7}
  n11 = NamedRerouteUsage()
  n12 = OneMinus(None=NamedRerouteUsage())
  n13 = VectorParameter() {group=Outline, parameterName=OutlineColor, sortPriority=8, defaultValue=<rgba 1 0.8919 0 1>}
  n14 = ScalarParameter() {defaultValue=1, group=Outline, parameterName=OutlineOpacity, sortPriority=9}
  NamedRerouteDeclaration(None=LinearInterpolate(A=LinearInterpolate(B=@n9.RGB, Alpha=@n10) {constAlpha=1}, B=LinearInterpolate(B=@n13.RGB, Alpha=@n14), Alpha=Max(A=@n11, B=@n12))) {name=Color, nodeColor=<rgba 1 0 1 1>}
  NamedRerouteUsage()
  NamedRerouteUsage()
 ]
}
