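# Translucent UI material that previews an item texture with rounded corners, an optional radial vignette and an animatable archived effect.
# Item UVs take container scaling, animated centred zoom and icon offsets.
# A channel-packed mode tints blue with DynamicColor, layers green and red over it, and builds alpha from their saturated sum.
# Edge-clipped item alpha, corner mask, vignette and archived gradient form named masks for a multiply chain.
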Material /FortUILibrary/Systems/Items/Materials/M_UIKit_ItemPreview
graph {
 expressions {
  MaterialExpressionAdd_0 {
   class=MaterialExpressionAdd
   inputs [
    A {from=MaterialExpressionNamedRerouteUsage_0}
    B {from=MaterialExpressionConvert_0}
   ]
  }
  MaterialExpressionAdd_21 {
   class=MaterialExpressionAdd
   inputs [
    A {from=MaterialExpressionReroute_50}
    B {from=MaterialExpressionReroute_49}
   ]
  }
  MaterialExpressionAdd_22 {
   class=MaterialExpressionAdd
   inputs [
    A {from=MaterialExpressionReroute_48}
    B {from=MaterialExpressionAdd_21}
   ]
  }
  MaterialExpressionAppendVector_0 {
   class=MaterialExpressionAppendVector
   inputs [
    A {from=MaterialExpressionScalarParameter_12, output=""}
    B {from=MaterialExpressionScalarParameter_13, output=""}
   ]
  }
  MaterialExpressionConvert_0 {
   class=MaterialExpressionConvert
   inputs [
    None {from=MaterialExpressionScalarParameter_15}
    None {from=MaterialExpressionScalarParameter_16}
   ]
   properties {
    convertInputs [
     {
      type=Scalar
      expressionInput {
       expression=&self:MaterialExpressionScalarParameter_15
       inputName=None
       mask=0
       maskA=0
       maskB=0
       maskG=0
       maskR=0
       outputIndex=0
      }
     }
     {
      type=Scalar
      expressionInput {
       expression=&self:MaterialExpressionScalarParameter_16
       inputName=None
       mask=0
       maskA=0
       maskB=0
       maskG=0
       maskR=0
       outputIndex=0
      }
     }
    ]
    convertOutputs [{type=Vector2}]
   }
  }
  MaterialExpressionLinearInterpolate_0 {
   class=MaterialExpressionLinearInterpolate
   inputs [
    A {from=MaterialExpressionMaterialFunctionCall_23, output=ScaleY}
    B {from=MaterialExpressionMaterialFunctionCall_23, output=ScaleY}
    Alpha {from=MaterialExpressionScalarParameter_2}
   ]
  }
  MaterialExpressionLinearInterpolate_10 {
   class=MaterialExpressionLinearInterpolate
   inputs [
    A {from=MaterialExpressionLinearInterpolate_39}
    B {from=""}
    Alpha {from=MaterialExpressionReroute_52}
   ]
   properties {constB=0}
  }
  MaterialExpressionLinearInterpolate_11 {
   class=MaterialExpressionLinearInterpolate
   inputs [
    A {from=""}
    B {from=MaterialExpressionVectorParameter_8, output=RGB}
    Alpha {from=MaterialExpressionReroute_48}
   ]
  }
  MaterialExpressionLinearInterpolate_19 {
   class=MaterialExpressionLinearInterpolate
   inputs [
    A {from=MaterialExpressionScalarParameter_24}
    B {from=MaterialExpressionScalarParameter_26}
    Alpha {from=MaterialExpressionScalarParameter_29}
   ]
  }
  MaterialExpressionLinearInterpolate_2 {
   class=MaterialExpressionLinearInterpolate
   inputs [
    A {from=""}
    B {from=MaterialExpressionMultiply_1}
    Alpha {from=MaterialExpressionScalarParameter_11}
   ]
   properties {constA=1}
  }
  MaterialExpressionLinearInterpolate_32 {
   class=MaterialExpressionLinearInterpolate
   inputs [
    A {from=MaterialExpressionReroute_45}
    B {from=MaterialExpressionLinearInterpolate_10}
    Alpha {from=MaterialExpressionScalarParameter_43}
   ]
  }
  MaterialExpressionLinearInterpolate_39 {
   class=MaterialExpressionLinearInterpolate
   inputs [
    A {from=MaterialExpressionLinearInterpolate_11}
    B {from=""}
    Alpha {from=MaterialExpressionReroute_51}
   ]
  }
  MaterialExpressionLinearInterpolate_40 {
   class=MaterialExpressionLinearInterpolate
   inputs [
    A {from=MaterialExpressionLinearInterpolate_9}
    B {from=MaterialExpressionSaturate_0}
    Alpha {from=MaterialExpressionScalarParameter_43}
   ]
  }
  MaterialExpressionLinearInterpolate_9 {
   class=MaterialExpressionLinearInterpolate
   inputs [
    A {from=MaterialExpressionMaterialFunctionCall_0, output=Result}
    B {from=""}
    Alpha {from=MaterialExpressionScalarParameter_22}
   ]
   properties {constB=0}
  }
  MaterialExpressionMaterialFunctionCall_0 {
   class=MaterialExpressionMaterialFunctionCall
   inputs [
    UVs {from=MaterialExpressionNamedRerouteUsage_2}
    Alpha {from=MaterialExpressionTextureSampleParameter2D_0, output=A}
   ]
   properties {materialFunction=/FortUILibrary/MaterialFunctions/Basics/MF_UI_ClipEdgePixel}
  }
  MaterialExpressionMaterialFunctionCall_1 {
   class=MaterialExpressionMaterialFunctionCall
   inputs [
    UV {from=""}
    Size {from=""}
    Corner Radius {from=MaterialExpressionMaterialFunctionCall_3, output=RGBA}
    Radius in Pixels {from=MaterialExpressionScalarParameter_14}
    Skew Angle {from=""}
    Outline Thickness {from=""}
    Corners Modulated by Size {from=""}
    Pixel Size {from=""}
   ]
   properties {materialFunction=&MF_UI_CoreUI_RoundedMask}
  }
  MaterialExpressionMaterialFunctionCall_10 {
   class=MaterialExpressionMaterialFunctionCall
   inputs [
    UVs {from=MaterialExpressionLinearInterpolate_0}
    Texture Scale {from=MaterialExpressionLinearInterpolate_19}
   ]
   properties {materialFunction=&ScaleUVsByCenter}
  }
  MaterialExpressionMaterialFunctionCall_2 {
   class=MaterialExpressionMaterialFunctionCall
   inputs [
    UVs {from=""}
    CenterPosition {from=MaterialExpressionAppendVector_0}
    Radius {from=MaterialExpressionScalarParameter_5}
    Density {from=MaterialExpressionScalarParameter_6}
    Invert Density {from=""}
   ]
   properties {materialFunction=&RadialGradientExponential}
  }
  MaterialExpressionMaterialFunctionCall_20 {
   class=MaterialExpressionMaterialFunctionCall
   inputs [
    UVs {from=""}
    Texture {from=""}
    GradientStart {from=MaterialExpressionScalarParameter_0}
    GradientEnd {from=MaterialExpressionScalarParameter_3}
    RotationAngle {from=MaterialExpressionScalarParameter_4}
    RotationCenter {from=""}
    GradientOffset {from=""}
   ]
   properties {materialFunction=&MF_UI_LinearGradient}
  }
  MaterialExpressionMaterialFunctionCall_23 {
   class=MaterialExpressionMaterialFunctionCall
   inputs [
    UV {from=""}
    TextureSize {from=""}
    ScalePivotPoint {from=""}
    ContainerSize {from=""}
   ]
   properties {materialFunction=&MF_UI_UVScaling}
  }
  MaterialExpressionMaterialFunctionCall_3 {
   class=MaterialExpressionMaterialFunctionCall
   inputs [
    R {from=MaterialExpressionScalarParameter_1}
    G {from=MaterialExpressionScalarParameter_10}
    B {from=MaterialExpressionScalarParameter_7}
    A {from=MaterialExpressionScalarParameter_9}
   ]
   properties {materialFunction=&AppendMany}
  }
  MaterialExpressionMultiply_0 {
   class=MaterialExpressionMultiply
   inputs [
    A {from=MaterialExpressionOneMinus_0}
    B {from=MaterialExpressionNamedRerouteUsage_1}
   ]
  }
  MaterialExpressionMultiply_1 {
   class=MaterialExpressionMultiply
   inputs [
    A {from=MaterialExpressionOneMinus_1}
    B {from=MaterialExpressionScalarParameter_8}
   ]
  }
  MaterialExpressionMultiply_17 {
   class=MaterialExpressionMultiply
   inputs [
    A {from=MaterialExpressionOneMinus_9}
    B {from=MaterialExpressionScalarParameter_20}
   ]
  }
  MaterialExpressionMultiply_2 {
   class=MaterialExpressionMultiply
   inputs [
    A {from=MaterialExpressionMultiply_5}
    B {from=MaterialExpressionNamedRerouteUsage_4}
   ]
  }
  MaterialExpressionMultiply_4 {
   class=MaterialExpressionMultiply
   inputs [
    A {from=MaterialExpressionSaturate_1}
    B {from=MaterialExpressionScalarParameter_19}
   ]
  }
  MaterialExpressionMultiply_5 {
   class=MaterialExpressionMultiply
   inputs [
    A {from=MaterialExpressionMultiply_0}
    B {from=MaterialExpressionNamedRerouteUsage_6}
   ]
  }
  MaterialExpressionNamedRerouteDeclaration_0 {
   class=MaterialExpressionNamedRerouteDeclaration
   inputs [None {from=MaterialExpressionMaterialFunctionCall_1, output=Shape Mask}]
   properties {name=ShapeMask, nodeColor=<rgba 0 0 0 1>}
  }
  MaterialExpressionNamedRerouteDeclaration_1 {
   class=MaterialExpressionNamedRerouteDeclaration
   inputs [None {from=MaterialExpressionMaterialFunctionCall_10, output=UVs}]
   properties {name=UVs, nodeColor=<rgba 1 0.6328 0 1>}
  }
  MaterialExpressionNamedRerouteDeclaration_2 {
   class=MaterialExpressionNamedRerouteDeclaration
   inputs [None {from=MaterialExpressionLinearInterpolate_32}]
   properties {name=Item, nodeColor=<rgba 0 0 0 1>}
  }
  MaterialExpressionNamedRerouteDeclaration_3 {
   class=MaterialExpressionNamedRerouteDeclaration
   inputs [None {from=MaterialExpressionLinearInterpolate_40}]
   properties {name=Item Alpha, nodeColor=<rgba 0 0 0 1>}
  }
  MaterialExpressionNamedRerouteDeclaration_8 {
   class=MaterialExpressionNamedRerouteDeclaration
   inputs [None {from=MaterialExpressionMultiply_4}]
   properties {name=Vignette Alpha, nodeColor=<rgba 0 0 0 1>}
  }
  MaterialExpressionNamedRerouteDeclaration_9 {
   class=MaterialExpressionNamedRerouteDeclaration
   inputs [None {from=MaterialExpressionLinearInterpolate_2}]
   properties {name=Archived Alpha, nodeColor=<rgba 0 0 0 1>}
  }
  MaterialExpressionNamedRerouteUsage_0 {class=MaterialExpressionNamedRerouteUsage}
  MaterialExpressionNamedRerouteUsage_1 {class=MaterialExpressionNamedRerouteUsage}
  MaterialExpressionNamedRerouteUsage_2 {class=MaterialExpressionNamedRerouteUsage}
  MaterialExpressionNamedRerouteUsage_3 {class=MaterialExpressionNamedRerouteUsage}
  MaterialExpressionNamedRerouteUsage_4 {class=MaterialExpressionNamedRerouteUsage}
  MaterialExpressionNamedRerouteUsage_6 {class=MaterialExpressionNamedRerouteUsage}
  MaterialExpressionNamedRerouteUsage_7 {class=MaterialExpressionNamedRerouteUsage}
  MaterialExpressionOneMinus_0 {class=MaterialExpressionOneMinus, inputs [None {from=MaterialExpressionNamedRerouteUsage_7}]}
  MaterialExpressionOneMinus_1 {class=MaterialExpressionOneMinus, inputs [None {from=MaterialExpressionMaterialFunctionCall_20, output=Result}]}
  MaterialExpressionOneMinus_9 {class=MaterialExpressionOneMinus, inputs [None {from=MaterialExpressionMaterialFunctionCall_2, output=RadialGradientExponential}]}
  MaterialExpressionReroute_45 {class=MaterialExpressionReroute, inputs [None {from=MaterialExpressionTextureSampleParameter2D_0, output=RGB}]}
  MaterialExpressionReroute_48 {class=MaterialExpressionReroute, inputs [None {from=MaterialExpressionTextureSampleParameter2D_0, output=B}]}
  MaterialExpressionReroute_49 {class=MaterialExpressionReroute, inputs [None {from=MaterialExpressionTextureSampleParameter2D_0, output=G}]}
  MaterialExpressionReroute_50 {class=MaterialExpressionReroute, inputs [None {from=MaterialExpressionTextureSampleParameter2D_0, output=R}]}
  MaterialExpressionReroute_51 {class=MaterialExpressionReroute, inputs [None {from=MaterialExpressionReroute_49}]}
  MaterialExpressionReroute_52 {class=MaterialExpressionReroute, inputs [None {from=MaterialExpressionReroute_50}]}
  MaterialExpressionSaturate_0 {class=MaterialExpressionSaturate, inputs [None {from=MaterialExpressionAdd_22}]}
  MaterialExpressionSaturate_1 {class=MaterialExpressionSaturate, inputs [None {from=MaterialExpressionMultiply_17}]}
  MaterialExpressionScalarParameter_0 {class=MaterialExpressionScalarParameter, properties {group=Archived, parameterName=ArchivedGradientStart, sortPriority=10}}
  MaterialExpressionScalarParameter_1 {class=MaterialExpressionScalarParameter, properties {defaultValue=20, group=Layout, parameterName=TopLeftCornerRadius, sortPriority=4}}
  MaterialExpressionScalarParameter_10 {class=MaterialExpressionScalarParameter, properties {defaultValue=20, group=Layout, parameterName=BottomLeftCornerRadius, sortPriority=7}}
  MaterialExpressionScalarParameter_11 {
   class=MaterialExpressionScalarParameter
   properties {
    desc="(0 - 1) Animate item Archiving effect.\r\n0: Item not archived.\r\n1: Item archived."
    group=Archived
    parameterName=ArchivedAnimation
    sliderMax=1
    sortPriority=0
   }
  }
  MaterialExpressionScalarParameter_12 {class=MaterialExpressionScalarParameter, properties {defaultValue=0.5, group=Vignette, parameterName=VignetteCenter_X, sortPriority=10}}
  MaterialExpressionScalarParameter_13 {class=MaterialExpressionScalarParameter, properties {group=Vignette, parameterName=VignetteCenter_Y, sortPriority=11}}
  MaterialExpressionScalarParameter_14 {
   class=MaterialExpressionScalarParameter
   properties {
    defaultValue=1
    desc="If enabled reads corner radius values in pixels.\r\nbool"
    group=Layout
    parameterName=UseValueInPixels
    sliderMax=1
    sortPriority=1
   }
  }
  MaterialExpressionScalarParameter_15 {class=MaterialExpressionScalarParameter, properties {parameterName=IconOffsetX}}
  MaterialExpressionScalarParameter_16 {class=MaterialExpressionScalarParameter, properties {parameterName=IconOffsetY}}
  MaterialExpressionScalarParameter_19 {
   class=MaterialExpressionScalarParameter
   properties {
    defaultValue=1
    desc="(0 or 1) If enabled applies vignette effect on the item texture."
    group=Vignette
    parameterName=UseVignette
    sliderMax=1
    sortPriority=0
   }
  }
  MaterialExpressionScalarParameter_2 {class=MaterialExpressionScalarParameter, properties {parameterName=UVScaleX}}
  MaterialExpressionScalarParameter_20 {class=MaterialExpressionScalarParameter, properties {defaultValue=1, group=Vignette, parameterName=VignetteStrength, sortPriority=6}}
  MaterialExpressionScalarParameter_22 {class=MaterialExpressionScalarParameter, properties {group=Item, parameterName=RemoveItem}}
  MaterialExpressionScalarParameter_24 {
   class=MaterialExpressionScalarParameter
   properties {defaultValue=1, desc=Start value for scale animation., group=Item, parameterName=ScaleFrom, sortPriority=2}
  }
  MaterialExpressionScalarParameter_26 {
   class=MaterialExpressionScalarParameter
   properties {defaultValue=1.1, desc=End value for scale animation., group=Item, parameterName=ScaleTo, sortPriority=3}
  }
  MaterialExpressionScalarParameter_29 {
   class=MaterialExpressionScalarParameter
   properties {desc="(0 - 1) Animate scale transition 'ScaleFrom' - 'ScaleTo'", group=Item, parameterName=ScaleAnimation, sliderMax=1, sortPriority=4}
  }
  MaterialExpressionScalarParameter_3 {class=MaterialExpressionScalarParameter, properties {defaultValue=2, group=Archived, parameterName=ArchivedGradientEnd, sortPriority=11}}
  MaterialExpressionScalarParameter_4 {
   class=MaterialExpressionScalarParameter
   properties {defaultValue=0.75, group=Archived, parameterName=ArchivedGradientAngle, sortPriority=12}
  }
  MaterialExpressionScalarParameter_43 {
   class=MaterialExpressionScalarParameter
   properties {
    desc="(0 or 1) If enabled switch to work with channel-packed texture."
    group=ChannelPacked
    parameterName=UseChannelPacked
    sliderMax=1
    sortPriority=1
   }
  }
  MaterialExpressionScalarParameter_5 {class=MaterialExpressionScalarParameter, properties {defaultValue=1.4, group=Vignette, parameterName=VignetteRadius, sortPriority=12}}
  MaterialExpressionScalarParameter_6 {class=MaterialExpressionScalarParameter, properties {defaultValue=3.5, group=Vignette, parameterName=VignetteDensity, sortPriority=13}}
  MaterialExpressionScalarParameter_7 {class=MaterialExpressionScalarParameter, properties {defaultValue=20, group=Layout, parameterName=TopRightCornerRadius, sortPriority=5}}
  MaterialExpressionScalarParameter_8 {
   class=MaterialExpressionScalarParameter
   properties {
    defaultValue=0.95
    desc=Multiplies Archived Colors alpha. A quick way to set desired threshold in applying the archived effect.
    group=Archived
    parameterName=ArchivedAlphaMultiplier
    sliderMax=1
    sortPriority=3
   }
  }
  MaterialExpressionScalarParameter_9 {class=MaterialExpressionScalarParameter, properties {defaultValue=20, group=Layout, parameterName=BottomRightCornerRadius, sortPriority=6}}
  MaterialExpressionTextureSampleParameter2D_0 {
   class=MaterialExpressionTextureSampleParameter2D
   inputs [
    UVs {from=MaterialExpressionAdd_0, output=""}
    Apply View MipBias {from="", output=""}
   ]
   properties {
    group=Item
    parameterName=Item
    samplerSource=SSM_Clamp_WorldGroupSettings
    sortPriority=1
    texture=/FortUILibrary/Systems/Items/Textures/T_UI_SoldierCommando_F
   }
  }
  MaterialExpressionVectorParameter_8 {
   class=MaterialExpressionVectorParameter
   properties {
    desc=Color that applies to channel-packed texture
    group=ChannelPacked
    parameterName=DynamicColor
    sortPriority=2
    defaultValue=<rgba 1 0 1 1>
   }
  }
 }
 properties {
  bUsedWithStaticMesh=false
  blendMode=BLEND_Translucent
  materialDomain=MD_UI
  pixelDepthOffsetMode=PDOM_Legacy
  usedShadingModels=Default Lit
 }
}
