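# Translucent UI material for drawing an image-error icon in UIKit items, using an MSDF texture.
# UVs are scaled to fit the widget, then scaled about the center by a PixelSize parameter converted to normalized units.
# An aspect-ratio test selects the axis for that conversion.
# The MSDF sample becomes an SDF and passes through thresholded anti-aliasing for a crisp shape, alongside a constant white color.
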Material /FortUILibrary/Systems/Items/Materials/M_UIKit_Item_Error
graph {
 properties {
  bAutomaticallySetUsageInEditor=false
  bUsedWithStaticMesh=false
  blendMode=BLEND_Translucent
  materialDomain=MD_UI
  usedShadingModels=Default Lit
 }
 nodes [
  Constant3Vector() {constant=<rgba 1 1 1 0>}
  n1 = MaterialFunctionCall() {materialFunction=&GetUserInterfaceUV}
  n2 = MaterialFunctionCall!(
   A=ComponentMask(None=MaterialFunctionCall!(Dimensions=@n1.Pixel Size) {materialFunction=/FortUILibrary/MaterialFunctions/Basics/MF_UI_GetAspectRatio}.Max Ratio) {r=true}
   B=Constant() {r=1}
  ) {materialFunction=/FortUILibrary/MaterialFunctions/Basics/MF_UI_BooleanOperations}
  n3 = MaterialFunctionCall(
   UVs=MaterialFunctionCall(UV=@n1.Normalized UV) {materialFunction=&MF_UI_UVScaling}.Scaled to Fit
   Texture Scale=MaterialFunctionCall!(Value PX=ScalarParameter() {defaultValue=48, parameterName=PixelSize}, UV Pixel Size, X or Y=LinearInterpolate(Alpha=@n2.IsGreater)) {materialFunction=/FortUILibrary/MaterialFunctions/Basics/MF_UI_PixelsToNormalized}.Normalized
  ) {materialFunction=&ScaleUVsByCenter}
  n4 = TextureObject() {samplerType=SAMPLERTYPE_LinearColor, texture=/FortUILibrary/Textures/Icons/T_UI_Icon_UIKit_ImageError_MSDF}
  MaterialFunctionCall(
   DistanceField=MaterialFunctionCall!(MSDF RGB=TextureSample(UVs=@n3.UVs, Tex=@n4).RGB) {materialFunction=/FortUILibrary/MaterialFunctions/Shapes/Modifiers/MF_UI_MSDFtoSDF}.SDF
   Threshold=Constant() {r=0.5}
  ) {materialFunction=&MF_UI_AntiAliasing}
 ]
}
