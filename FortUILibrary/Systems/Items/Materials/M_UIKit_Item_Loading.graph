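# Translucent UI material for an item loading placeholder: a white rounded rectangle with an animated shimmer in its opacity.
# A timed animation with delay, eased by a sine in-out curve, drives a linear gradient's end, with fully saturated areas zeroed to form the loading mask.
# The mask appears to blend MinOpacity to MaxOpacity, multiplied by a rounded shape mask with per-corner radii in optional pixel units.
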Material /FortUILibrary/Systems/Items/Materials/M_UIKit_Item_Loading
graph {
 properties {
  bUsedWithStaticMesh=false
  blendMode=BLEND_Translucent
  materialDomain=MD_UI
  pixelDepthOffsetMode=PDOM_Legacy
  usedShadingModels=Default Lit
 }
 nodes [
  Constant3Vector() {constant=<rgba 1 1 1 0>}
  Multiply(
   A=NamedRerouteUsage()
   B=LinearInterpolate(
    A=ScalarParameter() {defaultValue=0.05, parameterName=MinOpacity, sliderMax=1}
    B=ScalarParameter() {defaultValue=0.2, parameterName=MaxOpacity, sliderMax=1}
    Alpha=NamedRerouteUsage()
   ) {constA=0.05, constB=0.2}
  )
  n1 = ScalarParameter() {defaultValue=20, group=Layout, parameterName=TopLeftCornerRadius, sortPriority=4}
  n2 = ScalarParameter() {defaultValue=20, group=Layout, parameterName=BottomLeftCornerRadius, sortPriority=7}
  n3 = ScalarParameter() {defaultValue=20, group=Layout, parameterName=TopRightCornerRadius, sortPriority=5}
  n4 = ScalarParameter() {defaultValue=20, group=Layout, parameterName=BottomRightCornerRadius, sortPriority=6}
  n5 = ScalarParameter() {
   defaultValue=1
   desc="If enabled reads corner radius values in pixels.\r\nbool"
   group=Layout
   parameterName=UseValueInPixels
   sliderMax=1
   sortPriority=1
  }
  NamedRerouteDeclaration(None=MaterialFunctionCall(
    Corner Radius=MaterialFunctionCall(R=@n1, G=@n2, B=@n3, A=@n4) {materialFunction=&AppendMany}.RGBA
    Radius in Pixels=@n5
   ) {materialFunction=&MF_UI_CoreUI_RoundedMask}.Shape Mask) {name=ShapeMask, nodeColor=<rgba 0 0 0 1>}
  n6 = Constant() {r=2}
  n7 = Constant() {r=0.125}
  n8 = MaterialFunctionCall(
   GradientEnd=MaterialFunctionCall("Value (0-1)"=MaterialFunctionCall!(Animation Duration=@n6, Delay=@n7, Reset to Default) {materialFunction=/FortUILibrary/MaterialFunctions/Time/MF_UI_AnimationDelay}.Result) {materialFunction=&MF_UI_EaseCurves}.EaseInOutSine
   RotationAngle=Constant()
  ) {materialFunction=&MF_UI_LinearGradient}
  n9 = Floor(None=@n8.Result)
  NamedRerouteDeclaration(None=Multiply(A=@n8.Result, B=OneMinus(None=@n9))) {name=Loading Mask, nodeColor=<rgba 0 0 0 1>}
 ]
}
