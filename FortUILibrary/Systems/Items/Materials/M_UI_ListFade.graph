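# Translucent UI material that builds a rotatable linear gradient, named for fading lists.
# User interface UVs pass through a Rotator driven by a Rotation parameter, with speed near two pi so one unit is a full turn.
# The rotated U channel feeds MF_UI_LinearGradient, shaped by start, end and offset parameters.
# The graph shows no link from the gradient to the material outputs.
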Material /FortUILibrary/Systems/Items/Materials/M_UI_ListFade
graph {
 properties {
  bAutomaticallySetUsageInEditor=false
  bUsedWithStaticMesh=false
  blendMode=BLEND_Translucent
  materialDomain=MD_UI
  usedShadingModels=Default Lit
 }
 nodes [
  MaterialFunctionCall!("UVs (Optional)") {materialFunction=/Engine/Functions/UserInterface/GetSlatePost1}
  n1 = MaterialFunctionCall() {materialFunction=&GetUserInterfaceUV}
  n2 = ScalarParameter() {parameterName=Rotation}
  MaterialFunctionCall(
   UVs=ComponentMask(None=Rotator(Coordinate=@n1.Normalized UV, Time=@n2) {speed=6.2832}) {r=true}
   GradientStart=ScalarParameter() {parameterName=GradientStart}
   GradientEnd=ScalarParameter() {defaultValue=1, parameterName=GradientEnd}
   GradientOffset=ScalarParameter() {parameterName=GradientOffset}
  ) {materialFunction=&MF_UI_LinearGradient}
 ]
}
