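# UI material function that generates a signed distance field for a plus-sign shape.
# It calls MF_UI_Box twice for a horizontal and a vertical bar, both sharing UVs and corner roundness, with long side Height minus Stroke and short side InnerThickness.
# A Min of the two box SDFs unions them into the single SDF output.
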MaterialFunction /FortUILibrary/Systems/Items/Materials/MF_UI_PlusShape
graph {nodes [
  n1 = FunctionInput(Preview=TextureCoordinate()) {bUsePreviewValueAsDefault=true, inputName=UVs, inputType=FunctionInput_Vector2, sortPriority=1}
  n2 = FunctionInput(Preview=Constant() {r=1}) {bUsePreviewValueAsDefault=true, inputName=Height, inputType=FunctionInput_Scalar, sortPriority=2}
  n3 = FunctionInput(Preview=Constant() {r=0.1}) {bUsePreviewValueAsDefault=true, inputName=Stroke, inputType=FunctionInput_Scalar, sortPriority=4}
  n4 = FunctionInput(Preview=Constant() {r=0.25}) {bUsePreviewValueAsDefault=true, inputName=InnerThickness, inputType=FunctionInput_Scalar, sortPriority=3}
  n5 = AppendVector(A=Subtract(A=@n2, B=@n3), B=@n4)
  n6 = FunctionInput(Preview=Constant() {r=0.15}) {bUsePreviewValueAsDefault=true, inputName=Corner Roundness, inputType=FunctionInput_Scalar, sortPriority=5}
  n7 = AppendVector(A=@n4, B=Subtract(A=@n2, B=@n3))
  FunctionOutput(None=Min(
    A=MaterialFunctionCall(UVs=@n1, Size=@n5, CornerRoundness=@n6) {materialFunction=&MF_UI_Box}.SDF
    B=MaterialFunctionCall(UVs=@n1, Size=@n7, CornerRoundness=@n6) {materialFunction=&MF_UI_Box}.SDF
   )) {outputName=SDF}
 ]}
