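# M_UI_Tag_Enhanced is a translucent UI material for a tag with a plus icon, using the item shop collection's corner radius.
# An SDF plus shape, placed on aspect-corrected square UVs, builds fill and outline color and alpha.
# bIsEnhanced sets icon opacity, while bIsMultiGame sets background opacity and adjusts icon colors and outline.
# A rounded mask optionally clips the result, and named reroutes carry UV, opacity and rift-energy values between sections.
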Material /FortUILibrary/Systems/Items/Materials/M_UI_Tag_Enhanced
graph {
 properties {
  bUsedWithStaticMesh=false
  blendMode=BLEND_Translucent
  materialDomain=MD_UI
  pixelDepthOffsetMode=PDOM_Legacy
  usedShadingModels=Default Lit
 }
 nodes [
  n1 = Min(A=NamedRerouteUsage(), B=NamedRerouteUsage())
  n2 = CollectionParameter() {collection=/FortUILibrary/Systems/ItemShopTile/MPC_ItemShop, parameterName=CornerRadius}
  n3 = MaterialFunctionCall(G=@n2, A=@n2) {materialFunction=&AppendMany}
  n4 = Constant() {r=1}
  LinearInterpolate(
   A=@n1
   B=Min(
    A=@n1
    B=MaterialFunctionCall(Corner Radius=@n3.RGBA, Radius in Pixels=@n4) {materialFunction=&MF_UI_CoreUI_RoundedMask}.Shape Mask
   )
   Alpha=ScalarParameter() {defaultValue=1, group=2. Background, parameterName=bUseRoundedCorners, sortPriority=0}
  )
  n5 = ScalarParameter() {group=3. Plus icon, parameterName=Plus icon offset X}
  n6 = LinearInterpolate(Alpha=NamedRerouteUsage()) {constA=1, constB=-1}
  n7 = Add(A=NamedRerouteUsage(), B=AppendVector(A=Multiply(A=@n5, B=@n6), B=ScalarParameter() {group=3. Plus icon, parameterName=Plus icon offset Y}))
  n8 = ScalarParameter() {defaultValue=0.7, group=3. Plus icon, parameterName=Plus icon scale}
  n9 = ScalarParameter() {defaultValue=0.8, group=3. Plus icon, parameterName=Plus icon stroke}
  NamedRerouteDeclaration(None=MaterialFunctionCall!(
    UVs=MaterialFunctionCall(UVs=@n7, Texture Scale=@n8) {materialFunction=&ScaleUVsByCenter}.UVs
    Height
    InnerThickness=ScalarParameter() {defaultValue=0.235, group=3. Plus icon, parameterName=Plus icon inner thickness}
    Stroke=OneMinus(None=@n9)
    Corner Roundness=ScalarParameter() {defaultValue=0.07, group=3. Plus icon, parameterName=Plus icon border radius}
   ) {materialFunction=&/MF_UI_PlusShape}.SDF) {name=SDF plus shape, nodeColor=<rgba 0 0 0 1>}
  n10 = VectorParameter() {group=3. Plus icon, parameterName=Plus icon color, defaultValue=<rgba 1 1 1 1>}
  n11 = VectorParameter() {group=3. Plus icon, parameterName=Plus icon stroke color, defaultValue=<rgba 0.0048 0.004 0.0116 0.502>}
  n12 = LinearInterpolate(
   A=@n10.RGB
   B=LinearInterpolate(A=@n10.RGB, B=@n11.RGB, Alpha=ScalarParameter() {defaultValue=1, group=1. Controls, parameterName=bIsMultiGame, sliderMax=1})
   Alpha=NamedRerouteUsage()
  )
  n13 = NamedRerouteUsage()
  n14 = VectorParameter() {group=3. Plus icon, parameterName=Plus icon stroke color, defaultValue=<rgba 0.0048 0.004 0.0116 0.502>}
  NamedRerouteDeclaration(None=LinearInterpolate(A=@n14.RGB, B=Multiply(A=@n12, B=@n13), Alpha=NamedRerouteUsage())) {name=SDF Cross Colors, nodeColor=<rgba 0 0 0 1>}
  n15 = NamedRerouteUsage()
  n16 = OneMinus(None=SmoothStep(Max=ScalarParameter() {defaultValue=0.02, group=3. Plus icon, parameterName=Plus shape sharpness}, Value=@n15))
  n17 = NamedRerouteDeclaration(None=Multiply(
    A=VectorParameter() {group=3. Plus icon, parameterName=Plus icon color, defaultValue=<rgba 1 1 1 1>}.A
    B=NamedRerouteDeclaration(None=@n16) {name=plus icon shape, nodeColor=<rgba 1 0 0.375 1>}
   )) {name=plus icon alpha, nodeColor=<rgba 1 0.6797 0 1>}
  n18 = LinearInterpolate(A=@n14.A, Alpha=NamedRerouteUsage()) {constA=1, constB=0}
  n19 = ScalarParameter() {defaultValue=1, group=1. Controls, parameterName=bIsMultiGame, sliderMax=1}
  n20 = Abs(None=@n15)
  n21 = ScalarParameter() {defaultValue=0.08, group=3. Plus icon, parameterName=Plus shape stroke thickness}
  n22 = OneMinus(None=SmoothStep(Max=ScalarParameter() {defaultValue=0.03, group=3. Plus icon, parameterName=Plus shape outline sharpness}, Value=Subtract(A=@n20, B=@n21)))
  n23 = NamedRerouteDeclaration(None=Multiply(
    A=LinearInterpolate(A=@n14.A, B=@n18, Alpha=@n19)
    B=NamedRerouteDeclaration(None=@n22) {name=plus icon outline, nodeColor=<rgba 0 0.5547 1 1>}
   )) {name=plus icon outline alpha, nodeColor=<rgba 1 0 0.6094 1>}
  NamedRerouteDeclaration(None=Saturate(None=Add(A=@n17, B=@n23))) {name=SDF Cross Alpha, nodeColor=<rgba 0 0 0 1>}
  n24 = MaterialFunctionCall!(Dimensions) {materialFunction=/FortUILibrary/MaterialFunctions/Basics/MF_UI_GetAspectRatio}
  n25 = Multiply(A=TextureCoordinate(), B=@n24.Max Ratio)
  NamedRerouteDeclaration(None=Subtract(A=NamedRerouteDeclaration(None=@n25) {name=Left aligned square UVs, nodeColor=<rgba 0 1 0.625 1>}, B=Subtract(A=@n24.Max Ratio)) {constB=0}) {name=Right aligned square UVs, nodeColor=<rgba 0 1 0.625 1>}
  n26 = NamedRerouteUsage()
  n27 = Constant2Vector() {r=1}
  NamedRerouteDeclaration(None=LinearInterpolate(A=NamedRerouteUsage(), B=Add(A=@n26, B=@n27), Alpha=NamedRerouteUsage())) {name=Background UVs, nodeColor=<rgba 0 0.6953 1 1>}
  NamedRerouteDeclaration(None=ScalarParameter() {defaultValue=1, group=1. Controls, parameterName=bIsEnhanced, sliderMax=1}) {name=plus icon opacity, nodeColor=<rgba 1 0.1875 0 1>}
  NamedRerouteDeclaration(None=ScalarParameter() {defaultValue=1, group=1. Controls, parameterName=bIsMultiGame, sliderMax=1}) {name=background opacity, nodeColor=<rgba 1 0.1875 0 1>}
  NamedRerouteDeclaration(None=ScalarParameter() {group=1. Controls, parameterName=Rift energy expand}) {name=rift energy expand, nodeColor=<rgba 0.8047 0 1 1>}
  NamedRerouteDeclaration(None=LinearInterpolate(A=NamedRerouteUsage(), B=NamedRerouteUsage(), Alpha=NamedRerouteUsage())) {name=Icon UVs, nodeColor=<rgba 1 0.4634 0 1>}
  NamedRerouteDeclaration(None=LinearInterpolate(
    A=MaterialFunctionCall() {materialFunction=&ScreenAlignedUVs}."X 100%, Y 100%"
    B=MaterialFunctionCall() {materialFunction=&GetUserInterfaceUV}.Normalized UV
   ) {constAlpha=0.1}) {name=Screen Aligned UVs, nodeColor=<rgba 0 0 0 1>}
  NamedRerouteUsage()
 ]
}
