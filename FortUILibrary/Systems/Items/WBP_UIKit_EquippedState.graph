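# UI Kit widget that shows an item's equipped state as a rounded outline with a slot-label badge in the bottom-left corner.
# SetEquippedState creates dynamic material instances for both borders, tints them yellow when equipped or white otherwise, sets corner radii and label, and toggles badge visibility.
# An optional mode caches the outline thickness, zeroing it when unequipped.
# PreConstruct applies an equipped slot "1" state.
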WidgetBlueprint /FortUILibrary/Systems/Items/WBP_UIKit_EquippedState
graphs [
 EventGraph {nodes [
   Event PreConstruct()
   > SetEquippedState!(execute, self, IsCurrentSlotEquipped:bool=true, DisplaySlot:bool=true, EquippedSlot:Text="LOCTEXT(\"1\")", SlotImage:Soft Object Reference) -> (then)
   Event Construct()
   Event Tick()
  ]}
 SetCornerRadius {nodes [
   n1 = SetCornerRadius!() -> (
    then
    TopLeftCornerRadius:float=20
    TopRightCornerRadius:float=20
    BottomLeftCornerRadius:float=20
    BottomRightCornerRadius:float=20
    RadiusInPixels:bool=true
   ) {is=title K2Node_FunctionEntry}
   > Set TopLeftCornerRadius(TopLeftCornerRadius:float=@n1) -> (Output_Get:float)
   > Set TopRightCornerRadius(TopRightCornerRadius:float=@n1.TopRightCornerRadius) -> (Output_Get:float)
   > Set BottomLeftCornerRadius(BottomLeftCornerRadius:float=@n1.BottomLeftCornerRadius) -> (Output_Get:float)
   > Set BottomRightCornerRadius(BottomRightCornerRadius:float=@n1.BottomRightCornerRadius) -> (Output_Get:float)
   > Set RadiusInPixels(RadiusInPixels:bool=@n1.RadiusInPixels) -> (Output_Get:bool)
   > n2 = GetDynamicMaterial(self:Border Object Reference=Get Outline() -> (Outline:Common Border Object Reference))
   > SetScalarParameterValue(self=@n2, ParameterName=UseValueInPixels, Value="To Float (Boolean)"(InBool=Get RadiusInPixels() -> (RadiusInPixels:bool)))
   > SetScalarParameterValue(self=@n2, ParameterName=TopLeftCornerRadius, Value=Get TopLeftCornerRadius() -> (TopLeftCornerRadius:float))
   > SetScalarParameterValue(self=@n2, ParameterName=TopRightCornerRadius, Value=Get TopRightCornerRadius() -> (TopRightCornerRadius:float))
   > SetScalarParameterValue(self=@n2, ParameterName=BottomLeftCornerRadius, Value=Get BottomLeftCornerRadius() -> (BottomLeftCornerRadius:float))
   > SetScalarParameterValue(self=@n2, ParameterName=BottomRightCornerRadius, Value=Get BottomRightCornerRadius() -> (BottomRightCornerRadius:float))
  ]}
 SetEquippedState {nodes [
   SetEquippedState!() -> (then, IsCurrentSlotEquipped:bool=~, DisplaySlot:bool=~, EquippedSlot:Text="LOCTEXT(\"1\")", SlotImage:Soft Object Reference) {is=title K2Node_FunctionEntry}
   > n1 = Sequence!(execute) -> (then_0, then_1, then_2:Exec, then_3:Exec)
   > n2 = CreateDynamicMaterialInstance!(
    execute
    Parent=/FortUILibrary/Systems/Items/MaterialInstances/MI_UIKit_EquippedStateOutline
    OptionalName=MID_UIKit_EquippedOutline
    CreationFlags=None
   ) -> (then, ReturnValue)
   > SetBrushResourceToMaterial!(
    execute
    Brush:"Slate Brush Structure (by ref)"=Get OutlineColor() -> (OutlineColor:Slate Brush Structure)
    Material:Material Interface Object Reference=@n2
   ) -> (then)
   > SetVectorParameterValue(
    self=@n2
    ParameterName=OutlineColor
    Value=Select(
     Option 0:Linear Color Structure=(R=1,G=1,B=1,A=1)
     Option 1:Linear Color Structure=(R=0.9301,G=1,B=0.0103,A=1)
     Index=Get IsCurrentSlotEquipped() -> (IsCurrentSlotEquipped:bool)
    ) -> (ReturnValue:Linear Color Structure)
   )
   > UpdateOutlineThicknessState!(
    execute
    self
    ShouldHideOutlineWhenCurrentSlotNotEquipped:bool=Get HideOutlineWhenCurrentSlotNotEquipped() -> (HideOutlineWhenCurrentSlotNotEquipped:bool)
    InputPin:Material Instance Dynamic Object Reference=@n2
    IsCurrentSlotEquipped:bool=Get IsCurrentSlotEquipped()
   ) -> (then)
   > SetBrush(self:Border Object Reference=Get Outline(), InBrush=Get OutlineColor())
   SetCornerRadius!(
    execute=@n1.then_2
    self
    TopLeftCornerRadius:float=Get TopLeftCornerRadius()
    TopRightCornerRadius:float=Get TopRightCornerRadius()
    BottomLeftCornerRadius:float=Get BottomLeftCornerRadius()
    BottomRightCornerRadius:float=Get BottomRightCornerRadius()
    RadiusInPixels:bool=Get RadiusInPixels()
   ) -> (then) {is=title K2Node_CallFunction}
   "SetText (Text)"(
    execute=@n1.then_3
    self=Get Text_EquippedSlot() -> (Text_EquippedSlot:Common Text Object Reference)
    InText=Get EquippedSlot() -> (EquippedSlot:Text)
   )
   > SetVisibility(
    self=Get EquippedSection() -> (EquippedSection:Common Border Object Reference)
    InVisibility=Select(Option 0:ESlateVisibility Enum=Collapsed, Option 1:ESlateVisibility Enum=Visible, Index=Get DisplaySlot() -> (DisplaySlot:bool)) -> (ReturnValue:ESlateVisibility Enum)
   )
   n3 = CreateDynamicMaterialInstance(
    execute=@n1.then_1
    Parent=/FortUILibrary/Systems/Items/MaterialInstances/MI_UIKit_EquippedStateSection
    OptionalName=MID_UIKit_EquippedSlot
    CreationFlags=None
   )
   > SetBrushResourceToMaterial(Brush=Get SlotColor() -> (SlotColor:Slate Brush Structure), Material=@n3)
   n4 = Select(
    Option 0:Linear Color Structure=(R=1,G=1,B=1,A=1)
    Option 1:Linear Color Structure=(R=0.9301,G=1,B=0.0103,A=1)
    Index=Get IsCurrentSlotEquipped()
   ) -> (ReturnValue:Linear Color Structure)
   > SetVectorParameterValue(self=@n3, ParameterName=FillColor, Value=@n4)
   > SetVectorParameterValue(self=@n3, ParameterName=OutlineColor, Value=@n4)
   > SetBrush(self=Get EquippedSection(), InBrush=Get SlotColor())
  ]}
 UpdateOutlineThicknessState {nodes [
   n1 = UpdateOutlineThicknessState!() -> (then, ShouldHideOutlineWhenCurrentSlotNotEquipped:bool=~, InputPin:Material Instance Dynamic Object Reference, IsCurrentSlotEquipped:bool=~) {is=title K2Node_FunctionEntry}
   > n2 = Branch(Condition=@n1)
   > n3 = Branch(Condition="float < float"(A=Get CachedOutlineThickness() -> (CachedOutlineThickness:double)))
   > n4 = Get Scalar Parameter Value!(execute, self:Material Instance Dynamic Object Reference=@n1.InputPin, ParameterName:Name=OutlineThickness) -> (then, ReturnValue:float) {is=title K2Node_CallFunction}
   > n5 = Set CachedOutlineThickness(CachedOutlineThickness:double=@n4) -> (Output_Get:double)
   SetScalarParameterValue(
    execute=[@n5.then, @n3.else]
    self=@n1.InputPin
    ParameterName=OutlineThickness
    Value=Select(Option 0:float=0, Option 1:float=Get CachedOutlineThickness(), Index=@n1.IsCurrentSlotEquipped) -> (ReturnValue:float)
   )
   > Return Node()
   Return Node(execute=@n2.else)
  ]}
]
variables [
 BottomLeftCornerRadius {type=number, default=20}
 BottomRightCornerRadius {type=number, default=20}
 CachedOutlineThickness {type=number, default=-1}
 HideOutlineWhenCurrentSlotNotEquipped {type=boolean}
 OutlineColor {
  type=SlateBrush
  default {
   bIsDynamicallyLoaded=false
   drawAs=Image
   imageType=NoImage
   mirroring=NoMirror
   resourceName=None
   resourceObject=null
   tiling=NoTile
   imageSize=<32 32>
   margin=<ltrb 0 0 0 0>
   outlineSettings {
    bUseBrushTransparency=false
    roundingType=HalfHeightRadius
    width=0
    color {colorUseRule=UseColor_Specified, specifiedColor=<rgba 0 0 0 0>}
    cornerRadii=<0 0 0 1>
   }
   tintColor {colorUseRule=UseColor_Specified, specifiedColor=<rgba 1 1 1 1>}
   uVRegion {bIsValid=false, max=<0 0>, min=<0 0>}
  }
 }
 RadiusInPixels {type=boolean, default=true}
 SlotColor {
  type=SlateBrush
  default {
   bIsDynamicallyLoaded=false
   drawAs=Image
   imageType=NoImage
   mirroring=NoMirror
   resourceName=None
   resourceObject=null
   tiling=NoTile
   imageSize=<32 32>
   margin=<ltrb 0 0 0 0>
   outlineSettings {
    bUseBrushTransparency=false
    roundingType=HalfHeightRadius
    width=0
    color {colorUseRule=UseColor_Specified, specifiedColor=<rgba 0 0 0 0>}
    cornerRadii=<0 0 0 1>
   }
   tintColor {colorUseRule=UseColor_Specified, specifiedColor=<rgba 1 1 1 1>}
   uVRegion {bIsValid=false, max=<0 0>, min=<0 0>}
  }
 }
 TopLeftCornerRadius {type=number, default=20}
 TopRightCornerRadius {type=number, default=20}
]
widgets {
 info {inheritedWidgetCount=0, parentClass=&UserWidget, rootWidgetClass=&SizeBox, widgetCount=5}
 widgets [
  SizeBox_35 {
   widgetClassPath=&SizeBox
   widget=SizeBox_35
   properties {bOverride_HeightOverride=true, bOverride_WidthOverride=true, heightOverride=136, widthOverride=108}
  }
  Outline {
   widgetClassPath=&CommonBorder
   bIsVariable=true
   parent=SizeBox_35
   widget=Outline
   properties {
    horizontalAlignment=HAlign_Left
    verticalAlignment=VAlign_Bottom
    background {
     bIsDynamicallyLoaded=false
     drawAs=Image
     imageType=NoImage
     mirroring=NoMirror
     resourceName=None
     resourceObject=&MI_UIKit_EquippedStateOutline
     tiling=NoTile
     imageSize=<32 32>
     margin=<ltrb 0 0 0 0>
     outlineSettings {
      bUseBrushTransparency=false
      roundingType=HalfHeightRadius
      width=0
      color {colorUseRule=UseColor_Specified, specifiedColor=<rgba 0 0 0 0>}
      cornerRadii=<0 0 0 0>
     }
     tintColor {colorUseRule=UseColor_Specified, specifiedColor=<rgba 1 1 1 1>}
     uVRegion {bIsValid=false, max=<0 0>, min=<0 0>}
    }
    contentColorAndOpacity=<rgba 1 1 1 0.995>
    padding=<ltrb 0 0 0 0>
   }
   slot {class=SizeBoxSlot}
  }
  EquippedCorner {
   widgetClassPath=&SizeBox
   parent=Outline
   widget=EquippedCorner
   properties {bOverride_HeightOverride=true, bOverride_WidthOverride=true, heightOverride=40, widthOverride=40}
   slot {class=BorderSlot, properties {horizontalAlignment=HAlign_Left, verticalAlignment=VAlign_Bottom, padding=<ltrb 0 0 0 0>}}
  }
  EquippedSection {
   widgetClassPath=&CommonBorder
   bIsVariable=true
   parent=EquippedCorner
   widget=EquippedSection
   properties {
    horizontalAlignment=HAlign_Center
    verticalAlignment=VAlign_Center
    background {
     bIsDynamicallyLoaded=false
     drawAs=Image
     imageType=NoImage
     mirroring=NoMirror
     resourceName=None
     resourceObject=&MI_UIKit_EquippedStateSection
     tiling=NoTile
     imageSize=<32 32>
     margin=<ltrb 0 0 0 0>
     outlineSettings {
      bUseBrushTransparency=false
      roundingType=HalfHeightRadius
      width=0
      color {colorUseRule=UseColor_Specified, specifiedColor=<rgba 0 0 0 0>}
      cornerRadii=<0 0 0 0>
     }
     tintColor {colorUseRule=UseColor_Specified, specifiedColor=<rgba 1 1 1 1>}
     uVRegion {bIsValid=false, max=<0 0>, min=<0 0>}
    }
    padding=<ltrb 0 0 0 0>
   }
   slot {class=SizeBoxSlot, properties {padding=<ltrb 2 0 0 2>}}
  }
  Text_EquippedSlot {
   widgetClassPath=&CommonTextBlock
   bIsVariable=true
   parent=EquippedSection
   widget=Text_EquippedSlot
   properties {
    justification=Center
    mobileFontSizeMultiplier=1.2
    text=1
    colorAndOpacity {colorUseRule=UseColor_Specified, specifiedColor=<rgba 0.0048 0.004 0.0116 1>}
    font {
     bForceMonospaced=false
     bMaterialIsStencil=false
     fontMaterial=null
     fontObject=&HeadingNow
     letterSpacing=0
     monospacedWidth=1
     size=15
     skewAmount=0
     typefaceFontName=86 Bold
     outlineSettings {
      bApplyOutlineToDropShadows=false
      bMiteredCorners=false
      bSeparateFillAlpha=false
      outlineMaterial=null
      outlineSize=0
      outlineColor=<rgba 0 0 0 1>
     }
    }
    shadowOffset=<0 0>
   }
   slot {class=BorderSlot, properties {horizontalAlignment=HAlign_Center, verticalAlignment=VAlign_Center, padding=<ltrb 0 0 0 0>}}
  }
 ]
}
