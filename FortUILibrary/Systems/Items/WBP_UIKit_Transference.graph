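# UIKit item widget that shows a transference icon: a single Image using MI_UIKit_Transference.
# PreConstruct applies Size through SetSize, which scales the image by MobileSizeMultiplier on mobile.
# SetIsEnhanced and SetRoundedCorners drive scalar flags on the image's dynamic material.
# SetHasTransference and SetHasTransferenceModes hold bIsMultiGame logic, but their entry nodes are unwired and the mode-set check never reaches its Branch.
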WidgetBlueprint /FortUILibrary/Systems/Items/WBP_UIKit_Transference
graphs [
 EventGraph {nodes [
   Event PreConstruct()
   > SetSize!(execute, self, Size:Vector 2D Structure=Get Size() -> (Size:Vector 2D Structure)) -> (then)
  ]}
 SetHasTransference {nodes [
   n1 = GetDynamicMaterial(self=Get Image() -> (Image:Image Object Reference))
   > SetScalarParameterValue(self=@n1, ParameterName=bIsMultiGame, Value=0)
   SetHasTransference!() -> (then, Has Transference:bool=~) {is=title K2Node_FunctionEntry}
  ]}
 SetHasTransferenceModes {nodes [
   n1 = GetDynamicMaterial(self=Get Image())
   > n2 = Branch(Condition=false)
   > SetScalarParameterValue(self=@n1, ParameterName=bIsMultiGame, Value=1)
   SetScalarParameterValue(execute=@n2.else, self=@n1, ParameterName=bIsMultiGame, Value=0)
   n3 = SetHasTransferenceModes!() -> (then, Compatible Modes:Set of Bytes) {is=title K2Node_FunctionEntry}
   AND Boolean(
    A="integer > integer"(A=Length!(TargetSet:Set of Bytes=@n3) -> (ReturnValue:int=0) {is=title K2Node_CallFunction}, B=1)
    B=Contains Item!(TargetSet:Set of Bytes=@n3, ItemToFind:"Byte (by ref)") -> (ReturnValue:bool) {is=title K2Node_CallFunction}
   )
  ]}
 SetIsEnhanced {nodes [
   n1 = SetIsEnhanced!() -> (then, Is Enhanced:bool=~) {is=title K2Node_FunctionEntry}
   > n2 = GetDynamicMaterial(self=Get Image())
   > SetScalarParameterValue(self=@n2, ParameterName=bIsEnhanced, Value="To Float (Boolean)"(InBool=@n1))
  ]}
 SetRoundedCorners {nodes [
   n1 = SetRoundedCorners!() -> (then, Is Rounded:bool=~) {is=title K2Node_FunctionEntry}
   > n2 = GetDynamicMaterial(self=Get Image())
   > SetScalarParameterValue(self=@n2, ParameterName=bUseRoundedCorners, Value="To Float (Boolean)"(InBool=@n1))
  ]}
 SetSize {nodes [
   n1 = SetSize!() -> (then, Size:Vector 2D Structure) {is=title K2Node_FunctionEntry}
   > n2 = Set Size(Size:Vector 2D Structure=@n1) -> (Output_Get:Vector 2D Structure)
   n3 = Get MobileSizeMultiplier() -> (MobileSizeMultiplier:double)
   > SetDesiredSizeOverride!(
    execute
    self:Image Object Reference=Get Image()
    DesiredSize:Vector 2D Structure=Select(
     Option 0:Vector 2D Structure=@n2
     Option 1:Vector 2D Structure=vector2d * vector2d(A:Vector 2D Structure=@n1, B=@n3) -> (ReturnValue:Vector 2D Structure) {is=title K2Node_PromotableOperator}
     Index=IsMobileGame()
    ) -> (ReturnValue:Vector 2D Structure)
   ) -> (then)
  ]}
]
variables [
 MobileSizeMultiplier {type=number, default=1}
 Size {type=Vector2D, default=<48 48>}
]
widgets {
 info {inheritedWidgetCount=0, parentClass=&UserWidget, rootWidgetClass=&Image, widgetCount=1}
 widgets [Image {
   widgetClassPath=&Image
   bIsVariable=true
   widget=Image
   properties {brush {
     bIsDynamicallyLoaded=false
     drawAs=Image
     imageType=NoImage
     mirroring=NoMirror
     resourceName=None
     resourceObject=/FortUILibrary/Systems/Items/MaterialInstances/MI_UIKit_Transference
     tiling=NoTile
     imageSize=<48 48>
     margin=<ltrb 0 0 0 0>
     outlineSettings {
      bUseBrushTransparency=false
      roundingType=HalfHeightRadius
      width=0
      color {colorUseRule=UseColor_Specified, specifiedColor=<rgba 0 0 0 0>}
      cornerRadii=<0 0 0 0>
     }
     tintColor {colorUseRule=UseColor_Specified, specifiedColor=<rgba 1 1 1 1>}
     uVRegion {bIsValid=false, max=<0 0>, min=<0 0>}
    }}
  }]
}
