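# Base UIKit item card button showing a Fort item's lazy-loaded image, async material background, equipped outline and optional per-category custom widget.
# Status indicators are configured through a spawn-parameter map, cached by type and placed in grid panels, with cvar-switched icon loading.
# The preview widget's cosmetic actions become context menu entries using data table input rows.
# Button events drive focus, hover/press/hold sounds and failed-image retries; helpers check age gating.
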WidgetBlueprint /FortUILibrary/Systems/Items/WBP_UIKit_ItemCard_Base
animations [Anim_Hovered_Base {
  binding_count=1
  length_seconds=0.15
  object=&self:Anim_Hovered_Base
  sequence {
   bindings [LazyImage_Item {
     id=#1
     possessed_class=FortLazyImage
     tracks [{
       type=MovieSceneWidgetMaterialTrack
       sections [{
         type=MovieSceneParameterSection
         end_frame=36
         start_frame=0
         channels [
          VignetteStrength {type=MovieSceneScriptingFloatChannel, keys [[0,0.5],[36,0.1]]}
          ScaleAnimation {type=MovieSceneScriptingFloatChannel, keys [[0,0],[36,1]]}
         ]
        }]
      }]
    }]
   display_rate {numerator=240}
  }
 }]
graphs [
 Bind Preview Actions {nodes [
   n1 = Bind Preview Actions!() -> (
    then
    Preview Widget:Fort Item Preview Widget Object Reference
    Context Menu:WBP UIKit Context Menu Anchor Button Base Object Reference
    Item VM:Fort Item VM Object Reference
    Include Special Actions:bool=true
    Include Toggle Actions:bool=~
   ) {is=title K2Node_FunctionEntry}
   > Set Context Menu(Context Menu:WBP UIKit Context Menu Anchor Button Base Object Reference=@n1.Context Menu) -> (Output_Get:WBP UIKit Context Menu Anchor Button Base Object Reference)
   > Set PreviewWidget(PreviewWidget:Fort Item Preview Widget Object Reference=Get Preview Widget() -> (Preview Widget:Fort Item Preview Widget Object Reference)) -> (Output_Get:Fort Item Preview Widget Object Reference)
   > n2 = GetDataTableRowNames(Table=Get PreviewActionInputDataTable() -> (PreviewActionInputDataTable:Data Table Object Reference))
   > n3 = For Each Loop(Array:Array of Names=@n2) -> (Array Element:Name)
   n4 = Set Preview Actions(
    execute=@n3.Completed
    Preview Actions:Array of Names=GetCosmeticPreviewActions!(
     self:Fort Item Preview Widget Object Reference=Get Preview Widget()
     ItemVM:Fort Item VM Object Reference=Get Item VM() -> (Item VM:Fort Item VM Object Reference)
     bIncludeSpecialPreviewActions:bool=Get Include Special Actions() -> (Include Special Actions:bool)
     bIncludeTogglePreviewActions:bool=Get Include Toggle Actions() -> (Include Toggle Actions:bool)
     ActionState:EFortCosmeticPreviewActionState Enum=NextActionState
    ) -> (ReturnValue:Array of Names)
   ) -> (Output_Get:Array of Names)
   > n5 = For Each Loop(Array=@n4)
   > n6 = Set CurrentPreviewActionID(CurrentPreviewActionID:Name=@n5) -> (Output_Get:Name=None)
   > n7 = Get Data Table Row!(execute:Exec, DataTable:Data Table Object Reference=Get PreviewActionInputDataTable(), RowName:Name=@n6) -> (then:Exec, RowNotFound:Exec, ReturnValue:Common Input Action Data Base Structure)
   n8 = Break Common Input Action Data Base!(CommonInputActionDataBase:"Common Input Action Data Base Structure (by ref)"=@n7) -> (DisplayName:Text, HoldDisplayName:Text, NavBarPriority:int) {is=title K2Node_BreakStruct}
   > Set CurrentPreviewActionText(CurrentPreviewActionText:Text=@n8) -> (Output_Get:Text)
   > Set CurrentSortPriority(CurrentSortPriority:int=@n8.NavBarPriority) -> (Output_Get:int=0)
   > n9 = Branch(Condition=Is Valid Index!(TargetArray:Array of Booleans=Get ShouldClose() -> (ShouldClose:Array of Booleans), IndexToTest:int=@n5.Array Index) -> (ReturnValue:bool))
   n10 = TextIsEmpty(InText=@n8.HoldDisplayName)
   > n11 = "Set Boolean (by ref)"!(
    execute:Exec
    Target:"Boolean (by ref)"="Get (a ref)"(Array:Array of Booleans=Get ShouldClose(), Dimension 1=@n5.Array Index) -> (Output:"Boolean (by ref)")
    Value:bool=@n10
   ) -> (then:Exec) {is=title K2Node_VariableSetRef}
   n12 = Branch(execute=[@n11.then, @n18.then], Condition=TextIsEmpty(InText=@n8))
   n13 = Switch on Int!(execute:Exec=@n12.else, Selection:int=@n5.Array Index) -> (Default:Exec, 0:Exec, 1:Exec)
   n14 = Create Event()
   n15 = Create Event()
   n16 = Get CurrentPreviewActionID() -> (CurrentPreviewActionID:Name=None)
   n17 = Make Data Table Row Handle!(DataTable:Data Table Object Reference=Get PreviewActionInputDataTable(), RowName:Name=@n16) -> (DataTableRowHandle:Data Table Row Handle Structure) {is=title K2Node_MakeStruct}
   Set Context Menu Action!(
    execute=@n13.0
    self
    Context Menu:WBP UIKit Context Menu Anchor Button Base Object Reference=Get Context Menu() -> (Context Menu:WBP UIKit Context Menu Anchor Button Base Object Reference)
    Preview Action ID:Name=Get CurrentPreviewActionID() -> (CurrentPreviewActionID=None)
    Preview Action Text:Text=Get CurrentPreviewActionText() -> (CurrentPreviewActionText:Text)
    In Action Info Sort Priority:int=Get CurrentSortPriority() -> (CurrentSortPriority:int=0)
    In Action Info Interaction:Data Table Row Handle Structure=@n17
    In Action Info Pressed Delegate:Delegate=@n14
    In Action Info Released Delegate:Delegate=@n15
    Preview Action IteractionDelegate:"Delegate (by ref)"=Create Event()
   ) -> (then) {is=title K2Node_CallFunction}
   Set Context Menu Action(
    execute=@n13.1
    Context Menu=Get Context Menu()
    Preview Action ID=@n16
    Preview Action Text=Get CurrentPreviewActionText()
    In Action Info Sort Priority=Get CurrentSortPriority() -> (CurrentSortPriority=0)
    In Action Info Interaction=@n17
    In Action Info Pressed Delegate=@n14
    In Action Info Released Delegate=@n15
    Preview Action IteractionDelegate=Create Event()
   )
   n18 = Add(execute=@n9.else, TargetArray:Array of Booleans=Get ShouldClose(), NewItem:"Boolean (by ref)"=@n10) {is=title K2Node_CallArrayFunction}
   RemoveInteraction!(
    execute=@n3.LoopBody
    self:UIKit Context Menu Action Context Object Reference=Get InteractionContext!(self:WBP UIKit Context Menu Anchor Button Base Object Reference=Get Context Menu()) -> (InteractionContext:UIKit Context Menu Action Context Object Reference)
    InteractionNameID:Name=@n3
   ) -> (then)
  ]}
 CalculateCornerRadius {nodes [
   n1 = Get CornerRadius_TopLeft() -> (CornerRadius_TopLeft:double)
   n2 = Inputs!() -> (DefaultCornerRadius:double=~, Offset:double=~)
   n3 = Get OverrideCornerRadius() -> (OverrideCornerRadius:bool)
   n4 = Get CornerRadius_BottomLeft() -> (CornerRadius_BottomLeft:double)
   n5 = Get CornerRadius_TopRight() -> (CornerRadius_TopRight:double)
   n6 = Get CornerRadius_BottomRight() -> (CornerRadius_BottomRight:double)
   Outputs!(
    TopLeft:double=Select(Option 0:double=@n2, Option 1:double=float + float(A=@n1, B=@n2.Offset), Index=@n3) -> (ReturnValue:double)
    TopRight:double=Select(Option 0=@n2, Option 1=float + float(A=@n5, B=@n2.Offset), Index=@n3)
    BottomLeft:double=Select(Option 0=@n2, Option 1=float + float(A=@n4, B=@n2.Offset), Index=@n3)
    BottomRight:double=Select(Option 0=@n2, Option 1=float + float(A=@n6, B=@n2.Offset), Index=@n3)
    RadiusInPixels:bool=Select(Option 0:bool=true, Option 1:bool=Get RadiusInPixels() -> (RadiusInPixels:bool), Index=@n3) -> (ReturnValue:bool)
   )
  ]}
 EventGraph {nodes [
   "On Loading Errored (LazyImage_Item)"!() -> (OutputDelegate, then, ErrorCode:int=0, ModuleId:int=0, ErrorString:String, ErrorMessage:Text) {is=title K2Node_ComponentBoundEvent}
   > SetBrushFromMaterial(
    self=Get LazyImage_Item() -> (LazyImage_Item:Fort Lazy Image Object Reference)
    Material=Get ErrorStateMaterial() -> (ErrorStateMaterial:Material Instance Object Reference)
   )
   > Branch(Condition=IsValidImageReference!(ImageReference:"Fort Image Reference Structure (by ref)"=Get ItemReference() -> (ItemReference:Fort Image Reference Structure)) -> (ReturnValue:bool))
   > Set HasItemImageLoadFailed(HasItemImageLoadFailed:bool=true) -> (Output_Get:bool)
   n1 = "On Loading State Changed (LazyImage_Item)"!() -> (OutputDelegate, then, bIsLoading:bool) {is=title K2Node_ComponentBoundEvent}
   > SetVisibility(
    self=Get Background() -> (Background:Image Object Reference)
    InVisibility=Select(Option 0:ESlateVisibility Enum=SelfHitTestInvisible, Option 1:ESlateVisibility Enum=Collapsed, Index=@n1.bIsLoading) -> (ReturnValue:ESlateVisibility Enum)
   )
   n2 = SetAsyncBackground!() -> (OutputDelegate, then, MaterialInterfaceValue:Material Interface Soft Object Reference) {is=title K2Node_CustomEvent}
   n3 = Get DefaultBackgroundMaterial() -> (DefaultBackgroundMaterial:Material Interface Object Reference)
   > n4 = Async Load Asset(Asset=Select(
     Option 0:Material Interface Soft Object Reference=Make Soft Reference!(Input:Material Interface Object Reference=@n3) -> (Output:Material Interface Soft Object Reference)
     Option 1:Material Interface Soft Object Reference=@n2.MaterialInterfaceValue
     Index=IsValidSoftObjectReference(SoftObjectReference=@n2.MaterialInterfaceValue)
    ) -> (ReturnValue:Material Interface Soft Object Reference))
   n5 = Cast To MaterialInterface!(execute=@n4.Completed, Object=@n4) -> (then, CastFailed, AsMaterial Interface:Material Interface Object Reference) {is=title K2Node_DynamicCast}
   > SetBrushFromMaterial(self=Get Background(), Material=@n5)
   > n6 = GetDynamicMaterial(self=Get Background())
   > n7 = Set BackgroundDynamicMaterial(BackgroundDynamicMaterial:Material Instance Dynamic Object Reference=@n6) -> (Output_Get:Material Instance Dynamic Object Reference)
   > UpdateMaterialCorners!(execute, self, DynamicMaterialInstance:Material Instance Dynamic Object Reference=@n7) -> (then)
   n8 = Set Custom Widget!() -> (OutputDelegate, then, ItemVM:Fort Item VM Object Reference) {is=title K2Node_CustomEvent}
   > n9 = Branch(Condition=IsValid(Object=@n8.ItemVM))
   n10 = Get CustomCategoryItem!(self:Fort Locker Category Config Object Reference=GetItemCategory!(self:Fort Item VM Object Reference=@n8.ItemVM) -> (ReturnValue:Fort Locker Category Config Object Reference)) -> (CustomCategoryItem:User Widget Soft Class Reference)
   > n11 = Branch(Condition=AND Boolean(
     A=Get bIsValid!(self:Fort Item VM Object Reference=@n8.ItemVM) -> (bIsValid:bool)
     B=IsValidSoftClassReference!(SoftClassReference:"Object Soft Class Reference (by ref)"=@n10) -> (ReturnValue:bool)
    ))
   > n12 = CreateWidgetAsync!(
    execute:Exec
    UserWidgetSoftClass:User Widget Soft Class Reference=@n10
    OwningPlayer:Player Controller Object Reference=GetOwningPlayer()
    bSuspendInputUntilComplete:bool=false
   ) -> (then:Exec, OnComplete:Exec, OnFailure:Exec, UserWidget:User Widget Object Reference) {is=title K2Node_AsyncAction}
   SetViewModelByClass!(
    execute=@n12.OnComplete
    self:MVVM View Object Reference=GetExtension!(self:User Widget Object Reference=@n12, ExtensionType:User Widget Extension Class Reference=/Script/ModelViewViewModel.MVVMView) -> (ReturnValue:MVVM View Object Reference)
    NewValue:Notify Field Value Changed Interface=@n8.ItemVM
   ) -> (then, ReturnValue:bool)
   > SetContent!(
    execute
    self:Content Widget Object Reference=Get Slot_Custom() -> (Slot_Custom:Named Slot Object Reference)
    Content:Widget Object Reference=@n12
   ) -> (then, ReturnValue:Panel Slot Object Reference)
   SetContent(execute=[@n9.else, @n11.else], self=Get Slot_Custom())
   n13 = HeldAction() {is=title K2Node_CustomEvent}
   > PlaySoundHoldPressed()
   n14 = SetStatusIndicatorIconAsync!() -> (
    OutputDelegate
    then
    InMaterialInstanceSoft:Material Instance Soft Object Reference
    InStatusIndicatorObject:WBP UIKit Status Indicator Base Object Reference
   ) {is=title K2Node_CustomEvent}
   > n15 = Async Load Asset(Asset=@n14.InMaterialInstanceSoft)
   Is Valid(exec=@n15.Completed, InputObject=@n14.InStatusIndicatorObject)
   > SetIcon!(execute, self:WBP UIKit Status Indicator Base Object Reference=@n14.InStatusIndicatorObject, IconAsset:Object Reference=@n15) -> (then)
   n16 = Event PreConstruct()
   > n17 = Sequence!(execute) -> (then_0, then_1, then_2:Exec)
   > Branch(Condition=@n16.IsDesignTime)
   > PresetForDesignTime()
   > SetItemIsArchived!(execute, self, Is Archived:bool=true) -> (then)
   > SetItemHasConflict!(execute, self, HasConflict:bool=true) -> (then)
   > Set Item Is Juno!(execute, self, IsIconVisible:bool=true) -> (then) {is=title K2Node_CallFunction}
   > SetItemIsLocked!(execute, self, Is Locked:bool=true) -> (then)
   > SetItemIsFavorite!(execute, self, Is Favorite:bool=true) -> (then)
   > SetItemIsNew!(execute, self, Is New:bool=true) -> (then)
   > SetItemIsEquippedInOtherCategory!(execute, self, IsEquippedInOtherCategory:bool=true) -> (then)
   > SetItemIsTemporary!(execute, self, Is Temporary:bool=true) -> (then)
   > SetItemIsShuffled!(execute, self, Is Temporary:bool=true) -> (then)
   UnregisterAllModularBlocks(execute=@n17.then_1, self:UIKit Modular Button Object Reference)
   > RegisterModularBlock!(
    execute
    self:UIKit Modular Button Object Reference
    Block:UIKit Block Object Reference=Get Block_Outline() -> (Block_Outline:WBP UIKit Block Outline Object Reference)
   ) -> (then)
   InitializeItemCard(execute=@n17.then_2)
   Event On Pressed() {is=title K2Node_Event}
   > Call On Card Pressed!(execute:Exec, self:WBP UIKit Item Card Base Object Reference, ItemCard:WBP UIKit Item Card Base Object Reference=Self-Reference()) -> (then:Exec) {is=title K2Node_CallDelegate}
   > n18 = Branch(Condition=ShouldPlayHoldSound!() -> (Holdable:bool))
   > n19 = Set Timer by Event(Delegate=@n13, Time=0.2, bLooping=false)
   > Set HoldTimer(HoldTimer:Timer Handle Structure=@n19) -> (Output_Get:Timer Handle Structure)
   PlaySoundPressed(execute=@n18.else)
   n20 = Event OnActionProgress!() -> (OutputDelegate, then, HeldPercent:float) {is=title K2Node_Event}
   > Set CurrentHoldProgress(CurrentHoldProgress:double=@n20.HeldPercent) -> (Output_Get:double)
   Event On Released() {is=title K2Node_Event}
   > Clear and Invalidate Timer by Handle(Handle=Get HoldTimer() -> (HoldTimer:Timer Handle Structure))
   n21 = Get CurrentHoldProgress() -> (CurrentHoldProgress:double)
   > n22 = Branch(Condition=AND Boolean!(A=ShouldPlayHoldSound(), B="float > float"(A=@n21), C:bool="float < float"(A=@n21, B=1)) -> (ReturnValue))
   > StopSoundHoldReleased()
   FadeOutSound(execute=@n22.else)
   n23 = Event On Deselected()
   SetFocusState!(execute=[@n23.then, @n25.then], self, IsFocused:bool=false) -> (then)
   Event On Hovered()
   > n24 = PlaySoundHovered()
   SetFocusState(execute=[@n24.then, @n26.then], IsFocused=true)
   > Branch(Condition=Get HasItemImageLoadFailed() -> (HasItemImageLoadFailed:bool))
   > RetryItemImageLoading()
   n25 = Event On Unhovered()
   Event Destruct()
   > FadeOutSound()
   n26 = Event On Selected()
   Event On Clicked()
   > Branch(Condition=GetRequiresHold!(self:Common Button Base Object Reference) -> (ReturnValue:bool))
   > Set CurrentHoldProgress(CurrentHoldProgress=0)
   > PlaySoundHoldCompleted()
  ]}
 FadeOutSound {nodes [
   FadeOutSound() {is=title K2Node_FunctionEntry}
   > n1 = Sequence!(execute) -> (then_0, then_1)
   > n2 = Get!(execute:Exec) -> (then:Exec, else:Exec, PressedSoundAudioComponent:Audio Component Object Reference)
   > FadeOut!(
    execute
    self:Audio Component Object Reference=@n2
    FadeOutDuration:float=Get FadeOutTime() -> (FadeOutTime:double)
    FadeVolumeLevel:float=0
    FadeCurve:EAudioFaderCurve Enum=Linear
   ) -> (then)
   n3 = Get!(execute=@n1.then_1) -> (then, else, HoldSoundAudioComponent:Audio Component Object Reference)
   > FadeOut(self=@n3, FadeOutDuration=Get FadeOutTime(), FadeVolumeLevel=0, FadeCurve=Linear)
  ]}
 FindCachedStatusIndicator {nodes [
   n1 = FindCachedStatusIndicator!() -> (then, IndicatorType:E_UIKit_ItemCardStatusIndicator Enum=NewEnumerator0) {is=title K2Node_FunctionEntry}
   n2 = Find(
    TargetMap:Map of E_UIKit_ItemCardStatusIndicator Enums to WBP UIKit Status Indicator Base Object References=Get StatusIndicatorCache() -> (StatusIndicatorCache:Map of E_UIKit_ItemCardStatusIndicator Enums to WBP UIKit Status Indicator Base Object References=NewEnumerator10)
    Key:"E_UIKit_ItemCardStatusIndicator Enum (by ref)"=@n1
   ) -> (Value:WBP UIKit Status Indicator Base Object Reference)
   > Return Node!(execute, Status Indicator:WBP UIKit Status Indicator Base Object Reference=@n2, Valid:bool=@n2.ReturnValue)
  ]}
 FindStatusIndicatorPlacementContainer {nodes [
   FindStatusIndicatorPlacementContainer!() -> (then, Placement:E_UIKit_StatusIndicatorPlacement Enum=NewEnumerator0) {is=title K2Node_FunctionEntry}
   > Return Node!(
    execute
    ReturnValue:Grid Panel Object Reference=Select!(
     NewEnumerator0:Grid Panel Object Reference=Get LeftIndicators() -> (LeftIndicators:Grid Panel Object Reference)
     NewEnumerator1:Grid Panel Object Reference=Get RightIndicators() -> (RightIndicators:Grid Panel Object Reference)
     NewEnumerator2:Grid Panel Object Reference=Get IconsContainer() -> (IconsContainer:Grid Panel Object Reference)
     NewEnumerator3:Grid Panel Object Reference=Get BottomLeftIndicators() -> (BottomLeftIndicators:Grid Panel Object Reference)
     Index:E_UIKit_StatusIndicatorPlacement Enum=Get Placement() -> (Placement:E_UIKit_StatusIndicatorPlacement Enum=NewEnumerator0)
    ) -> (ReturnValue:Grid Panel Object Reference)
   )
  ]}
 GetAreItemsAgeGated {nodes [
   n1 = GetAreItemsAgeGated!() -> (then, ItemVMs:Array of Fort Item VM Object References) {is=title K2Node_FunctionEntry}
   > n2 = For Each Loop(Array:Array of Fort Item VM Object References=@n1) -> (Array Element:Fort Item VM Object Reference)
   Return Node!(execute=@n2.Completed, Is Item Gated:bool=false)
   > n3 = Cast To AthenaCosmeticItemDefinition!(execute, Object=Get ItemDefinition!(self:Fort Item VM Object Reference=@n2) -> (ItemDefinition:Fort Item Definition Object Reference)) -> (then, CastFailed, AsAthena Cosmetic Item Definition:Athena Cosmetic Item Definition Object Reference) {is=title K2Node_DynamicCast}
   > n4 = HasRatingRestrictions!(
    execute
    LocalPlayer:Local Player Object Reference=GetOwningLocalPlayer!(self:Widget Object Reference) -> (ReturnValue:Local Player Object Reference)
    CosmeticItemGatingInterface:"Cosmetic Item Gating Interface Interface (by ref)"=@n3
   ) -> (then, ReturnValue:bool)
   > Branch(Condition=@n4)
   > Return Node(Is Item Gated=true)
  ]}
 GetIsItemAgeGated {nodes [
   n1 = GetIsItemAgeGated!() -> (then, Item Def:Fort Item Definition Object Reference) {is=title K2Node_FunctionEntry}
   > n2 = Cast To AthenaCosmeticItemDefinition(Object=@n1)
   > n3 = HasRatingRestrictions(LocalPlayer=GetOwningLocalPlayer(), CosmeticItemGatingInterface=@n2)
   > Return Node(Is Item Gated=@n3)
   Return Node(execute=@n2.CastFailed, Is Item Gated=false)
  ]}
 GetIsItemAgeGatedFortCurrentExperience {nodes [
   n1 = GetIsItemAgeGatedFortCurrentExperience!() -> (then, Item Def:Fort Item Definition Object Reference) {is=title K2Node_FunctionEntry}
   > n2 = Cast To AthenaCosmeticItemDefinition(Object=@n1)
   > n3 = ConvertCosmeticItemDefinitionToInterface!(execute, CosmeticItemDefinition:Athena Cosmetic Item Definition Object Reference=@n2) -> (then, ReturnValue:Cosmetic Item Gating Interface Interface)
   > n4 = ShouldBeGatedByAgeRating!(
    execute
    InLocalPlayer:Local Player Object Reference=GetOwningLocalPlayer()
    InCosmeticItemGatingInterface:"Cosmetic Item Gating Interface Interface (by ref)"=@n3
    OptionalQueryContextTags:Gameplay Tag Container Structure
   ) -> (then, ReturnValue:bool)
   > n5 = HasRatingRestrictions(LocalPlayer=GetOwningLocalPlayer(), CosmeticItemGatingInterface=@n3)
   > Return Node(Is Item Gated=AND Boolean!(A=@n5, B=@n4) -> (ReturnValue))
   Return Node(execute=@n2.CastFailed, Is Item Gated=false)
  ]}
 InitializeItemCard {nodes [
   InitializeItemCard() {is=title K2Node_FunctionEntry}
   > n1 = GetDynamicMaterial(self=Get Background())
   > Set BackgroundDynamicMaterial(BackgroundDynamicMaterial=@n1)
   > SetItemImageReference!(execute, self, Item Image:Fort Image Reference Structure=Get ItemReference()) -> (then)
   > SetItemCardSize!(execute, self, ItemSize:Vector 2D Structure=Get ItemCardSize() -> (ItemCardSize:Vector 2D Structure)) -> (then)
   n2 = Calculate Corner Radius!(
    DefaultCornerRadius:double=Get DefaultOutlineCornerRadius() -> (DefaultOutlineCornerRadius:double)
    Offset:double=Get OutlineCornerOffset() -> (OutlineCornerOffset:double)
   ) -> (TopLeft:double=~, TopRight:double=~, BottomLeft:double=~, BottomRight:double=~, RadiusInPixels:bool=~) {is=title K2Node_MacroInstance}
   > SetCornerRadiuses!(
    execute
    self:WBP UIKit Block Outline Object Reference=Get Block_Outline()
    TopLeft:float=@n2
    TopRight:float=@n2.TopRight
    BottomLeft:float=@n2.BottomLeft
    BottomRight:float=@n2.BottomRight
    RadiusInPixels:bool=@n2.RadiusInPixels
   ) -> (then)
   n3 = Calculate Corner Radius(DefaultCornerRadius=Get DefaultEquipCornerRadius() -> (DefaultEquipCornerRadius:double)) -> (TopLeft=~, TopRight=~, BottomLeft=~, BottomRight=~, RadiusInPixels=~)
   > SetCornerRadius!(
    execute
    self:WBP UIKit Equipped State Object Reference=Get EquippedState_Outline() -> (EquippedState_Outline:WBP UIKit Equipped State Object Reference)
    TopLeftCornerRadius:float=@n3
    TopRightCornerRadius:float=@n3.TopRight
    BottomLeftCornerRadius:float=@n3.BottomLeft
    BottomRightCornerRadius:float=@n3.BottomRight
    RadiusInPixels:bool=@n3.RadiusInPixels
   ) -> (then)
  ]}
 On Preview Action Button Pressed {nodes [
   n1 = On Preview Action Button Pressed!() -> (then, InteractionNameID:Name) {is=title K2Node_FunctionEntry}
   > PreviewActionPressed!(
    execute
    self:Fort Item Preview Screen Object Reference=Get PreviewWidget() -> (PreviewWidget:Fort Item Preview Widget Object Reference)
    ActionName:Name=@n1
   ) -> (then)
  ]}
 On Preview Action Button Released {nodes [
   n1 = On Preview Action Button Released!() -> (then, InteractionNameID:Name) {is=title K2Node_FunctionEntry}
   > PreviewActionReleased!(execute, self:Fort Item Preview Screen Object Reference=Get PreviewWidget(), ActionName:Name=@n1) -> (then)
  ]}
 On Primary Preview Action Button Clicked {nodes [
   n1 = On Primary Preview Action Button Clicked!() -> (then, NameID:Name) {is=title K2Node_FunctionEntry}
   > PreviewAction!(execute, self:Fort Item Preview Screen Object Reference=Get PreviewWidget(), ActionName:Name=@n1) -> (then)
   n2 = Get Preview Actions() -> (Preview Actions:Array of Names)
   > Branch(Condition="Get (a ref)"(Array=Get ShouldClose(), Dimension 1=Find Item!(TargetArray:Array of Names=@n2, ItemToFind:"Name (by ref)"=@n1) -> (ReturnValue)))
   > Close(self:Menu Anchor Object Reference=Get Anchor!(self:UIKit Context Menu Anchor Object Reference=Get Context Menu()) -> (Anchor:Menu Anchor Object Reference))
  ]}
 On Secondary Action Button Clicked {nodes [
   On Secondary Action Button Clicked!() -> (then, NameID:Name) {is=title K2Node_FunctionEntry}
   > PreviewAction(self=Get PreviewWidget(), ActionName="Get (a ref)"(Array:Array of Names=Get Preview Actions(), Dimension 1=1) -> (Output:"Name (by ref)"))
   > Close(self=Get Anchor(self=Get Context Menu()))
  ]}
 OnCardHovered {nodes [OnCardHovered() {is=title K2Node_FunctionEntry}]}
 OnCardPressed {nodes [OnCardPressed!() -> (then, ItemCard:WBP UIKit Item Card Base Object Reference) {is=title K2Node_FunctionEntry}]}
 OnCardUnhovered {nodes [OnCardUnhovered() {is=title K2Node_FunctionEntry}]}
 OverrideStatusIndicatorBangCount {nodes [
   n1 = OverrideStatusIndicatorBangCount!() -> (then, Indicator:E_UIKit_ItemCardStatusIndicator Enum=NewEnumerator0, Value:int) {is=title K2Node_FunctionEntry}
   > n2 = FindCachedStatusIndicator!(execute, self, IndicatorType:E_UIKit_ItemCardStatusIndicator Enum=@n1) -> (then, Status Indicator:WBP UIKit Status Indicator Base Object Reference, Valid:bool) {is=title K2Node_CallFunction}
   > n3 = Branch(Condition=@n2.Valid)
   > OverrideCount!(execute, self:WBP UIKit Status Indicator Base Object Reference=@n2, Count:int=@n1.Value) -> (then)
   Add!(
    execute=@n3.else
    TargetMap:Map of E_UIKit_ItemCardStatusIndicator Enums to Integers=Get OverrideBangCounts() -> (OverrideBangCounts:Map of E_UIKit_ItemCardStatusIndicator Enums to Integers=NewEnumerator10)
    Key:"E_UIKit_ItemCardStatusIndicator Enum (by ref)"=@n1
    Value:"Integer (by ref)"=@n1.Value
   ) -> (then) {is=title K2Node_CallFunction}
  ]}
 PlaySoundHoldCompleted {nodes [
   PlaySoundHoldCompleted() {is=title K2Node_FunctionEntry}
   > n1 = Sequence()
   > n2 = Get!(execute) -> (then, else, HoldCompletedSound:Sound Base Object Reference)
   > SpawnSound2D!(
    execute
    Sound:Sound Base Object Reference=@n2
    VolumeMultiplier:float=1
    PitchMultiplier:float=1
    StartTime:float=0
    ConcurrencySettings:Sound Concurrency Object Reference
    bPersistAcrossLevelTransition:bool=true
    bAutoDestroy:bool=true
   ) -> (then, ReturnValue:Audio Component Object Reference)
   FadeOutSound(execute=@n1.then_1) {is=title K2Node_CallFunction}
  ]}
 PlaySoundHoldPressed {nodes [
   PlaySoundHoldPressed() {is=title K2Node_FunctionEntry}
   > n1 = Sequence()
   > n2 = Get!(execute) -> (then, else, HoldPressedSound:Sound Base Object Reference)
   > n3 = SpawnSound2D(Sound=@n2, VolumeMultiplier=1, PitchMultiplier=1, StartTime=0, bPersistAcrossLevelTransition=false, bAutoDestroy=true)
   > Set PressedSoundAudioComponent(PressedSoundAudioComponent:Audio Component Object Reference=@n3) -> (Output_Get:Audio Component Object Reference)
   n4 = Get!(execute=@n1.then_1) -> (then, else, HoldSound:Sound Base Object Reference)
   n5 = Get HoldSoundAudioComponent() -> (HoldSoundAudioComponent:Audio Component Object Reference)
   > n6 = Branch(Condition=IsValid(Object=@n5))
   > n7 = Branch(Condition=IsPlaying!(self:Audio Component Object Reference=@n5) -> (ReturnValue:bool))
   > n8 = Stop(self:Audio Component Object Reference=@n5)
   n9 = SpawnSound2D(
    execute=[@n8.then, @n7.else, @n6.else]
    Sound=@n4
    VolumeMultiplier=1
    PitchMultiplier=1
    StartTime=0
    bPersistAcrossLevelTransition=false
    bAutoDestroy=true
   )
   > Set HoldSoundAudioComponent(HoldSoundAudioComponent:Audio Component Object Reference=@n9) -> (Output_Get:Audio Component Object Reference)
  ]}
 PlaySoundHovered {nodes [
   PlaySoundHovered() {is=title K2Node_FunctionEntry}
   > n1 = Get!(execute) -> (then, else, HoveredSound:Sound Base Object Reference)
   > SpawnSound2D(Sound=@n1, VolumeMultiplier=1, PitchMultiplier=1, StartTime=0, bPersistAcrossLevelTransition=false, bAutoDestroy=true)
  ]}
 PlaySoundPressed {nodes [
   PlaySoundPressed() {is=title K2Node_FunctionEntry}
   > n1 = Get!(execute) -> (then, else, PressedSound:Sound Base Object Reference)
   > SpawnSound2D(Sound=@n1, VolumeMultiplier=1, PitchMultiplier=1, StartTime=0, bPersistAcrossLevelTransition=false, bAutoDestroy=true)
  ]}
 PresetForDesignTime {nodes [
   PresetForDesignTime() {is=title K2Node_FunctionEntry}
   n1 = Get Slot_Main() -> (Slot_Main:Named Slot Object Reference)
   > SetVisibility(
    self=@n1
    InVisibility=Select!(
     Option 0:ESlateVisibility Enum=Collapsed
     Option 1:ESlateVisibility Enum=SelfHitTestInvisible
     Index:bool=HasAnyChildren!(self:Panel Widget Object Reference=@n1) -> (ReturnValue:bool)
    ) -> (ReturnValue:ESlateVisibility Enum)
   )
   n2 = Get Slot_Background() -> (Slot_Background:Named Slot Object Reference)
   > SetVisibility(self=@n2, InVisibility=Select(Option 0=Collapsed, Option 1=SelfHitTestInvisible, Index=HasAnyChildren(self=@n2)))
  ]}
 Refresh Equipped Visibility {nodes [
   Refresh Equipped Visibility() {is=title K2Node_FunctionEntry}
   > SetVisibility(
    self=Get EquippedState_Outline()
    InVisibility=Select(Option 0=Collapsed, Option 1=SelfHitTestInvisible, Index=Get Is Equipped() -> (Is Equipped:bool))
   )
   n1 = Get Is Equipped()
   > n2 = Branch(Condition=AND Boolean(A=NOT Boolean(A=@n1), B=Get Is Secondary Equipped() -> (Is Secondary Equipped:bool)))
   > n3 = SetVisibility(self=Get EquippedState_Outline(), InVisibility=SelfHitTestInvisible)
   SetEquippedState!(
    execute=[@n3.then, @n2.else]
    self:WBP UIKit Equipped State Object Reference=Get EquippedState_Outline()
    IsCurrentSlotEquipped:bool=Get Is Equipped()
    DisplaySlot:bool=Get ShowEquipNumbers() -> (ShowEquipNumbers:bool)
    EquippedSlot:Text="To Text (Integer)"(Value=Get Equipped Slot() -> (Equipped Slot:int=0))
    SlotImage:Soft Object Reference=Get Slot Icon() -> (Slot Icon:Soft Object Reference)
   ) -> (then)
   n4 = Get EquippedSection!(self:WBP UIKit Equipped State Object Reference=Get EquippedState_Outline()) -> (EquippedSection:Common Border Object Reference)
   > SetGridSlotNudge!(
    execute
    Widget:Widget Object Reference=Get BottomLeftIndicators()
    Nudge:Vector 2D Structure=Select(Option 0:Vector 2D Structure, Option 1:Vector 2D Structure=(X=35,Y=0), Index=IsVisible(self=@n4)) -> (ReturnValue:Vector 2D Structure)
   ) -> (then)
  ]}
 RemoveStatusIndicator {nodes [
   n1 = RemoveStatusIndicator!() -> (then, Indicator:E_UIKit_ItemCardStatusIndicator Enum=NewEnumerator0) {is=title K2Node_FunctionEntry}
   n2 = Find(TargetMap=Get StatusIndicatorCache() -> (StatusIndicatorCache=NewEnumerator10), Key=@n1)
   > Branch(Condition=@n2.ReturnValue)
   > RemoveFromParent(self=@n2)
   > Remove!(
    execute
    TargetMap:Map of E_UIKit_ItemCardStatusIndicator Enums to WBP UIKit Status Indicator Base Object References=Get StatusIndicatorCache() -> (StatusIndicatorCache=NewEnumerator10)
    Key:"E_UIKit_ItemCardStatusIndicator Enum (by ref)"=@n1
   ) -> (then, ReturnValue:bool)
  ]}
 ResetAnimationState {nodes [
   ResetAnimationState() {is=title K2Node_FunctionEntry}
   > StopAllAnimations(self:User Widget Object Reference)
   > PlayAnimationTimeRange!(
    execute
    self:User Widget Object Reference
    InAnimation:Widget Animation Object Reference=Get Anim_Hovered_Base() -> (Anim_Hovered_Base:Widget Animation Object Reference)
    StartAtTime:float=0
    EndAtTime:float=0
    NumLoopsToPlay:int=1
    PlayMode:EUMGSequencePlayMode Enum
    PlaybackSpeed:float=1
    bRestoreState:bool=false
   ) -> (then, ReturnValue:Widget Animation Handle Structure)
   > ResetAnimationState(self:WBP UIKit Block Outline Object Reference=Get Block_Outline()) {is=title K2Node_CallFunction}
  ]}
 ResetIconStates {nodes [
   ResetIconStates() {is=title K2Node_FunctionEntry}
   > n1 = Sequence!(execute) -> (then_0, then_1, then_2:Exec, then_3:Exec, then_4:Exec, then_5:Exec, then_6:Exec, then_7:Exec, then_8:Exec)
   > Set IsEquippedStateSet(IsEquippedStateSet:bool=false) -> (Output_Get:bool)
   Set IsLockedStateSet(execute=@n1.then_1, IsLockedStateSet:bool=false) -> (Output_Get:bool)
   Set IsFavoriteStateSet(execute=@n1.then_2, IsFavoriteStateSet:bool=false) -> (Output_Get:bool)
   Set IsArchiveStateSet(execute=@n1.then_3, IsArchiveStateSet:bool=false) -> (Output_Get:bool)
   Set IsConflictStateSet(execute=@n1.then_4, IsConflictStateSet:bool=false) -> (Output_Get:bool)
   Set IsNewStateSet(execute=@n1.then_5, IsNewStateSet:bool=false) -> (Output_Get:bool)
   Set IsAgeGateStateSet(execute=@n1.then_6, IsAgeGateStateSet:bool=false) -> (Output_Get:bool)
   Set IsTemporaryStateSet(execute=@n1.then_7, IsTemporaryStateSet:bool=false) -> (Output_Get:bool)
   Set IsCustomizeType(execute=@n1.then_8, IsCustomizeType:bool=false) -> (Output_Get:bool)
   Set IsShuffleStateSet(IsShuffleStateSet:bool=false) -> (Output_Get:bool)
  ]}
 RetryItemImageLoading {nodes [
   RetryItemImageLoading() {is=title K2Node_FunctionEntry}
   n1 = Get LazyImage_Item()
   > Branch(Condition=NOT Boolean(A=IsLoading!(self:Common Lazy Image Object Reference=@n1) -> (ReturnValue:bool)))
   > SetItemImageReference(Item Image=Get ItemReference())
  ]}
 SelectCard {nodes [SelectCard() {is=title K2Node_FunctionEntry}]}
 Set Context Menu Action {nodes [
   n1 = Set Context Menu Action!() -> (
    then
    Context Menu:WBP UIKit Context Menu Anchor Button Base Object Reference
    Preview Action ID:Name
    Preview Action Text:Text
    In Action Info Sort Priority:int
    In Action Info Interaction:Data Table Row Handle Structure
    In Action Info Pressed Delegate:Delegate
    In Action Info Released Delegate:Delegate
    Preview Action IteractionDelegate:"Delegate (by ref)"
   ) {is=title K2Node_FunctionEntry}
   n2 = Get InteractionContext(self=@n1)
   > RemoveInteraction(self=@n2, InteractionNameID=@n1.Preview Action ID)
   > AddInteraction!(
    execute
    self:UIKit Context Menu Action Context Object Reference=@n2
    InActionInfo_DisplayName:Text=@n1.Preview Action Text
    InActionInfo_IconTexture:Texture 2D Soft Object Reference
    InActionInfo_IconMaterial:Material Instance Soft Object Reference
    InActionInfo_SortPriority:int=@n1.In Action Info Sort Priority
    InActionInfo_bIsAvailable:bool=True
    InActionInfo_bIsEnabled:bool=True
    InActionInfo_bIsSelected:bool=False
    InActionInfo_bAddDivider:bool=False
    InActionInfo_bPositionDividerAfter:bool=False
    InActionInfo_bCloseMenuOnSelect:bool=False
    InActionInfo_InteractionNameID:Name=@n1.Preview Action ID
    InActionInfo_Interaction:Data Table Row Handle Structure=@n1.In Action Info Interaction
    InActionInfo_PressedDelegate:Delegate=@n1.In Action Info Pressed Delegate
    InActionInfo_ReleasedDelegate:Delegate=@n1.In Action Info Released Delegate
    InInteractionDelegate:"Delegate (by ref)"=@n1.Preview Action IteractionDelegate
   ) -> (then)
   > SetContext!(
    execute
    self:WBP UIKit Context Menu Anchor Button Base Object Reference=@n1
    InteractionContext_Data:UIKit Context Menu Action Context Object Reference=@n2
   ) -> (then)
  ]}
 Set Default Background {nodes [
   n1 = Set Default Background!() -> (then, Color 1:Linear Color Structure, Color 2:Linear Color Structure, GradientStart:float=~, GradientEnd:float=~, GradientAngle:float=~) {is=title K2Node_FunctionEntry}
   n2 = Get BackgroundDynamicMaterial() -> (BackgroundDynamicMaterial:Material Instance Dynamic Object Reference)
   > SetVectorParameterValue(self=@n2, ParameterName=BgColor_1, Value=@n1)
   > SetVectorParameterValue(self=@n2, ParameterName=BgColor_2, Value=@n1.Color 2)
   > SetScalarParameterValue(self=@n2, ParameterName=BgGradientStart, Value=@n1.GradientStart)
   > SetScalarParameterValue(self=@n2, ParameterName=BgGradientEnd, Value=@n1.GradientEnd)
   > SetScalarParameterValue(self=@n2, ParameterName=BgGradientAngle, Value=@n1.GradientAngle)
  ]}
 Set Item Color {nodes [
   n1 = Set Item Color!() -> (then, VariantColor:Slate Color Structure) {is=title K2Node_FunctionEntry}
   > SetScalarParameterValue(self=Get ItemDynamicMaterial() -> (ItemDynamicMaterial:Material Instance Dynamic Object Reference), ParameterName=RemoveItem, Value=1)
   n2 = Get BackgroundDynamicMaterial()
   > SetScalarParameterValue(self=@n2, ParameterName=SingleColorBg, Value=1)
   > SetScalarParameterValue(self=@n2, ParameterName=UseDefaultBg, Value=0)
   > SetVectorParameterValue(
    self=@n2
    ParameterName=BGColor_1
    Value=Break Slate Color!(SlateColor:"Slate Color Structure (by ref)"=@n1) -> (SpecifiedColor:Linear Color Structure, ColorUseRule:ESlateColorStylingMode Enum) {is=title K2Node_BreakStruct}
   )
  ]}
 Set Item Is Juno {nodes [
   n1 = Set Item Is Juno!() -> (then, IsIconVisible:bool=true) {is=title K2Node_FunctionEntry}
   > Set Status Indicator Visibility!(execute, self, Indicator:E_UIKit_ItemCardStatusIndicator Enum=NewEnumerator4, Visibility:bool=@n1) -> (then) {is=title K2Node_CallFunction}
  ]}
 Set Status Indicator Visibility {nodes [
   n1 = Set Status Indicator Visibility!() -> (then, Indicator:E_UIKit_ItemCardStatusIndicator Enum=NewEnumerator0, Visibility:bool=~) {is=title K2Node_FunctionEntry}
   > Set L_IndicatorType(L_IndicatorType:E_UIKit_ItemCardStatusIndicator Enum=@n1) -> (Output_Get:E_UIKit_ItemCardStatusIndicator Enum=NewEnumerator10)
   n2 = Find(
    TargetMap:Map of E_UIKit_ItemCardStatusIndicator Enums to S UIKit Status Indicator Spawn Params Structures=Get StatusIndicators() -> (StatusIndicators:Map of E_UIKit_ItemCardStatusIndicator Enums to S UIKit Status Indicator Spawn Params Structures=NewEnumerator10)
    Key=Get L_IndicatorType() -> (L_IndicatorType:E_UIKit_ItemCardStatusIndicator Enum=NewEnumerator10)
   ) -> (Value:S UIKit Status Indicator Spawn Params Structure)
   > n3 = Branch(Condition=@n2.ReturnValue)
   > n4 = Set L_SpawnParams(L_SpawnParams:S UIKit Status Indicator Spawn Params Structure=@n2) -> (Output_Get:S UIKit Status Indicator Spawn Params Structure)
   > n5 = Branch(Condition=@n1.Visibility)
   > Set L_PlacementContainer(L_PlacementContainer:Grid Panel Object Reference=FindStatusIndicatorPlacementContainer!(Placement:E_UIKit_StatusIndicatorPlacement Enum=Break S UIKit Status Indicator Spawn Params!(S_UIKit_StatusIndicatorSpawnParams:"S UIKit Status Indicator Spawn Params Structure (by ref)"=@n4) -> (
      Placement_36_#2:E_UIKit_StatusIndicatorPlacement Enum
      Background_47_#3:Material Instance Soft Object Reference
     ) {is=title K2Node_BreakStruct}) -> (ReturnValue:Grid Panel Object Reference) {is=title K2Node_CallFunction}) -> (Output_Get:Grid Panel Object Reference)
   > n6 = Sequence!(execute) -> (then_0, then_1, then_2)
   > Branch(Condition=Contains!(
     TargetMap:Map of E_UIKit_ItemCardStatusIndicator Enums to WBP UIKit Status Indicator Base Object References=Get StatusIndicatorCache() -> (StatusIndicatorCache=NewEnumerator10)
     Key:"E_UIKit_ItemCardStatusIndicator Enum (by ref)"=Get L_IndicatorType() -> (L_IndicatorType=NewEnumerator10)
    ) -> (ReturnValue))
   > Return Node!(execute)
   n7 = Get L_SpawnedStatusIndicator() -> (L_SpawnedStatusIndicator:WBP UIKit Status Indicator Base Object Reference)
   Set Bang State!(execute=@n6.then_2, self:Fort Bang Wrapper NUI Object Reference=@n7, bEnabled:bool=true, Count:int=0) -> (then) {is=title K2Node_CallFunction}
   n8 = Break S UIKit Status Indicator Spawn Params!(S_UIKit_StatusIndicatorSpawnParams=Get L_SpawnParams() -> (L_SpawnParams:S UIKit Status Indicator Spawn Params Structure)) -> (
    Configuration_2_#4:E_UI_StatusIndicator_Configuration Enum
    Size_9_#5:E_UI_StatusIndicator_Size Enum
    Type_11_#6:E_UI_StatusIndicator_Type Enum
    Icon_28_#7:Material Instance Soft Object Reference
    Placement_36_#2
    OverrideColumn_39_#8:int
    Background_47_#3
   )
   > SetIndicatorConfiguration!(
    execute
    self:WBP UIKit Status Indicator Base Object Reference=@n7
    InCount:int=0
    IndicatorConfig:E_UI_StatusIndicator_Configuration Enum=@n8
   ) -> (then)
   > UpdateSize!(
    execute
    self:WBP UIKit Status Indicator Base Object Reference=@n7
    IconSize:E_UI_StatusIndicator_Size Enum=@n8.Size_9_#5
   ) -> (then)
   > SetIndicatorType!(
    execute
    self:WBP UIKit Status Indicator Base Object Reference=@n7
    IndicatorType:E_UI_StatusIndicator_Type Enum=@n8.Type_11_#6
   ) -> (then)
   > SetPadding!(
    execute
    self:User Widget Object Reference=@n7
    InPadding_Left:float=0
    InPadding_Top:float=0
    InPadding_Right:float=Select(
     Option 0:float=0
     Option 1:float=2
     Index="Equal (Enum)"(
      A:E_UIKit_StatusIndicatorPlacement Enum=@n8.Placement_36_#2
      B:E_UIKit_StatusIndicatorPlacement Enum=NewEnumerator3
     )
    ) -> (ReturnValue:float)
    InPadding_Bottom:float=0
   ) -> (then)
   > SetAsyncBackground!(
    execute
    self:WBP UIKit Status Indicator Base Object Reference=@n7
    MaterialInterfaceValue:Material Interface Soft Object Reference=@n8.Background_47_#3
   ) -> (then) {is=title K2Node_CallFunction}
   > n9 = GetConsoleVariableBoolValue!(execute, VariableName:String=EnableAsyncLoadForStatusIndicatorIcon) -> (then, ReturnValue:bool)
   > n10 = Branch(Condition=NOT Boolean(A=@n9))
   > n11 = LoadAsset_Blocking(Asset=@n8.Icon_28_#7)
   > n12 = SetIcon(self=@n7, IconAsset=@n11)
   n13 = Find(
    TargetMap:Map of E_UIKit_ItemCardStatusIndicator Enums to Integers=Get OverrideBangCounts() -> (OverrideBangCounts=NewEnumerator10)
    Key=Get L_IndicatorType() -> (L_IndicatorType=NewEnumerator10)
   ) -> (Value:int)
   Branch(execute=[@n12.then, @n14.then], Condition=@n13.ReturnValue)
   > OverrideCount(self=Get L_SpawnedStatusIndicator(), Count=@n13)
   n14 = SetStatusIndicatorIconAsync!(
    execute=@n10.else
    self
    InMaterialInstanceSoft:Material Instance Soft Object Reference=@n8.Icon_28_#7
    InStatusIndicatorObject:WBP UIKit Status Indicator Base Object Reference=Get L_SpawnedStatusIndicator()
   ) -> (then) {is=title K2Node_CallFunction}
   n15 = Create Widget!(
    execute:Exec=@n6.then_1
    Class:User Widget Class Reference=/FortUILibrary/UIKit/StatusIndicator/WBP_UIKit_StatusIndicator.WBP_UIKit_StatusIndicator_C
    OwningPlayer:Player Controller Object Reference=GetOwningPlayer()
   ) -> (then:Exec, ReturnValue:WBP UIKit Status Indicator Object Reference)
   > n16 = Set L_SpawnedStatusIndicator(L_SpawnedStatusIndicator:WBP UIKit Status Indicator Base Object Reference=@n15) -> (Output_Get:WBP UIKit Status Indicator Base Object Reference)
   > AddChildToGrid!(
    execute
    self:Grid Panel Object Reference=Get L_PlacementContainer() -> (L_PlacementContainer:Grid Panel Object Reference)
    Content:Widget Object Reference=@n16
    InRow:int=0
    InColumn:int=Break S UIKit Status Indicator Spawn Params!(S_UIKit_StatusIndicatorSpawnParams=Get L_SpawnParams()) -> (OverrideColumn_39_#8, Background_47_#3)
   ) -> (then, ReturnValue:Grid Slot Object Reference)
   > Add(
    TargetMap:Map of E_UIKit_ItemCardStatusIndicator Enums to WBP UIKit Status Indicator Base Object References=Get StatusIndicatorCache() -> (StatusIndicatorCache=NewEnumerator10)
    Key=Get L_IndicatorType() -> (L_IndicatorType=NewEnumerator10)
    Value:"WBP UIKit Status Indicator Base Object Reference (by ref)"=Get L_SpawnedStatusIndicator()
   )
   RemoveStatusIndicator!(execute=@n5.else, self, Indicator:E_UIKit_ItemCardStatusIndicator Enum=Get L_IndicatorType() -> (L_IndicatorType=NewEnumerator10)) -> (then) {is=title K2Node_CallFunction}
   Return Node(execute=@n3.else)
  ]}
 SetBackgroundBlendMode {nodes [
   n1 = SetBackgroundBlendMode!() -> (then, UseAdaptiveColor:bool=~) {is=title K2Node_FunctionEntry}
   > Set UseAdaptiveBackground(UseAdaptiveBackground:bool=@n1) -> (Output_Get:bool)
   n2 = Get UseAdaptiveBackground() -> (UseAdaptiveBackground:bool)
   > SetScalarParameterValue(self=Get BackgroundDynamicMaterial(), ParameterName=Brightness Percentage, Value=Select(Option 0=1, Option 1=1, Index=@n2))
   > SetScalarParameterValue(self=Get BackgroundDynamicMaterial(), ParameterName=Additive Percentage, Value=Select(Option 0=1, Option 1=0.4, Index=@n2))
  ]}
 SetCornerRadius {nodes [
   n1 = SetCornerRadius!() -> (
    then
    UseOverrideValues:bool=~
    RadiusInPixels:bool=true
    CornerRadius_TopLeft:double=20
    CornerRadius_TopRight:double=20
    CornerRadius_BottomLeft:double=20
    CornerRadius_BottomRight:double=20
    OutlineCornerOffset:double=20
   ) {is=title K2Node_FunctionEntry}
   > Set OverrideCornerRadius(OverrideCornerRadius:bool=@n1) -> (Output_Get:bool)
   > Set RadiusInPixels(RadiusInPixels:bool=@n1.RadiusInPixels) -> (Output_Get:bool)
   > Set CornerRadius_TopLeft(CornerRadius_TopLeft:double=@n1.CornerRadius_TopLeft) -> (Output_Get:double)
   > Set CornerRadius_TopRight(CornerRadius_TopRight:double=@n1.CornerRadius_TopRight) -> (Output_Get:double)
   > Set CornerRadius_BottomLeft(CornerRadius_BottomLeft:double=@n1.CornerRadius_BottomLeft) -> (Output_Get:double)
   > Set CornerRadius_BottomRight(CornerRadius_BottomRight:double=@n1.CornerRadius_BottomRight) -> (Output_Get:double)
   > Set OutlineCornerOffset(OutlineCornerOffset:double=@n1.OutlineCornerOffset) -> (Output_Get:double)
   > UpdateMaterialCorners(DynamicMaterialInstance=Get ItemDynamicMaterial())
   > UpdateMaterialCorners(DynamicMaterialInstance=Get BackgroundDynamicMaterial())
   n2 = Calculate Corner Radius(DefaultCornerRadius=Get DefaultOutlineCornerRadius(), Offset=Get OutlineCornerOffset()) -> (TopLeft=~, TopRight=~, BottomLeft=~, BottomRight=~, RadiusInPixels=~)
   > SetCornerRadiuses(
    self=Get Block_Outline()
    TopLeft=@n2
    TopRight=@n2.TopRight
    BottomLeft=@n2.BottomLeft
    BottomRight=@n2.BottomRight
    RadiusInPixels=@n2.RadiusInPixels
   )
   n3 = Calculate Corner Radius(DefaultCornerRadius=Get DefaultEquipCornerRadius()) -> (TopLeft=~, TopRight=~, BottomLeft=~, BottomRight=~, RadiusInPixels=~)
   > SetCornerRadius(
    self=Get EquippedState_Outline()
    TopLeftCornerRadius=@n3
    TopRightCornerRadius=@n3.TopRight
    BottomLeftCornerRadius=@n3.BottomLeft
    BottomRightCornerRadius=@n3.BottomRight
    RadiusInPixels=@n3.RadiusInPixels
   ) {is=title K2Node_CallFunction}
  ]}
 SetFocusState {nodes [
   n1 = SetFocusState!() -> (then, IsFocused:bool=~) {is=title K2Node_FunctionEntry}
   > Set IsButtonFocused(IsButtonFocused:bool=@n1) -> (Output_Get:bool)
   > StopAnimation(InAnimation=Get Anim_Hovered_Base())
   > n2 = Branch(Condition=@n1)
   > RequestTransitionOnBlocks!(execute, self:UIKit Modular Button Object Reference, State:EUIKitBlockVisualState Enum=Focused) -> (then)
   n3 = Get Anim_Hovered_Base()
   > PlayAnimationForward(InAnimation=@n3)
   RequestTransitionOnBlocks(execute=@n2.else, State=Unfocused)
   > PlayAnimationReverse(InAnimation=@n3)
  ]}
 SetGridAlignment {nodes [
   n1 = SetGridAlignment!() -> (then, WIdget:Widget Object Reference, Horizontal:EHorizontalAlignment Enum=HAlign_Fill, Vertical:EVerticalAlignment Enum=VAlign_Fill) {is=title K2Node_FunctionEntry}
   > n2 = Cast To GridSlot!(execute, Object=Get Slot!(self:Widget Object Reference=@n1) -> (Slot:Panel Slot Object Reference)) -> (then, CastFailed, AsGrid Slot:Grid Slot Object Reference) {is=title K2Node_DynamicCast}
   > SetHorizontalAlignment(self:Grid Slot Object Reference=@n2, InHorizontalAlignment=@n1.Horizontal)
   > SetVerticalAlignment(self:Grid Slot Object Reference=@n2, InVerticalAlignment=@n1.Vertical)
  ]}
 SetGridColumnAndRow {nodes [
   n1 = SetGridColumnAndRow!() -> (then, WIdget:Widget Object Reference, Row:int, Row Span:int, Column:int, Column Span:int) {is=title K2Node_FunctionEntry}
   > n2 = Cast To GridSlot(Object=Get Slot(self=@n1))
   > SetColumn!(execute, self:Grid Slot Object Reference=@n2, InColumn:int=@n1.Column) -> (then)
   > SetColumnSpan!(execute, self:Grid Slot Object Reference=@n2, InColumnSpan:int=@n1.Column Span) -> (then)
   > SetRow!(execute, self:Grid Slot Object Reference=@n2, InRow:int=@n1.Row) -> (then)
   > SetRowSpan!(execute, self:Grid Slot Object Reference=@n2, InRowSpan:int=@n1.Row Span) -> (then)
  ]}
 SetGridLayer {nodes [
   n1 = SetGridLayer!() -> (then, WIdget:Widget Object Reference, Layer:int) {is=title K2Node_FunctionEntry}
   > n2 = Cast To GridSlot(Object=Get Slot(self=@n1))
   > SetLayer!(execute, self:Grid Slot Object Reference=@n2, InLayer:int=@n1.Layer) -> (then)
  ]}
 SetGridNudge {nodes [
   n1 = SetGridNudge!() -> (then, WIdget:Widget Object Reference, Nudge:Vector 2D Structure) {is=title K2Node_FunctionEntry}
   > n2 = Cast To GridSlot(Object=Get Slot(self=@n1))
   > SetNudge!(execute, self:Grid Slot Object Reference=@n2, InNudge:Vector 2D Structure=@n1.Nudge) -> (then)
  ]}
 SetGridPadding {nodes [
   n1 = SetGridPadding!() -> (then, WIdget:Widget Object Reference, Padding:Margin Structure) {is=title K2Node_FunctionEntry}
   > n2 = Cast To GridSlot(Object=Get Slot(self=@n1))
   > SetPadding!(execute, self:Grid Slot Object Reference=@n2, InPadding:Margin Structure=@n1.Padding) -> (then)
  ]}
 SetIsItemCustomizable {nodes [
   n1 = SetIsItemCustomizable!() -> (then, IsCustomizable:bool=~) {is=title K2Node_FunctionEntry}
   > Set Status Indicator Visibility(Indicator=NewEnumerator12, Visibility=@n1) {is=title K2Node_CallFunction}
   > Set IsCustomizeType(IsCustomizeType=true)
  ]}
 SetIsItemCustomizableCharacter {nodes [
   n1 = SetIsItemCustomizableCharacter!() -> (then, IsCustomizableCharacter:bool=~) {is=title K2Node_FunctionEntry}
   > Set Status Indicator Visibility(Indicator=NewEnumerator13, Visibility=@n1)
  ]}
 SetItemAnimationScale {nodes [
   n1 = SetItemAnimationScale!() -> (then, Item Scale:"Float (double-precision) (by ref)") {is=title K2Node_FunctionEntry}
   n2 = Get ItemDynamicMaterial()
   > SetScalarParameterValue(self=@n2, ParameterName=ScaleFrom, Value=@n1)
   > SetScalarParameterValue(self=@n2, ParameterName=ScaleTo, Value=float + float(A=@n1, B=Get ItemScaleDelta() -> (ItemScaleDelta:double)))
  ]}
 SetItemCardSize {nodes [
   n1 = SetItemCardSize!() -> (then, ItemSize:Vector 2D Structure) {is=title K2Node_FunctionEntry}
   > n2 = Set ItemCardSize(ItemCardSize:Vector 2D Structure=@n1) -> (Output_Get:Vector 2D Structure)
   n3 = BreakVector2D(InVec=@n2)
   n4 = Get SizeBox() -> (SizeBox:Size Box Object Reference)
   > SetWidthOverride(self=@n4, InWidthOverride=@n3)
   > SetHeightOverride(self=@n4, InHeightOverride=@n3.Y)
  ]}
 SetItemHasConflict {nodes [
   n1 = SetItemHasConflict!() -> (then, HasConflict:bool=~) {is=title K2Node_FunctionEntry}
   > Set Status Indicator Visibility(Indicator=NewEnumerator3, Visibility=@n1)
   > Set IsConflictStateSet(IsConflictStateSet=true)
  ]}
 SetItemImage {nodes [
   n1 = SetItemImage!() -> (then, Item Image:Soft Object Reference) {is=title K2Node_FunctionEntry}
   > SetItemImageReference(Item Image=Make Fort Image Reference!(Image:Soft Object Reference=@n1, ImageUrl:String, MaterialTextureParamName:Name=None) -> (FortImageReference:Fort Image Reference Structure) {is=title K2Node_MakeStruct})
  ]}
 SetItemImageReference {nodes [
   n1 = SetItemImageReference!() -> (then, Item Image:Fort Image Reference Structure) {is=title K2Node_FunctionEntry}
   > Set HasItemImageLoadFailed(HasItemImageLoadFailed=false)
   > n2 = Set ItemReference(ItemReference:Fort Image Reference Structure=@n1) -> (Output_Get:Fort Image Reference Structure)
   n3 = Break Fort Image Reference!(FortImageReference:"Fort Image Reference Structure (by ref)"=@n2) -> (Image:Soft Object Reference, ImageUrl:String, MaterialTextureParamName:Name) {is=title K2Node_BreakStruct}
   n4 = InEditor!() -> (ReturnValue:bool)
   n5 = NOT Boolean(A=IsAssetNull!(Asset:Soft Object Reference=@n3) -> (ReturnValue:bool))
   > n6 = Set members in Fort Image Reference!(
    execute:Exec
    StructRef:"Fort Image Reference Structure (by ref)"=Copy!(Item:Fort Image Reference Structure=@n2) -> (ReturnValue:Fort Image Reference Structure) {is=title K2Node_Copy}
    Image:Soft Object Reference=Select(
     Option 0:"Soft Object Reference (by ref)"=Get Fallback Image() -> (Fallback Image:Soft Object Reference)
     Option 1:"Soft Object Reference (by ref)"=@n3
     Index=OR Boolean(A=@n5, B=@n4)
    ) -> (ReturnValue:"Soft Object Reference (by ref)")
   ) -> (then:Exec, StructOut:"Fort Image Reference Structure (by ref)") {is=title K2Node_SetFieldsInStruct}
   > SetItemMaterial!(
    execute
    self
    Material Instance:"Material Instance Object Reference (by ref)"=Get ItemMaterial() -> (ItemMaterial:Material Instance Object Reference)
   ) -> (then)
   > SetMaterialTextureParamName!(execute, self:Common Lazy Image Object Reference=Get LazyImage_Item(), TextureParamName:Name=Item) -> (then)
   > SetBrushFromImageReference!(
    execute
    self:Fort Lazy Image Object Reference=Get LazyImage_Item()
    ImageReference:"Fort Image Reference Structure (by ref)"=@n6
    bMatchSize:bool=false
    DownloadPriority:ECMSDownloadPriority Enum=Normal
   ) -> (then)
  ]}
 SetItemIsAgeGated {nodes [
   n1 = SetItemIsAgeGated!() -> (then, IsAgeGated:bool=~) {is=title K2Node_FunctionEntry}
   > SetVisibility(
    self=Get CosmeticGating_Icon() -> (CosmeticGating_Icon:WBP Cosmetic Gating Warning Icon Object Reference)
    InVisibility=Select(Option 0:ESlateVisibility Enum=Collapsed, Option 1:ESlateVisibility Enum=SelfHitTestInvisible, Index=@n1) -> (ReturnValue:ESlateVisibility Enum)
   )
   > Set IsAgeGateStateSet(IsAgeGateStateSet=true)
  ]}
 SetItemIsArchived {nodes [
   n1 = SetItemIsArchived!() -> (then, Is Archived:bool=~) {is=title K2Node_FunctionEntry}
   > Set Status Indicator Visibility(Indicator=NewEnumerator0, Visibility=@n1)
   > Set IsArchiveStateSet(IsArchiveStateSet=true)
  ]}
 SetItemIsEnhanced {nodes [
   n1 = SetItemIsEnhanced!() -> (then, Is Enhanced:bool=~) {is=title K2Node_FunctionEntry}
   > SetIsEnhanced!(
    execute
    self:WBP UIKit Transference Object Reference=Get UIKit_Transference() -> (UIKit_Transference:WBP UIKit Transference Object Reference)
    Is Enhanced:bool=@n1
   ) -> (then)
  ]}
 SetItemIsEquipped {nodes [
   n1 = SetItemIsEquipped!() -> (then, Is Equipped:bool=~) {is=title K2Node_FunctionEntry}
   > Set Is Equipped(Is Equipped:bool=@n1) -> (Output_Get:bool)
   > Set IsEquippedStateSet(IsEquippedStateSet=true)
   > Refresh Equipped Visibility() {is=title K2Node_CallFunction}
  ]}
 SetItemIsEquippedInOtherCategory {nodes [
   n1 = SetItemIsEquippedInOtherCategory!() -> (then, IsEquippedInOtherCategory:bool=~) {is=title K2Node_FunctionEntry}
   > Set Status Indicator Visibility(Indicator=NewEnumerator7, Visibility=@n1)
  ]}
 SetItemIsEquippedSecondary {nodes [
   SetItemIsEquippedSecondary!() -> (then, Is Equipped:bool=~, ShowEquippedSlotIndex:bool=~, EquippedSlotIndex:int, SlotIcon:Soft Object Reference) {is=title K2Node_FunctionEntry}
   > Set Is Secondary Equipped(Is Secondary Equipped:bool=Get Is Equipped()) -> (Output_Get:bool)
   > Set Equipped Slot(Equipped Slot:int=Get EquippedSlotIndex() -> (EquippedSlotIndex:int=0)) -> (Output_Get:int=0)
   > Set ShowEquipNumbers(ShowEquipNumbers:bool=Get ShowEquippedSlotIndex() -> (ShowEquippedSlotIndex:bool)) -> (Output_Get:bool)
   > Set Slot Icon(Slot Icon:Soft Object Reference=Get SlotIcon() -> (SlotIcon:Soft Object Reference)) -> (Output_Get:Soft Object Reference)
   > Refresh Equipped Visibility()
  ]}
 SetItemIsFavorite {nodes [
   n1 = SetItemIsFavorite!() -> (then, Is Favorite:bool=~) {is=title K2Node_FunctionEntry}
   > Set Status Indicator Visibility(Indicator=NewEnumerator1, Visibility=@n1)
   > Set IsFavoriteStateSet(IsFavoriteStateSet=true)
  ]}
 SetItemIsLocked {nodes [
   n1 = SetItemIsLocked!() -> (then, Is Locked:bool=~) {is=title K2Node_FunctionEntry}
   > Set Status Indicator Visibility(Indicator=NewEnumerator2, Visibility=@n1)
   > Set IsLockedStateSet(IsLockedStateSet=true)
  ]}
 SetItemIsMultiGame {nodes [SetHasTransference!(
    execute
    self:WBP UIKit Transference Object Reference=Get UIKit_Transference()
    Has Transference:bool=SetItemIsMultiGame!() -> (then, Has Transference:bool=~) {is=title K2Node_FunctionEntry}
   ) -> (then)]}
 SetItemIsNew {nodes [
   n1 = SetItemIsNew!() -> (then, Is New:bool=~) {is=title K2Node_FunctionEntry}
   > Set Status Indicator Visibility(Indicator=NewEnumerator5, Visibility=@n1)
   > Set IsNewStateSet(IsNewStateSet=true)
  ]}
 SetItemIsRestricted {nodes [
   n1 = SetItemIsRestricted!() -> (then, IsCompatible:bool=~) {is=title K2Node_FunctionEntry}
   > SetOpacity!(execute, self:Image Object Reference=Get LazyImage_Item(), InOpacity:float=SelectFloat(A=0.25, B=1, bPickA=@n1)) -> (then)
  ]}
 SetItemIsShuffled {nodes [
   n1 = SetItemIsShuffled!() -> (then, Is Temporary:bool=~) {is=title K2Node_FunctionEntry}
   > Set Status Indicator Visibility(Indicator=NewEnumerator11, Visibility=@n1)
   > Set IsShuffleStateSet(IsShuffleStateSet=true)
  ]}
 SetItemIsTemporary {nodes [
   n1 = SetItemIsTemporary!() -> (then, Is Temporary:bool=~) {is=title K2Node_FunctionEntry}
   > Set Status Indicator Visibility(Indicator=NewEnumerator8, Visibility=@n1)
   > Set IsTemporaryStateSet(IsTemporaryStateSet=true)
  ]}
 SetItemMaterial {nodes [
   n1 = SetItemMaterial!() -> (then, Material Instance:"Material Instance Object Reference (by ref)") {is=title K2Node_FunctionEntry}
   > n2 = Set ItemMaterial(ItemMaterial:Material Instance Object Reference=@n1) -> (Output_Get:Material Instance Object Reference)
   n3 = Get LazyImage_Item()
   > SetBrushFromMaterial(self=@n3, Material=@n2)
   > n4 = GetDynamicMaterial(self=@n3)
   > n5 = Set ItemDynamicMaterial(ItemDynamicMaterial:Material Instance Dynamic Object Reference=@n4) -> (Output_Get:Material Instance Dynamic Object Reference)
   > UpdateMaterialCorners(DynamicMaterialInstance=@n5)
   > Branch(Condition=Get OverrideCornerRadius())
   > UpdateMaterialCorners(DynamicMaterialInstance=Get BackgroundDynamicMaterial())
  ]}
 ShouldPlayHoldSound {nodes [
   ShouldPlayHoldSound() {is=title K2Node_FunctionEntry}
   > n1 = Branch(Condition=Get bRequiresHold() -> (bRequiresHold:bool))
   > Return Node!(execute, Holdable:bool=true)
   n2 = Get!(execute=@n1.else) -> (then, else, InputActionWidget:Common Action Widget Object Reference)
   > n3 = Branch(Condition=IsInputType!(InputType:ECommonInputType Enum=Gamepad, IsNot:bool=true) -> (Result:bool))
   n4 = IsHeldAction!(self:Common Action Widget Object Reference=@n2) -> (ReturnValue:bool)
   > n5 = Branch(Condition=@n4)
   > Return Node(Holdable=NOT Boolean(A=IsHovered()))
   Return Node(execute=[@n5.else, @n2.else], Holdable=false)
   Return Node(execute=@n3.else, Holdable=@n4)
  ]}
 StopSoundHoldReleased {nodes [
   StopSoundHoldReleased() {is=title K2Node_FunctionEntry}
   > n1 = Sequence!(execute) -> (then_0, then_1)
   > n2 = Get!(execute) -> (then, else, HoldReleasedSound:Sound Base Object Reference)
   > SpawnSound2D(Sound=@n2, VolumeMultiplier=1, PitchMultiplier=1, StartTime=0, bPersistAcrossLevelTransition=false, bAutoDestroy=true)
   FadeOutSound(execute=@n1.then_1)
  ]}
 UpdateMaterialCorners {nodes [
   UpdateMaterialCorners!() -> (then, DynamicMaterialInstance:Material Instance Dynamic Object Reference) {is=title K2Node_FunctionEntry}
   n1 = Calculate Corner Radius(DefaultCornerRadius=Get!() -> (
     CornerRadius_Backplate_11_#9:double=~
     CornerRadius_InteractiveComponent_12_#10:double=~
     CornerRadius_Button_13_#11:double=~
     CornerRadius_Tab_14_#12:double=~
     CornerRadius_Circle_15_#13:double=~
    ).CornerRadius_InteractiveComponent_12_#10) -> (TopLeft=~, TopRight=~, BottomLeft=~, BottomRight=~, RadiusInPixels=~)
   n2 = Get DynamicMaterialInstance() -> (DynamicMaterialInstance:Material Instance Dynamic Object Reference)
   > SetScalarParameterValue(self=@n2, ParameterName=UseValueInPixels, Value="To Float (Boolean)"(InBool=@n1.RadiusInPixels))
   > SetScalarParameterValue(self=@n2, ParameterName=TopLeftCornerRadius, Value=@n1)
   > SetScalarParameterValue(self=@n2, ParameterName=TopRightCornerRadius, Value=@n1.TopRight)
   > SetScalarParameterValue(self=@n2, ParameterName=BottomLeftCornerRadius, Value=@n1.BottomLeft)
   > SetScalarParameterValue(self=@n2, ParameterName=BottomRightCornerRadius, Value=@n1.BottomRight)
  ]}
]
variables [
 BackgroundDynamicMaterial {type=MaterialInstanceDynamic}
 Context Menu {type=WBP_UIKit_ContextMenuAnchorButton_Base_C, default=null}
 CornerRadius_BottomLeft {type=number, default=20}
 CornerRadius_BottomRight {type=number, default=20}
 CornerRadius_TopLeft {type=number, default=20}
 CornerRadius_TopRight {type=number, default=20}
 CurrentHoldProgress {type=number}
 DefaultBackgroundMaterial {type=MaterialInterface, default=/FortUILibrary/Systems/Items/MaterialInstances/MI_UIKit_Item_Background}
 Equipped Slot {type=integer}
 ErrorStateMaterial {type=MaterialInstance, default=/FortUILibrary/Systems/Items/MaterialInstances/MI_UIKit_Item_Error}
 FadeOutTime {type=number, default=0.2}
 Fallback Image {type=Object, default=null}
 HasItemImageLoadFailed {type=boolean}
 HoldCompletedSound {
  type=SoundBase
  default=/Game/Sounds/Fort_UI_Sounds/FrontEnd/DefaultSounds/HoldableButtons/Holdable_Loud/MSS_UI_Button_Holdable_Loud_Complete
 }
 HoldPressedSound {type=SoundBase, default=/Game/Sounds/Fort_UI_Sounds/FrontEnd/DefaultSounds/HoldableButtons/Holdable_Loud/MSS_UI_Button_Holdable_Loud_Press}
 HoldReleasedSound {type=SoundBase, default=/Game/Sounds/Fort_UI_Sounds/FrontEnd/DefaultSounds/HoldableButtons/Holdable_Loud/MSS_UI_Button_Holdable_Loud_Release}
 HoldSound {type=SoundBase, default=/Game/Sounds/Fort_UI_Sounds/FrontEnd/DefaultSounds/HoldableButtons/Holdable_Loud/MSS_UI_Button_Holdable_Loud_Hold_lp}
 HoldSoundAudioComponent {type=AudioComponent, default=null}
 HoldTimer {type=TimerHandle}
 HoveredSound {type=SoundBase, default=/Game/Sounds/Fort_UI_Sounds/FrontEnd/New_FrontEnd/DefaultSounds/PressedButtons/FN_UI_Default_Button_Hovered_Cue}
 Is Equipped {type=boolean}
 Is Secondary Equipped {type=boolean}
 IsAgeGateStateSet {type=boolean}
 IsArchiveStateSet {type=boolean}
 IsButtonFocused {type=boolean}
 IsConflictStateSet {type=boolean}
 IsCustomizeType {type=boolean}
 IsEquippedStateSet {type=boolean}
 IsFavoriteStateSet {type=boolean}
 IsLockedStateSet {type=boolean}
 IsNewStateSet {type=boolean}
 IsShuffleStateSet {type=boolean}
 IsTemporaryStateSet {type=boolean}
 ItemCardSize {type=Vector2D, default=<108 136>}
 ItemDynamicMaterial {type=MaterialInstanceDynamic}
 ItemImage {type=Object, default=null}
 ItemMaterial {type=MaterialInstance, default=/FortUILibrary/Systems/Items/MaterialInstances/MI_UIKit_ItemPreview}
 ItemReference {type=FortImageReference, default {image=null, imageUrl="", materialTextureParamName=None}}
 ItemScaleDelta {type=number, default=0.1}
 MobileItemSize {type=Vector2D, default=<128 264>}
 MobileRewardItemSize {type=Vector2D, default=<64 128>}
 OnCardHovered {}
 OnCardPressed {}
 OnCardUnhovered {}
 OutlineCornerOffset {type=number, default=10}
 OverrideBangCounts {default {}}
 OverrideCornerRadius {type=boolean}
 PressedSound {type=SoundBase, default=/Game/Sounds/Fort_UI_Sounds/FrontEnd/New_FrontEnd/DefaultSounds/PressedButtons/FN_UI_Default_Button_Pressed_Cue}
 PressedSoundAudioComponent {type=AudioComponent, default=null}
 Preview Actions {type="array<string>", default []}
 PreviewWidget {type=FortItemPreviewWidget, default=null}
 RadiusInPixels {type=boolean, default=true}
 ShouldClose {type="array<boolean>", default []}
 ShowEquipNumbers {type=boolean}
 ShowSmallAgeGateIcon {type=boolean}
 Slot Icon {type=Object, default=null}
 StatusIndicatorCache {default {}}
 StatusIndicators {default {
   Archived {
    type=Utility
    background=null
    configuration=Icon
    icon=/FortUILibrary/Textures/MaterialInstances/MI_UIKit_Icon_Archive
    overrideColumn=1
    placement=TopRight
    size=Default
   }
   Crew {
    type=Utility
    background=null
    configuration=Icon
    icon=/FortUILibrary/Textures/MaterialInstances/MI_UIKit_Icon_CrewLoan
    overrideColumn=2
    placement=TopLeft
    size=Compact
   }
   Customizable {
    type=Creative
    background=/FortUILibrary/Systems/StatusIndicator/MaterialInstances/MI_UIKit_CustomIndicator_Background
    configuration=Icon
    icon=/FortUILibrary/Textures/MaterialInstances/MI_UI_Icon_Customizable
    overrideColumn=0
    placement=BottomLeft
    size=Default
   }
   CustomizableCharacter {
    type=Creative
    background=&MI_UIKit_CustomIndicator_Background
    configuration=Icon
    icon=/FortUILibrary/Textures/MaterialInstances/MI_UI_Icon_CustomizableCharacter
    overrideColumn=0
    placement=TopRight
    size=Compact
   }
   EquippedInOtherCategory {
    type=Utility
    background=null
    configuration=Icon
    icon=/FortUILibrary/Textures/MaterialInstances/MI_UIKit_Icon_LobbyImage
    overrideColumn=0
    placement=TopRight
    size=Default
   }
   Favorite {
    type=Utility
    background=null
    configuration=Icon
    icon=/FortUILibrary/Textures/MaterialInstances/MI_UIKit_Icon_Favorite
    overrideColumn=1
    placement=TopRight
    size=Default
   }
   HasConflict {
    type=Utility
    background=null
    configuration=Icon
    icon=/FortUILibrary/Textures/MaterialInstances/MI_UIKit_Icon_Warning
    overrideColumn=0
    placement=Center
    size=Default
   }
   Juno {
    type=Utility
    background=null
    configuration=Icon
    icon=/FortUILibrary/Textures/MaterialInstances/MI_UIKit_Icon_JunoStyle
    overrideColumn=1
    placement=TopLeft
    size=Compact
   }
   Locked {
    type=Utility
    background=null
    configuration=Icon
    icon=/FortUILibrary/Textures/MaterialInstances/MI_UIKit_Icon_Locked
    overrideColumn=0
    placement=Center
    size=Default
   }
   New {
    type=Attention
    background=null
    configuration=Icon
    icon=/FortUILibrary/Textures/MaterialInstances/MI_UIKit_Icon_ExclamationMark
    overrideColumn=2
    placement=TopRight
    size=Compact
   }
   Shuffle {
    type=Utility
    background=null
    configuration=Icon
    icon=/FortUILibrary/Textures/MaterialInstances/MI_UIKit_Icon_Shuffle
    overrideColumn=0
    placement=BottomLeft
    size=Default
   }
   Temporary {
    type=Utility
    background=null
    configuration=Icon
    icon=/FortUILibrary/Textures/MaterialInstances/MI_UIKit_Icon_UpdatePending_Small
    overrideColumn=1
    placement=BottomLeft
    size=Default
   }
  }}
 UseAdaptiveBackground {type=boolean, default=true}
]
widgets {
 info {inheritedWidgetCount=1, parentClass=/Script/FortniteUI.UIKitHoldableModularButton, rootWidgetClass=&SizeBox, widgetCount=19}
 widgets [
  SizeBox {
   widgetClassPath=&SizeBox
   bIsVariable=true
   widget=SizeBox
   properties {bOverride_HeightOverride=true, bOverride_WidthOverride=true, heightOverride=160, widthOverride=128}
  }
  Root {
   widgetClassPath=&GridPanel
   bIsVariable=true
   parent=SizeBox
   widget=Root
   properties {columnFill [1], rowFill [1]}
   slot {class=SizeBoxSlot}
  }
  Slot_Background {widgetClassPath=&NamedSlot, bIsVariable=true, parent=Root, widget=Slot_Background, slot {class=GridSlot, properties {layer=1}}}
  Background {
   widgetClassPath=&Image
   bIsVariable=true
   parent=Root
   widget=Background
   properties {
    visibility=SelfHitTestInvisible
    brush {
     bIsDynamicallyLoaded=false
     drawAs=Image
     imageType=NoImage
     mirroring=NoMirror
     resourceName=None
     resourceObject=&MI_UIKit_Item_Background
     tiling=NoTile
     imageSize=<32 32>
     margin=<ltrb 0 0 0 0>
     outlineSettings {
      bUseBrushTransparency=false
      roundingType=HalfHeightRadius
      width=0
      color {colorUseRule=UseColor_Specified, specifiedColor=<rgba 0 0 0 0>}
      cornerRadii=<0 0 0 0>
     }
     tintColor {colorUseRule=UseColor_Specified, specifiedColor=<rgba 1 1 1 1>}
     uVRegion {bIsValid=false, max=<0 0>, min=<0 0>}
    }
   }
   slot {class=GridSlot}
  }
  GridPanel_0 {
   widgetClassPath=&GridPanel
   parent=Root
   widget=GridPanel_0
   properties {columnFill [1], rowFill [1]}
   slot {class=GridSlot, properties {layer=2}}
  }
  LazyImage_Item {
   widgetClassPath=&FortLazyImage
   bIsVariable=true
   parent=GridPanel_0
   widget=LazyImage_Item
   properties {
    brush {
     bIsDynamicallyLoaded=false
     drawAs=Image
     imageType=FullColor
     mirroring=NoMirror
     resourceName=None
     resourceObject=&MI_UIKit_Item_Error
     tiling=NoTile
     imageSize=<32 32>
     margin=<ltrb 0 0 0 0>
     outlineSettings {
      bUseBrushTransparency=false
      roundingType=HalfHeightRadius
      width=0
      color {colorUseRule=UseColor_Specified, specifiedColor=<rgba 0 0 0 0>}
      cornerRadii=<0 0 0 0>
     }
     tintColor {colorUseRule=UseColor_Specified, specifiedColor=<rgba 1 1 1 1>}
     uVRegion {bIsValid=false, max=<0 0>, min=<0 0>}
    }
    loadingBackgroundBrush {
     bIsDynamicallyLoaded=false
     drawAs=Image
     imageType=FullColor
     mirroring=NoMirror
     resourceName=None
     resourceObject=/FortUILibrary/Systems/Items/MaterialInstances/MI_UIKit_Item_Loading
     tiling=NoTile
     imageSize=<32 32>
     margin=<ltrb 0 0 0 0>
     outlineSettings {
      bUseBrushTransparency=false
      roundingType=HalfHeightRadius
      width=0
      color {colorUseRule=UseColor_Specified, specifiedColor=<rgba 0 0 0 0>}
      cornerRadii=<0 0 0 0>
     }
     tintColor {colorUseRule=UseColor_Specified, specifiedColor=<rgba 1 1 1 1>}
     uVRegion {bIsValid=false, max=<0 0>, min=<0 0>}
    }
    loadingThrobberBrush {
     bIsDynamicallyLoaded=false
     drawAs=Image
     imageType=FullColor
     mirroring=NoMirror
     resourceName=None
     resourceObject=/Game/Athena/UI/Frontend/Spatial/Art/T_Ui_Empty
     tiling=NoTile
     imageSize=<16 16>
     margin=<ltrb 0 0 0 0>
     outlineSettings {
      bUseBrushTransparency=false
      roundingType=HalfHeightRadius
      width=0
      color {colorUseRule=UseColor_Specified, specifiedColor=<rgba 0 0 0 0>}
      cornerRadii=<0 0 0 0>
     }
     tintColor {colorUseRule=UseColor_Specified, specifiedColor=<rgba 1 1 1 1>}
     uVRegion {bIsValid=false, max=<0 0>, min=<0 0>}
    }
   }
   slot {class=GridSlot}
  }
  Slot_BelowImage {
   widgetClassPath=&NamedSlot
   bIsVariable=true
   parent=GridPanel_0
   widget=Slot_BelowImage
   slot {class=GridSlot, properties {layer=9, row=1}}
  }
  Slot_Custom {widgetClassPath=&NamedSlot, bIsVariable=true, parent=Root, widget=Slot_Custom, slot {class=GridSlot, properties {layer=3}}}
  LeftIndicators {
   widgetClassPath=&GridPanel
   bIsVariable=true
   parent=Root
   widget=LeftIndicators
   slot {class=GridSlot, properties {horizontalAlignment=HAlign_Left, layer=5, verticalAlignment=VAlign_Top, padding=<ltrb 8 8 8 8>}}
  }
  CosmeticGating_Icon {
   widgetClassPath=/Game/Athena/UI/Frontend/CosmeticGating/WBP_CosmeticGating_WarningIcon
   bIsVariable=true
   parent=LeftIndicators
   widget=CosmeticGating_Icon
   properties {visibility=Collapsed, iconSize=<28 28>}
   slot {class=GridSlot, properties {horizontalAlignment=HAlign_Left, verticalAlignment=VAlign_Bottom}}
  }
  RightIndicators {
   widgetClassPath=&GridPanel
   bIsVariable=true
   parent=Root
   widget=RightIndicators
   slot {class=GridSlot, properties {horizontalAlignment=HAlign_Right, layer=4, verticalAlignment=VAlign_Top, padding=<ltrb 8 8 8 8>}}
  }
  BottomLeftIndicators {
   widgetClassPath=&GridPanel
   bIsVariable=true
   parent=Root
   widget=BottomLeftIndicators
   slot {class=GridSlot, properties {horizontalAlignment=HAlign_Left, layer=4, verticalAlignment=VAlign_Bottom, padding=<ltrb 8 8 8 8>}}
  }
  IconsContainer {
   widgetClassPath=&GridPanel
   bIsVariable=true
   parent=Root
   widget=IconsContainer
   properties {columnFill [1], rowFill [1]}
   slot {class=GridSlot, properties {layer=6}}
  }
  UIKit_Transference {
   widgetClassPath=&/WBP_UIKit_Transference
   bIsVariable=true
   parent=IconsContainer
   widget=UIKit_Transference
   slot {class=GridSlot, properties {horizontalAlignment=HAlign_Right, verticalAlignment=VAlign_Bottom}}
  }
  Compatibility_Icon {
   widgetClassPath=&Image
   bIsVariable=true
   parent=IconsContainer
   widget=Compatibility_Icon
   properties {
    visibility=Collapsed
    brush {
     bIsDynamicallyLoaded=false
     drawAs=Image
     imageType=NoImage
     mirroring=NoMirror
     resourceName=None
     resourceObject=null
     tiling=NoTile
     imageSize=<0 0>
     margin=<ltrb 0 0 0 0>
     outlineSettings {
      bUseBrushTransparency=false
      roundingType=HalfHeightRadius
      width=0
      color {colorUseRule=UseColor_Specified, specifiedColor=<rgba 0 0 0 0>}
      cornerRadii=<0 0 0 0>
     }
     tintColor {colorUseRule=UseColor_Specified, specifiedColor=<rgba 1 1 1 1>}
     uVRegion {bIsValid=false, max=<0 0>, min=<0 0>}
    }
    colorAndOpacity=<rgba 1 1 1 0>
   }
   slot {class=GridSlot, properties {horizontalAlignment=HAlign_Right, verticalAlignment=VAlign_Bottom}}
  }
  Block_Outline {
   widgetClassPath=/FortUILibrary/UIKit/Buttons/Blocks/WBP_UIKit_Block_Outline
   bIsVariable=true
   parent=Root
   widget=Block_Outline
   properties {bottomLeftCornerRadius=30, bottomRightCornerRadius=30, radiusInPixels=true, topLeftCornerRadius=30, topRightCornerRadius=30}
   slot {class=GridSlot, properties {layer=9}}
  }
  Slot_Main {widgetClassPath=&NamedSlot, bIsVariable=true, parent=Root, widget=Slot_Main, slot {class=GridSlot, properties {layer=10}}}
  EquippedState_Outline {
   widgetClassPath=&/WBP_UIKit_EquippedState
   bIsVariable=true
   parent=Root
   widget=EquippedState_Outline
   slot {class=GridSlot, properties {layer=11, padding=<ltrb -2 -2 -2 -2>}}
  }
  InputActionWidget {widgetClassPath=&CommonActionWidget, bInherited=true, widget=null}
 ]
}
