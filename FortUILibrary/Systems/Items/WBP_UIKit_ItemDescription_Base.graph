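# Base UIKit widget for item descriptions, showing title, rarity/category/style tag, rich-text detail lines, compatibility tags and a bottom slot.
# PreConstruct sets design-time placeholders or clears them, then applies exposed styling, grid rows, padding and alignment, mirrored for Arabic.
# Text setters collapse empty or hidden lines and refresh the body container.
# SetCompatibilityTags maps product tags to cosmetic-config names and spawns a tag widget for each displayable one.
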WidgetBlueprint /FortUILibrary/Systems/Items/WBP_UIKit_ItemDescription_Base
graphs [
 AddCompatibilityTags {nodes [
   n1 = AddCompatibilityTags!() -> (then, GameTags:Array of Texts) {is=title K2Node_FunctionEntry}
   > n2 = Set GameTags(GameTags:Array of Texts=@n1) -> (Output_Get:Array of Texts)
   > SetVisibility(
    self=Get Text_Compatibility() -> (Text_Compatibility:Common Text Object Reference)
    InVisibility=Select(Option 0=SelfHitTestInvisible, Option 1=Collapsed, Index=Is Empty!(TargetArray:Array of Texts=@n2) -> (ReturnValue:bool))
   )
   > ClearChildren(self=Get ModeTagsBox() -> (ModeTagsBox:Wrap Box Object Reference))
   > n3 = For Each Loop(Array:Array of Texts=Get GameTags() -> (GameTags:Array of Texts)) -> (Array Element:Text)
   > Set CurrentArrayElement(CurrentArrayElement:Text=@n3) -> (Output_Get:Text)
   > n4 = Create Widget!(
    execute:Exec
    Class:User Widget Class Reference=Get GameTagClass() -> (GameTagClass:WBP UIKit Tag Base Class Reference)
    OwningPlayer:Player Controller Object Reference=GetOwningPlayer()
   ) -> (then:Exec, ReturnValue:WBP UIKit Tag Base Object Reference)
   > n5 = AddChildToWrapBox!(execute, self:Wrap Box Object Reference=Get ModeTagsBox(), Content:Widget Object Reference=@n4) -> (then, ReturnValue:Wrap Box Slot Object Reference)
   > SetPadding!(execute, self:Wrap Box Slot Object Reference=@n5, InPadding:Margin Structure=Make Margin(Left=0, Top=0, Right=8, Bottom=0)) -> (then)
   > SetTagType!(
    execute
    self:WBP UIKit Tag Base Object Reference=@n4
    Type:E_UI_TagType Enum=Get GameTagType() -> (GameTagType:E_UI_TagType Enum=NewEnumerator0)
   ) -> (then)
   > SetLabelText!(execute, self:WBP UIKit Tag Base Object Reference=@n4, Text:Text) -> (then)
   > SetTagText!(execute, self:WBP UIKit Tag Base Object Reference=@n4, Text:Text=Get CurrentArrayElement() -> (CurrentArrayElement:Text)) -> (then)
   > SetTagSize!(execute, self:WBP UIKit Tag Base Object Reference=@n4, TagSize:E_UI_TagSize Enum=NewEnumerator0) -> (then)
  ]}
 ClearPlaceholderData {nodes [
   ClearPlaceholderData() {is=title K2Node_FunctionEntry}
   > "SetText (Text)"(self=Get Text_Title() -> (Text_Title:Common Text Object Reference))
   > SetText(self=[
     Get Text_GameplayTags() -> (Text_GameplayTags:Common Rich Text Block Object Reference)
     Get Text_Season() -> (Text_Season:Common Rich Text Block Object Reference)
     Get Text_Set() -> (Text_Set:Common Rich Text Block Object Reference)
     Get Text_Variance() -> (Text_Variance:Common Rich Text Block Object Reference)
     Get Text_Body() -> (Text_Body:Common Rich Text Block Object Reference)
    ])
   > SetTagText(self=Get RarityTag() -> (RarityTag:WBP UIKit Tag Object Reference))
   > SetLabelText(self=Get RarityTag())
   > SetStyleText!(execute, self:WBP UIKit Tag Base Object Reference=Get RarityTag(), Text:Text) -> (then)
   > AddCompatibilityTags!(execute, self, GameTags:Array of Texts) -> (then) {is=title K2Node_CallFunction}
   n1 = Get Slot_Bottom() -> (Slot_Bottom:Named Slot Object Reference)
   > SetVisibility(
    self=@n1
    InVisibility=Select(Option 0=Collapsed, Option 1=SelfHitTestInvisible, Index=HasAnyChildren!(self:Panel Widget Object Reference=@n1) -> (ReturnValue:bool))
   )
   > Set MaxDesiredHeight!(
    execute
    MaxDesiredHeight:float=Get TitleMaxHeight() -> (TitleMaxHeight:float)
    self:Size Box Object Reference=Get SizeBoxTitle() -> (SizeBoxTitle:Size Box Object Reference)
   ) -> (then, Output_Get:float)
  ]}
 EventGraph {nodes [
   n1 = Event PreConstruct()
   > n2 = Sequence!(execute) -> (then_0, then_1, then_2:Exec)
   > n3 = Branch(Condition=@n1.IsDesignTime)
   > PresetForDesignTime()
   ClearPlaceholderData(execute=@n3.else) {is=title K2Node_CallFunction}
   InitializeWidget(execute=@n2.then_1)
   SetCompactUIStyle(execute=@n2.then_2) {is=title K2Node_CallFunction}
   Event Construct()
   > SetUpDisplayInternalItemName()
   Event On Widget Styled() {is=title K2Node_Event}
   > InitializeWidget()
  ]}
 InitializeWidget {nodes [
   InitializeWidget() {is=title K2Node_FunctionEntry}
   > n1 = Sequence!(execute) -> (then_0, then_1, then_2, then_3:Exec, then_4:Exec, then_5:Exec)
   SetWidgetLayout(execute=@n1.then_5)
   > SetTagsOrientation!(execute, self, Orientation:EOrientation Enum=Get TagsOrientation() -> (TagsOrientation:EOrientation Enum=Orient_Horizontal)) -> (then)
   n2 = Get EquippedTag() -> (EquippedTag:WBP UIKit Tag Object Reference)
   SetTagSize(execute=@n1.then_0, self=@n2, TagSize=Get TagSize() -> (TagSize:E_UI_TagSize Enum=NewEnumerator0))
   > SetTagBackgroundColor!(
    execute
    self:WBP UIKit Tag Base Object Reference=@n2
    Color:Linear Color Structure=Get EquippedTagBackgroundColor() -> (EquippedTagBackgroundColor:Linear Color Structure)
   ) -> (then)
   > SetTagTextColor!(
    execute
    self:WBP UIKit Tag Base Object Reference=@n2
    Color:Linear Color Structure=Get EquippedTagForegroundColor() -> (EquippedTagForegroundColor:Linear Color Structure)
   ) -> (then)
   n3 = Get TitleWrap() -> (TitleWrap:E_UIKit_ItemDescription_TextWrapType Enum=NewEnumerator0)
   n4 = Get Text_Title()
   SetScrollingEnabled!(
    execute=@n1.then_2
    self:Common Text Object Reference=@n4
    bInIsScrollingEnabled:bool=Select!(
     NewEnumerator0:bool=false
     NewEnumerator2:bool=false
     NewEnumerator1:bool=true
     NewEnumerator3:bool=true
     Index:E_UIKit_ItemDescription_TextWrapType Enum=@n3
    ) -> (ReturnValue:bool)
   ) -> (then)
   > SetAutoWrapText!(
    execute
    self:Text Object Reference=@n4
    InAutoTextWrap:bool=Select(NewEnumerator0=false, NewEnumerator2=true, NewEnumerator1=true, NewEnumerator3=false, Index=@n3)
   ) -> (then)
   > SetScrollOrientation!(
    execute
    self:Common Text Object Reference=@n4
    InScrollOrientation:EOrientation Enum=Select(
     NewEnumerator0:EOrientation Enum=Orient_Horizontal
     NewEnumerator2:EOrientation Enum=Orient_Vertical
     NewEnumerator1:EOrientation Enum=Orient_Vertical
     NewEnumerator3:EOrientation Enum=Orient_Horizontal
     Index=@n3
    ) -> (ReturnValue:EOrientation Enum)
   ) -> (then)
   > SetFont(self=@n4, InFontInfo=Get TitleFont() -> (TitleFont:Slate Font Info Structure))
   > SetColorAndOpacity(self=@n4, InColorAndOpacity=Get TitleColor() -> (TitleColor:Slate Color Structure))
   n5 = Get Text_GameplayTags()
   n6 = Get Text_Season()
   n7 = Get Text_Set()
   n8 = Get Text_Variance()
   n9 = Get Text_Body()
   SetDefaultColorAndOpacity!(
    execute=@n1.then_3
    self:Rich Text Block Object Reference=[@n5, @n6, @n7, @n8, @n9]
    InColorAndOpacity:Slate Color Structure=Get BodyTextColor() -> (BodyTextColor:Slate Color Structure)
   ) -> (then)
   n10 = Get BodyTextCustomTextStyle() -> (BodyTextCustomTextStyle:Data Table Object Reference)
   > n11 = Is Valid(InputObject=@n10)
   > n12 = SetTextStyleSet!(execute, self:Rich Text Block Object Reference=[@n5, @n6, @n7, @n8, @n9], NewTextStyleSet:Data Table Object Reference=@n10) -> (then)
   RefreshVisibility!(
    execute=[@n12.then, @n11.Is Not Valid]
    self
    TextBlock:Common Rich Text Block Object Reference=@n9
    Hide:bool=Get HideBodyText() -> (HideBodyText:bool)
   ) -> (then)
   > RefreshVisibility(TextBlock=@n5, Hide=Get HideGameplayTagsText() -> (HideGameplayTagsText:bool))
   > RefreshVisibility(TextBlock=@n6, Hide=Get HideSeasonText() -> (HideSeasonText:bool))
   > RefreshVisibility(TextBlock=@n7, Hide=Get HideSetText() -> (HideSetText:bool))
   > RefreshVisibility(TextBlock=@n8, Hide=Get HideVarianceText() -> (HideVarianceText:bool))
   n13 = SetCompatibilityVisibility!(execute=@n1.then_4, self, Show:bool=Get HideCompatibility() -> (HideCompatibility:bool)) -> (then, IsVisible:bool)
   > Branch(Condition=@n13)
   > SetColorAndOpacity(self=Get Text_Compatibility(), InColorAndOpacity=Get CompatibilityTitleColor() -> (CompatibilityTitleColor:Slate Color Structure))
   SetRarityTagVisibility!(execute=@n1.then_1, self, ShowTag:bool=NOT Boolean(A=Get HideRarityTag() -> (HideRarityTag:bool))) -> (then)
   > SetCategoryTextVisibility!(execute, self, ShowText:bool=NOT Boolean(A=Get HideCategoryText() -> (HideCategoryText:bool))) -> (then)
   > SetStyleTextVisibility!(execute, self, ShowText:bool=NOT Boolean(A=Get HideStyleText() -> (HideStyleText:bool))) -> (then)
   n14 = Get HideCategoryText()
   n15 = Get HideStyleText()
   n16 = Get HideRarityTag()
   n17 = Get RarityTag()
   > SetVisibility(
    self=@n17
    InVisibility=Select!(
     Option 0:ESlateVisibility Enum=SelfHitTestInvisible
     Option 1:ESlateVisibility Enum=Collapsed
     Index:bool=AND Boolean!(A=@n16, B=@n14, C:bool=@n15) -> (ReturnValue)
    ) -> (ReturnValue:ESlateVisibility Enum)
   )
   > SetLabelTextColor!(
    execute
    self:WBP UIKit Tag Base Object Reference=@n17
    Color:Linear Color Structure=Get TagTitleColor() -> (TagTitleColor:Linear Color Structure)
   ) -> (then)
   > SetTagSize(self=@n17, TagSize=Get TagSize() -> (TagSize=NewEnumerator0))
  ]}
 PresetForDesignTime {nodes [
   PresetForDesignTime() {is=title K2Node_FunctionEntry}
   > "SetText (Text)"(self=Get Text_Title(), InText=Format Text(Format="INVTEXT(\"Item long long name, believe me, it\\'s really long\")"))
   > SetText(self=Get Text_Body(), InText=Format Text(Format="INVTEXT(\"[PH] Item Description\")"))
   > SetText(self=Get Text_Set(), InText=Format Text(Format="INVTEXT(\"[PH] Item Set\")"))
   > SetText(self=Get Text_Season(), InText=Format Text(Format="INVTEXT(\"[PH] Item Season\")"))
   > SetText(self=Get Text_GameplayTags(), InText=Format Text(Format="INVTEXT(\"[PH] Item Tags\")"))
   > SetText(self=Get Text_Variance(), InText=Format Text(Format="INVTEXT(\"[PH] Item Variance\")"))
   > SetTagText(self=Get RarityTag(), Text=Format Text(Format="INVTEXT(\"Rarity\")"))
   > SetLabelText(self=Get RarityTag(), Text=Format Text(Format="INVTEXT(\"Category\")"))
   > SetStyleText(self=Get RarityTag(), Text=Format Text(Format="INVTEXT(\"Style\")"))
   > "SetText (Text)"(self=Get Text_Compatibility(), InText=Format Text(Format="INVTEXT(\"This item works with:\")"))
   > AddCompatibilityTags(GameTags=Get GameTags())
   n1 = Get Slot_Bottom()
   > SetVisibility(self=@n1, InVisibility=Select(Option 0=Collapsed, Option 1=SelfHitTestInvisible, Index=HasAnyChildren(self=@n1)))
   > Set MaxDesiredHeight(MaxDesiredHeight=Get TitleMaxHeight(), self=Get SizeBoxTitle())
  ]}
 RefreshBodyContainerVisibility {nodes [
   RefreshBodyContainerVisibility() {is=title K2Node_FunctionEntry}
   > n1 = For Each Loop with Break!(
    Exec:Exec
    Array:Array of Widget Object References=GetAllChildren(self=Get VBox_Body() -> (VBox_Body:Vertical Box Object Reference))
    Break:Exec=@n2.then
   ) -> (LoopBody:Exec, Array Element:Widget Object Reference, Array Index:int, Completed:Exec)
   SetVisibility(
    execute=@n1.Completed
    self=Get VBox_Body()
    InVisibility=Select(Option 0=Collapsed, Option 1=SelfHitTestInvisible, Index=Get HasAnyVisibleChildren() -> (HasAnyVisibleChildren:bool))
   )
   Branch(execute=@n1.LoopBody, Condition=IsVisible(self=@n1))
   > n2 = Set HasAnyVisibleChildren(HasAnyVisibleChildren:bool=true) -> (Output_Get:bool)
  ]}
 RefreshVisibility {nodes [
   n1 = RefreshVisibility!() -> (then, TextBlock:Common Rich Text Block Object Reference, Hide:bool=~) {is=title K2Node_FunctionEntry}
   n2 = TextIsEmpty(InText=GetText!(self:Rich Text Block Object Reference=@n1) -> (ReturnValue:Text))
   > SetVisibility(self=@n1, InVisibility=Select(Option 0=HitTestInvisible, Option 1=Collapsed, Index=OR Boolean(A=@n2, B=@n1.Hide)))
   > RefreshBodyContainerVisibility() {is=title K2Node_CallFunction}
  ]}
 Set Alignments {nodes [
   n1 = Set Alignments!() -> (then, Alignment:EHorizontalAlignment Enum=HAlign_Fill) {is=title K2Node_FunctionEntry}
   > n2 = Set HorizontalAlignment(HorizontalAlignment:EHorizontalAlignment Enum=@n1) -> (Output_Get:EHorizontalAlignment Enum=HAlign_Fill)
   > n3 = For Each Loop(Array:Array of Text Layout Widget Object References=Make Array!(
     "[0]":Text Layout Widget Object Reference=Get Text_Title()
     "[1]":Text Layout Widget Object Reference=Get Text_Body()
     "[2]":Text Layout Widget Object Reference=Get Text_Set()
     "[3]":Text Layout Widget Object Reference=Get Text_Season()
     "[4]":Text Layout Widget Object Reference=Get Text_GameplayTags()
     "[5]":Text Layout Widget Object Reference=Get Text_Compatibility()
     "[6]":Text Layout Widget Object Reference=Get Text_Variance()
    ) -> (Array:Array of Text Layout Widget Object References)) -> (Array Element:Text Layout Widget Object Reference)
   n4 = Get HorizontalAlignment() -> (HorizontalAlignment:EHorizontalAlignment Enum=HAlign_Fill)
   n5 = Select(
    Option 0:EHorizontalAlignment Enum=@n4
    Option 1:EHorizontalAlignment Enum=Select!(
     HAlign_Fill:EHorizontalAlignment Enum=HAlign_Fill
     HAlign_Left:EHorizontalAlignment Enum=HAlign_Right
     HAlign_Center:EHorizontalAlignment Enum=HAlign_Center
     HAlign_Right:EHorizontalAlignment Enum=HAlign_Left
     Index:EHorizontalAlignment Enum=@n4
    ) -> (ReturnValue)
    Index="Equal, Case Insensitive (String)"(A=GetCurrentCulture!() -> (ReturnValue:String), B=ar)
   ) -> (ReturnValue:EHorizontalAlignment Enum)
   Set HorizontalAlignment!(execute=@n3.Completed, HorizontalAlignment=@n5, self:Wrap Box Object Reference=Get ModeTagsBox()) -> (then, Output_Get=HAlign_Fill)
   > SetGridSlotAlignment!(
    execute
    Widget:Widget Object Reference=Get RarityTag()
    Horizontal:EHorizontalAlignment Enum=@n5
    Vertical:EVerticalAlignment Enum=VAlign_Fill
   ) -> (then)
   > SetHorizontalAlignment(self:Grid Slot Object Reference=SlotAsGridSlot(Widget=Get SB_Tags() -> (SB_Tags:Stack Box Object Reference)), InHorizontalAlignment=@n5)
   Set Justification!(
    execute=@n3.LoopBody
    Justification:ETextJustify Enum=Select(
     HAlign_Fill:ETextJustify Enum=Left
     HAlign_Left:ETextJustify Enum=Left
     HAlign_Center:ETextJustify Enum=Center
     HAlign_Right:ETextJustify Enum=Right
     Index=@n2
    ) -> (ReturnValue:ETextJustify Enum)
    self:Text Layout Widget Object Reference=@n3
   ) -> (then, Output_Get:ETextJustify Enum=Left)
  ]}
 SetBodyText {nodes [
   n1 = SetBodyText!() -> (then, Text:"Text (by ref)") {is=title K2Node_FunctionEntry}
   n2 = Get Text_Body()
   > SetText(self=@n2, InText=@n1)
   > RefreshVisibility(TextBlock=@n2, Hide=Get HideBodyText()) {is=title K2Node_CallFunction}
  ]}
 SetBodyTextVisibility {nodes [
   n1 = SetBodyTextVisibility!() -> (then, Show Text:bool=~) {is=title K2Node_FunctionEntry}
   > n2 = Set HideBodyText(HideBodyText:bool=NOT Boolean(A=@n1)) -> (Output_Get:bool)
   > RefreshVisibility(TextBlock=Get Text_Body(), Hide=@n2)
  ]}
 SetCategoryText {nodes [
   n1 = SetCategoryText!() -> (then, Text:Text) {is=title K2Node_FunctionEntry}
   > n2 = Branch(Condition=Get HideCategoryText())
   SetLabelText(execute=@n2.else, self=Get RarityTag(), Text=@n1)
  ]}
 SetCategoryTextVisibility {nodes [
   n1 = SetCategoryTextVisibility!() -> (then, ShowText:bool=~) {is=title K2Node_FunctionEntry}
   > n2 = Set HideCategoryText(HideCategoryText:bool=NOT Boolean(A=@n1)) -> (Output_Get:bool)
   > Branch(Condition=@n2)
   > SetLabelText(self=Get RarityTag())
  ]}
 SetCompactUIStyle {nodes [
   SetCompactUIStyle() {is=title K2Node_FunctionEntry}
   > Branch(Condition=IsMobileGame())
   > SetPadding!(
    execute
    self:Grid Slot Object Reference=SlotAsGridSlot(Widget=Get Slot_Bottom())
    InPadding_Left:float=0
    InPadding_Top:float=24
    InPadding_Right:float=0
    InPadding_Bottom:float=0
   ) -> (then)
  ]}
 SetCompatibilityTags {nodes [
   n1 = SetCompatibilityTags!() -> (then, Product Tags:Set of Gameplay Tag Structures) {is=title K2Node_FunctionEntry}
   > Clear(TargetArray:Array of Texts=Get GameTags())
   > n2 = To Array!(execute, A:Set of Gameplay Tag Structures=@n1) -> (then, Result:Array of Gameplay Tag Structures)
   > n3 = For Each Loop(Array:Array of Gameplay Tag Structures=@n2) -> (Array Element:Gameplay Tag Structure)
   SetCompatibilityVisibility(execute=@n3.Completed, Show=Is Not Empty!(TargetArray:Array of Texts=Get GameTags()) -> (ReturnValue:bool))
   > AddCompatibilityTags(GameTags=Get GameTags())
   n4 = FindCosmeticConfig!(execute=@n3.LoopBody, ProductTag:Gameplay Tag Structure=@n3) -> (then, ReturnValue:Fort Product Cosmetic Config Object Reference)
   > Is Valid(InputObject=@n4)
   n5 = Get ProductName!(self:Fort Product Cosmetic Config Object Reference=@n4) -> (ProductName:Text)
   n6 = TextIsEmpty(InText=@n5)
   > Branch(Condition=AND Boolean!(
     A=NOT Boolean(A=@n6)
     B=Get bDisplayCompatibilityTag!(self:Fort Product Cosmetic Config Object Reference=@n4) -> (bDisplayCompatibilityTag:bool)
    ) -> (ReturnValue))
   > Add Unique(TargetArray:Array of Texts=Get GameTags(), NewItem:"Text (by ref)"=@n5)
  ]}
 SetCompatibilityText {nodes [
   n1 = SetCompatibilityText!() -> (then, Text:Text) {is=title K2Node_FunctionEntry}
   > "SetText (Text)"(self=Get Text_Compatibility(), InText=@n1)
  ]}
 SetCompatibilityVisibility {nodes [
   n1 = SetCompatibilityVisibility!() -> (then, Show:bool=~) {is=title K2Node_FunctionEntry}
   > Set HideCompatibility(HideCompatibility:bool=NOT Boolean(A=@n1)) -> (Output_Get:bool)
   n2 = AND Boolean(A=@n1, B=Get PermitCompatibilityTagDisplay() -> (PermitCompatibilityTagDisplay:bool))
   > SetVisibility(
    self=Get Grid_Compatibility() -> (Grid_Compatibility:Grid Panel Object Reference)
    InVisibility=Select!(Option 0:ESlateVisibility Enum=Collapsed, Option 1:ESlateVisibility Enum=HitTestInvisible, Index:bool=@n2) -> (ReturnValue:ESlateVisibility Enum)
   )
   > Return Node!(execute, IsVisible:bool=@n2)
  ]}
 SetGameplayTagsText {nodes [
   n1 = SetGameplayTagsText!() -> (then, Text:"Text (by ref)") {is=title K2Node_FunctionEntry}
   n2 = Get Text_GameplayTags()
   > SetText(self=@n2, InText=@n1)
   > RefreshVisibility(TextBlock=@n2, Hide=Get HideGameplayTagsText())
  ]}
 SetGameplayTagsTextVisibility {nodes [
   n1 = SetGameplayTagsTextVisibility!() -> (then, Show Text:bool=~) {is=title K2Node_FunctionEntry}
   > n2 = Set HideGameplayTagsText(HideGameplayTagsText:bool=NOT Boolean(A=@n1)) -> (Output_Get:bool)
   > RefreshVisibility(TextBlock=Get Text_GameplayTags(), Hide=@n2)
  ]}
 SetRarityTagColors {nodes [
   n1 = SetRarityTagColors!() -> (then, FortColorPalette:"Fort Color Palette Structure (by ref)") {is=title K2Node_FunctionEntry}
   n2 = Break Fort Color Palette!(FortColorPalette:"Fort Color Palette Structure (by ref)"=@n1) -> (
    Color1:Linear Color Structure
    Color2:Linear Color Structure
    Color3:Linear Color Structure
    Color4:Linear Color Structure
    Color5:Linear Color Structure
   ) {is=title K2Node_BreakStruct}
   n3 = Get RarityTag()
   > SetTagBackgroundColor(self=@n3, Color=@n2.Color2)
   > SetTagTextColor(self=@n3, Color=@n2)
  ]}
 SetRarityTagVisibility {nodes [
   n1 = SetRarityTagVisibility!() -> (then, ShowTag:bool=~) {is=title K2Node_FunctionEntry}
   > Set HideRarityTag(HideRarityTag:bool=NOT Boolean(A=@n1)) -> (Output_Get:bool)
   > SetTagVisibility!(execute, self:WBP UIKit Tag Base Object Reference=Get RarityTag(), Visible:bool=@n1) -> (then)
  ]}
 SetRarityText {nodes [
   n1 = SetRarityText!() -> (then, Text:Text) {is=title K2Node_FunctionEntry}
   > SetTagText(self=Get RarityTag(), Text=@n1)
   > SetRarityTagVisibility(ShowTag=NOT Boolean(A=Get HideRarityTag())) {is=title K2Node_CallFunction}
  ]}
 SetSeasonText {nodes [
   n1 = SetSeasonText!() -> (then, Text:"Text (by ref)") {is=title K2Node_FunctionEntry}
   n2 = Get Text_Season()
   > SetText(
    self=@n2
    InText=SelectText(B=Format Text!(Format="LOCTEXT(\"Introduced in <SeasonText>{0}</>.\")", 0:"Text (by ref)"=@n1) -> (Result), bPickA=TextIsEmpty(InText=@n1))
   )
   > RefreshVisibility(TextBlock=@n2, Hide=Get HideSeasonText())
  ]}
 SetSeasonTextVisibility {nodes [
   n1 = SetSeasonTextVisibility!() -> (then, Show Text:bool=~) {is=title K2Node_FunctionEntry}
   > n2 = Set HideSeasonText(HideSeasonText:bool=NOT Boolean(A=@n1)) -> (Output_Get:bool)
   > RefreshVisibility(TextBlock=Get Text_Season(), Hide=@n2)
  ]}
 SetSetText {nodes [
   n1 = SetSetText!() -> (then, Text:"Text (by ref)") {is=title K2Node_FunctionEntry}
   n2 = Get Text_Set()
   > SetText(self=@n2, InText=SelectText(B=Format Text(Format="LOCTEXT(\"Part of the <SetName>{0}</> set.\")", 0=@n1), bPickA=TextIsEmpty(InText=@n1)))
   > RefreshVisibility(TextBlock=@n2, Hide=Get HideSetText())
  ]}
 SetSetTextVisibility {nodes [
   n1 = SetSetTextVisibility!() -> (then, Show Text:bool=~) {is=title K2Node_FunctionEntry}
   > n2 = Set HideSetText(HideSetText:bool=NOT Boolean(A=@n1)) -> (Output_Get:bool)
   > RefreshVisibility(TextBlock=Get Text_Set(), Hide=@n2)
  ]}
 SetStyleText {nodes [
   n1 = SetStyleText!() -> (then, Text:Text) {is=title K2Node_FunctionEntry}
   > n2 = Branch(Condition=Get HideStyleText())
   SetStyleText(execute=@n2.else, self=Get RarityTag(), Text=@n1) {is=title K2Node_CallFunction}
  ]}
 SetStyleTextVisibility {nodes [
   n1 = SetStyleTextVisibility!() -> (then, ShowText:bool=~) {is=title K2Node_FunctionEntry}
   > n2 = Set HideStyleText(HideStyleText:bool=NOT Boolean(A=@n1)) -> (Output_Get:bool)
   > Branch(Condition=@n2)
   > SetStyleText(self=Get RarityTag())
  ]}
 SetTagsOrientation {nodes [
   n1 = SetTagsOrientation!() -> (then, Orientation:EOrientation Enum=Orient_Horizontal) {is=title K2Node_FunctionEntry}
   > n2 = Set TagsOrientation(TagsOrientation:EOrientation Enum=@n1) -> (Output_Get:EOrientation Enum=Orient_Horizontal)
   > n3 = Set Orientation!(execute, Orientation:EOrientation Enum=@n2, self:Stack Box Object Reference=Get SB_Tags()) -> (then, Output_Get:EOrientation Enum=Orient_Horizontal)
   > SetTagOrientation!(execute, self:WBP UIKit Tag Base Object Reference=Get RarityTag(), Orientation:EOrientation Enum=@n3) -> (then)
   > n4 = Cast To StackBoxSlot!(execute, Object=Get Slot!(self:Widget Object Reference=Get EquippedTag()) -> (Slot:Panel Slot Object Reference)) -> (then, CastFailed, AsStack Box Slot:Stack Box Slot Object Reference) {is=title K2Node_DynamicCast}
   > n5 = Set Padding!(
    execute
    Padding:Margin Structure=Select!(
     Orient_Horizontal_Left:float=0
     Orient_Horizontal_Top:float=0
     Orient_Horizontal_Right:float=8
     Orient_Horizontal_Bottom:float=0
     Orient_Vertical_Left:float=0
     Orient_Vertical_Top:float=0
     Orient_Vertical_Right:float=0
     Orient_Vertical_Bottom:float=8
     Index:EOrientation Enum=Get TagsOrientation() -> (TagsOrientation=Orient_Horizontal)
    ) -> (ReturnValue:Margin Structure)
    self:Stack Box Slot Object Reference=@n4
   ) -> (then, Output_Get:Margin Structure)
   n6 = GetCurrentCulture()
   n7 = Get Orientation() -> (Orientation:EOrientation Enum=Orient_Horizontal)
   Branch(
    execute=[@n5.then, @n4.CastFailed]
    Condition=AND Boolean(A="Equal (Enum)"(A:EOrientation Enum=@n7, B:EOrientation Enum=Orient_Horizontal), B="Equal, Case Insensitive (String)"(A=@n6, B=ar))
   )
   > Reverse!(execute, TargetArray:Array of Widget Object References=GetAllChildren(self=Get SB_Tags())) -> (then) {is=title K2Node_CallArrayFunction}
  ]}
 SetTitleText {nodes [
   n1 = SetTitleText!() -> (then, Text:Text) {is=title K2Node_FunctionEntry}
   > "SetText (Text)"(self=Get Text_Title(), InText=@n1)
  ]}
 SetUpDisplayInternalItemName {nodes [
   SetUpDisplayInternalItemName() {is=title K2Node_FunctionEntry}
   > n1 = GetConsoleVariableBoolValue!(execute, VariableName:String=VehicleCosmetics.bShowInternalDisplayName) -> (then, ReturnValue:bool)
   > Branch(Condition=@n1)
   n2 = Get Text_Title()
   > n3 = Set members in Slate Font Info!(
    execute:Exec
    StructRef:"Slate Font Info Structure (by ref)"=Get Font!(self:Text Object Reference=@n2) -> (Font:Slate Font Info Structure)
    Size:float=Get TitleInternalNameFont() -> (TitleInternalNameFont:float)
   ) -> (then:Exec, StructOut:"Slate Font Info Structure (by ref)") {is=title K2Node_SetFieldsInStruct}
   > Set Font!(execute, Font:Slate Font Info Structure=@n3, self:Text Object Reference=@n2) -> (then, Output_Get:Slate Font Info Structure)
  ]}
 SetVarianceText {nodes [
   n1 = SetVarianceText!() -> (then, Text:"Text (by ref)") {is=title K2Node_FunctionEntry}
   n2 = Get Text_Variance()
   > SetText(self=@n2, InText=@n1)
   > RefreshVisibility(TextBlock=@n2, Hide=Get HideVarianceText())
  ]}
 SetVarianceTextVisibility {nodes [
   n1 = SetVarianceTextVisibility!() -> (then, Show Text:bool=~) {is=title K2Node_FunctionEntry}
   > n2 = Set HideVarianceText(HideVarianceText:bool=NOT Boolean(A=@n1)) -> (Output_Get:bool)
   > RefreshVisibility(TextBlock=Get Text_Variance(), Hide=@n2)
  ]}
 SetWidgetLayout {nodes [
   SetWidgetLayout() {is=title K2Node_FunctionEntry}
   n1 = SlotAsGridSlot(Widget=Get SizeBoxTitle())
   > SetRow!(execute, self:Grid Slot Object Reference=@n1, InRow:int=Get TitleRow() -> (TitleRow:int=0)) -> (then)
   > Set Padding(Padding=Get TitlePadding() -> (TitlePadding:Margin Structure), self:Grid Slot Object Reference=@n1)
   n2 = SlotAsGridSlot(Widget=Get VBox_Body())
   > SetRow(self=@n2, InRow=Get BodyTextRow() -> (BodyTextRow:int=0))
   > Set Padding(Padding=Get BodyTextPadding() -> (BodyTextPadding:Margin Structure), self=@n2)
   n3 = SlotAsGridSlot(Widget=Get RarityTag())
   > SetRow(self=@n3, InRow=Get TagRow() -> (TagRow:int=0))
   > Set Padding(Padding=Get TagPadding() -> (TagPadding:Margin Structure), self=@n3)
   n4 = SlotAsGridSlot(Widget=Get Grid_Compatibility())
   > SetRow(self=@n4, InRow=Get CompatibilityRow() -> (CompatibilityRow:int=0))
   > Set Padding(Padding=Get CompatibilityPadding() -> (CompatibilityPadding:Margin Structure), self=@n4)
   n5 = SlotAsGridSlot(Widget=Get Slot_Bottom())
   > SetRow(self=@n5, InRow=Get BottomSlotRow() -> (BottomSlotRow:int=0))
   > Set Padding(Padding=Get BottonSlotPadding() -> (BottonSlotPadding:Margin Structure), self=@n5)
   > Set MaxDesiredHeight(MaxDesiredHeight=Get TitleMaxHeight(), self=Get SizeBoxTitle())
   > Set Alignments!(execute, self, Alignment:EHorizontalAlignment Enum=Get HorizontalAlignment() -> (HorizontalAlignment=HAlign_Fill)) -> (then) {is=title K2Node_CallFunction}
  ]}
]
variables [
 BodyTextColor {type=SlateColor, default {colorUseRule=UseColor_Specified, specifiedColor=<rgba 1 1 1 1>}}
 BodyTextCustomTextStyle {type=DataTable, default=null}
 BodyTextPadding {type=Margin, default=<ltrb 0 32 0 0>}
 BodyTextRow {type=integer, default=2}
 BottomSlotRow {type=integer, default=4}
 BottonSlotPadding {type=Margin, default=<ltrb 0 16 0 0>}
 CanShowEquippedTag {type=boolean}
 CompatibilityPadding {type=Margin, default=<ltrb 0 32 0 0>}
 CompatibilityRow {type=integer, default=3}
 CompatibilityTitleColor {type=SlateColor, default {colorUseRule=UseColor_Specified, specifiedColor=<rgba 1 1 1 1>}}
 EquippedTagBackgroundColor {type=LinearColor, default=<rgba 0 0.0331 0.0296 1>}
 EquippedTagForegroundColor {type=LinearColor, default=<rgba 0.0395 1 0.2307 1>}
 GameTagClass {type=Class, default=/FortUILibrary/UIKit/Tag/WBP_UIKit_Tag.WBP_UIKit_Tag_C}
 GameTags {type="array<string>", default ["[PH] Battle Royale", "[PH] Fortnite Festival", "[PH] Rocket Racing"]}
 GameTagType {type=E_UI_TagType, default=Secondary Tag}
 HideBodyText {type=boolean}
 HideCategoryText {type=boolean}
 HideCompatibility {type=boolean}
 HideGameplayTagsText {type=boolean, default=true}
 HideRarityTag {type=boolean}
 HideSeasonText {type=boolean, default=true}
 HideSetText {type=boolean, default=true}
 HideStyleText {type=boolean}
 HideVarianceText {type=boolean, default=true}
 HorizontalAlignment {type=EHorizontalAlignment, default=HAlign_Left}
 PermitCompatibilityTagDisplay {type=boolean, default=true}
 TagPadding {type=Margin, default=<ltrb 0 0 0 16>}
 TagRow {type=integer}
 TagSize {type=E_UI_TagSize, default=Default}
 TagsOrientation {type=EOrientation, default=Orient_Horizontal}
 TagTitleColor {type=LinearColor, default=<rgba 1 1 1 0.702>}
 TagWrapSizeDefault {type=boolean}
 TagWrapSizeMobile {type=boolean, default=true}
 TitleColor {type=SlateColor, default {colorUseRule=UseColor_Specified, specifiedColor=<rgba 1 1 1 1>}}
 TitleFont {
  type=SlateFontInfo
  default {
   bForceMonospaced=false
   bMaterialIsStencil=false
   fontMaterial=null
   fontObject=&HeadingNow
   letterSpacing=0
   monospacedWidth=1
   size=37.5
   skewAmount=0
   typefaceFontName=86 Bold Italic
   outlineSettings {
    bApplyOutlineToDropShadows=false
    bMiteredCorners=false
    bSeparateFillAlpha=false
    outlineMaterial=null
    outlineSize=0
    outlineColor=<rgba 0 0 0 1>
   }
  }
 }
 TitleInternalNameFont {type=number, default=15}
 TitleMaxHeight {type=number, default=125}
 TitlePadding {type=Margin, default=<ltrb 0 0 0 0>}
 TitleRow {type=integer, default=1}
 TitleWrap {type=E_UIKit_ItemDescription_TextWrapType, default=Marquee vertical scrolling}
]
widgets {
 info {inheritedWidgetCount=0, parentClass=&UserWidget, rootWidgetClass=&GridPanel, widgetCount=17}
 widgets [
  Root {widgetClassPath=&GridPanel, bIsVariable=true, widget=Root, properties {columnFill [1], rowFill [0]}}
  SB_Tags {
   widgetClassPath=&StackBox
   bIsVariable=true
   parent=Root
   widget=SB_Tags
   properties {flowDirectionPreference=Culture}
   slot {class=GridSlot, properties {horizontalAlignment=HAlign_Left}}
  }
  EquippedTag {
   widgetClassPath=/FortUILibrary/UIKit/Tag/WBP_UIKit_Tag
   bIsVariable=true
   parent=SB_Tags
   widget=EquippedTag
   properties {
    labelText=""
    tagText=Equipped
    visibility=Collapsed
    foregroundColor {colorUseRule=UseColor_Foreground, specifiedColor=<rgba 0 1 0.231 1>}
    iconSizeCompact=<0 0>
    primaryFontCompact {
     bForceMonospaced=false
     bMaterialIsStencil=false
     fontMaterial=null
     fontObject=null
     letterSpacing=0
     monospacedWidth=1
     size=24
     skewAmount=0
     typefaceFontName=None
     outlineSettings {
      bApplyOutlineToDropShadows=false
      bMiteredCorners=false
      bSeparateFillAlpha=false
      outlineMaterial=null
      outlineSize=0
      outlineColor=<rgba 0 0 0 1>
     }
    }
   }
   slot {class=StackBoxSlot, properties {padding=<ltrb 0 0 8 0>}}
  }
  RarityTag {
   widgetClassPath=&WBP_UIKit_Tag
   bIsVariable=true
   parent=SB_Tags
   widget=RarityTag
   properties {
    labelText=Category
    styleText=Style
    tagText=Rarity
    iconSizeCompact=<0 0>
    primaryFontCompact {
     bForceMonospaced=false
     bMaterialIsStencil=false
     fontMaterial=null
     fontObject=null
     letterSpacing=0
     monospacedWidth=1
     size=24
     skewAmount=0
     typefaceFontName=None
     outlineSettings {
      bApplyOutlineToDropShadows=false
      bMiteredCorners=false
      bSeparateFillAlpha=false
      outlineMaterial=null
      outlineSize=0
      outlineColor=<rgba 0 0 0 1>
     }
    }
   }
   slot {class=StackBoxSlot}
  }
  SizeBoxTitle {
   widgetClassPath=&SizeBox
   bIsVariable=true
   parent=Root
   widget=SizeBoxTitle
   properties {bOverride_MinDesiredWidth=true, heightOverride=88, maxDesiredHeight=29.1695, minDesiredWidth=720, widthOverride=740}
   slot {class=GridSlot, properties {verticalAlignment=VAlign_Top}}
  }
  Text_Title {
   widgetClassPath=&CommonTextBlock
   bIsVariable=true
   parent=SizeBoxTitle
   widget=Text_Title
   properties {
    autoWrapText=true
    bAutoCollapseWithEmptyText=true
    bIsScrollingEnabled=false
    lineHeightPercentage=0.9
    scrollOrientation=Orient_Vertical
    scrollStyle=/FortUILibrary/Styles/Typography/Foundation/TextScrollStyle-NoFade.TextScrollStyle-NoFade_C
    text="Item long long name, believe me, it's really long"
    textOverflowPolicy=Ellipsis
    textTransformPolicy=ToUpper
    wrappingPolicy=AllowPerCharacterWrapping
    font {
     bForceMonospaced=false
     bMaterialIsStencil=false
     fontMaterial=null
     fontObject=&HeadingNow
     letterSpacing=0
     monospacedWidth=1
     size=32.25
     skewAmount=0
     typefaceFontName=HeadingNow-86BoldItalic
     outlineSettings {
      bApplyOutlineToDropShadows=false
      bMiteredCorners=false
      bSeparateFillAlpha=false
      outlineMaterial=null
      outlineSize=0
      outlineColor=<rgba 0 0 0 1>
     }
    }
    shadowOffset=<0 0>
   }
   slot {class=SizeBoxSlot}
  }
  VBox_Body {
   widgetClassPath=&VerticalBox
   bIsVariable=true
   parent=Root
   widget=VBox_Body
   slot {class=GridSlot, properties {row=2, verticalAlignment=VAlign_Top, padding=<ltrb 0 32 0 0>}}
  }
  Text_Variance {
   widgetClassPath=&CommonRichTextBlock
   bIsVariable=true
   parent=VBox_Body
   widget=Text_Variance
   properties {
    autoWrapText=true
    bOverrideDefaultStyle=true
    lineHeightPercentage=1.15
    mobileTextBlockScale=1.2
    text="[PH] Item Variance"
    textStyleSet=/Game/Athena/Items/Cosmetics/Metadata/ItemDescriptionTextStyle
    defaultTextStyleOverride {
     overflowPolicy=Clip
     transformPolicy=None
     colorAndOpacity {colorUseRule=UseColor_Specified, specifiedColor=<rgba 1 1 1 0.7>}
     font {
      bForceMonospaced=false
      bMaterialIsStencil=false
      fontMaterial=null
      fontObject=&HeadingNow
      letterSpacing=0
      monospacedWidth=1
      size=15.75
      skewAmount=0
      typefaceFontName=74 Regular
      outlineSettings {
       bApplyOutlineToDropShadows=true
       bMiteredCorners=false
       bSeparateFillAlpha=false
       outlineMaterial=null
       outlineSize=0
       outlineColor=<rgba 0 0 0 1>
      }
     }
     highlightColor {colorUseRule=UseColor_Specified, specifiedColor=<rgba 0 0 0 0>}
     highlightShape {
      bIsDynamicallyLoaded=false
      drawAs=Image
      imageType=NoImage
      mirroring=NoMirror
      resourceName=None
      resourceObject=null
      tiling=NoTile
      imageSize=<32 32>
      margin=<ltrb 0 0 0 0>
      outlineSettings {
       bUseBrushTransparency=false
       roundingType=HalfHeightRadius
       width=0
       color {colorUseRule=UseColor_Specified, specifiedColor=<rgba 0 0 0 0>}
       cornerRadii=<0 0 0 0>
      }
      tintColor {colorUseRule=UseColor_Specified, specifiedColor=<rgba 1 1 1 1>}
      uVRegion {bIsValid=false, max=<0 0>, min=<0 0>}
     }
     selectedBackgroundColor {colorUseRule=UseColor_Foreground, specifiedColor=<rgba 1 0 1 1>}
     shadowColorAndOpacity=<rgba 0 0 0 1>
     shadowOffset=<0 0>
     strikeBrush {
      bIsDynamicallyLoaded=false
      drawAs=Image
      imageType=NoImage
      mirroring=NoMirror
      resourceName=None
      resourceObject=null
      tiling=NoTile
      imageSize=<32 32>
      margin=<ltrb 0 0 0 0>
      outlineSettings {
       bUseBrushTransparency=false
       roundingType=HalfHeightRadius
       width=0
       color {colorUseRule=UseColor_Specified, specifiedColor=<rgba 0 0 0 0>}
       cornerRadii=<0 0 0 0>
      }
      tintColor {colorUseRule=UseColor_Specified, specifiedColor=<rgba 1 1 1 1>}
      uVRegion {bIsValid=false, max=<0 0>, min=<0 0>}
     }
     underlineBrush {
      bIsDynamicallyLoaded=false
      drawAs=Image
      imageType=NoImage
      mirroring=NoMirror
      resourceName=None
      resourceObject=null
      tiling=NoTile
      imageSize=<32 32>
      margin=<ltrb 0 0 0 0>
      outlineSettings {
       bUseBrushTransparency=false
       roundingType=HalfHeightRadius
       width=0
       color {colorUseRule=UseColor_Specified, specifiedColor=<rgba 0 0 0 0>}
       cornerRadii=<0 0 0 0>
      }
      tintColor {colorUseRule=UseColor_Specified, specifiedColor=<rgba 1 1 1 1>}
      uVRegion {bIsValid=false, max=<0 0>, min=<0 0>}
     }
    }
   }
   slot {class=VerticalBoxSlot}
  }
  Text_Body {
   widgetClassPath=&CommonRichTextBlock
   bIsVariable=true
   parent=VBox_Body
   widget=Text_Body
   properties {
    autoWrapText=true
    bOverrideDefaultStyle=true
    lineHeightPercentage=1.15
    mobileTextBlockScale=1.2
    text="[PH] Item Description"
    textStyleSet=&ItemDescriptionTextStyle
    defaultTextStyleOverride {
     overflowPolicy=Clip
     transformPolicy=None
     colorAndOpacity {colorUseRule=UseColor_Specified, specifiedColor=<rgba 1 1 1 0.7>}
     font {
      bForceMonospaced=false
      bMaterialIsStencil=false
      fontMaterial=null
      fontObject=&HeadingNow
      letterSpacing=0
      monospacedWidth=1
      size=15.75
      skewAmount=0
      typefaceFontName=74 Regular
      outlineSettings {
       bApplyOutlineToDropShadows=true
       bMiteredCorners=false
       bSeparateFillAlpha=false
       outlineMaterial=null
       outlineSize=0
       outlineColor=<rgba 0 0 0 1>
      }
     }
     highlightColor {colorUseRule=UseColor_Specified, specifiedColor=<rgba 0 0 0 0>}
     highlightShape {
      bIsDynamicallyLoaded=false
      drawAs=Image
      imageType=NoImage
      mirroring=NoMirror
      resourceName=None
      resourceObject=null
      tiling=NoTile
      imageSize=<32 32>
      margin=<ltrb 0 0 0 0>
      outlineSettings {
       bUseBrushTransparency=false
       roundingType=HalfHeightRadius
       width=0
       color {colorUseRule=UseColor_Specified, specifiedColor=<rgba 0 0 0 0>}
       cornerRadii=<0 0 0 0>
      }
      tintColor {colorUseRule=UseColor_Specified, specifiedColor=<rgba 1 1 1 1>}
      uVRegion {bIsValid=false, max=<0 0>, min=<0 0>}
     }
     selectedBackgroundColor {colorUseRule=UseColor_Foreground, specifiedColor=<rgba 1 0 1 1>}
     shadowColorAndOpacity=<rgba 0 0 0 1>
     shadowOffset=<0 0>
     strikeBrush {
      bIsDynamicallyLoaded=false
      drawAs=Image
      imageType=NoImage
      mirroring=NoMirror
      resourceName=None
      resourceObject=null
      tiling=NoTile
      imageSize=<32 32>
      margin=<ltrb 0 0 0 0>
      outlineSettings {
       bUseBrushTransparency=false
       roundingType=HalfHeightRadius
       width=0
       color {colorUseRule=UseColor_Specified, specifiedColor=<rgba 0 0 0 0>}
       cornerRadii=<0 0 0 0>
      }
      tintColor {colorUseRule=UseColor_Specified, specifiedColor=<rgba 1 1 1 1>}
      uVRegion {bIsValid=false, max=<0 0>, min=<0 0>}
     }
     underlineBrush {
      bIsDynamicallyLoaded=false
      drawAs=Image
      imageType=NoImage
      mirroring=NoMirror
      resourceName=None
      resourceObject=null
      tiling=NoTile
      imageSize=<32 32>
      margin=<ltrb 0 0 0 0>
      outlineSettings {
       bUseBrushTransparency=false
       roundingType=HalfHeightRadius
       width=0
       color {colorUseRule=UseColor_Specified, specifiedColor=<rgba 0 0 0 0>}
       cornerRadii=<0 0 0 0>
      }
      tintColor {colorUseRule=UseColor_Specified, specifiedColor=<rgba 1 1 1 1>}
      uVRegion {bIsValid=false, max=<0 0>, min=<0 0>}
     }
    }
   }
   slot {class=VerticalBoxSlot}
  }
  Text_Set {
   widgetClassPath=&CommonRichTextBlock
   bIsVariable=true
   parent=VBox_Body
   widget=Text_Set
   properties {
    autoWrapText=true
    bOverrideDefaultStyle=true
    lineHeightPercentage=1.15
    mobileTextBlockScale=1.2
    text="[PH] Item Set"
    textStyleSet=&ItemDescriptionTextStyle
    visibility=Collapsed
    defaultTextStyleOverride {
     overflowPolicy=Clip
     transformPolicy=None
     colorAndOpacity {colorUseRule=UseColor_Specified, specifiedColor=<rgba 1 1 1 0.7>}
     font {
      bForceMonospaced=false
      bMaterialIsStencil=false
      fontMaterial=null
      fontObject=&HeadingNow
      letterSpacing=0
      monospacedWidth=1
      size=15.75
      skewAmount=0
      typefaceFontName=74 Regular
      outlineSettings {
       bApplyOutlineToDropShadows=true
       bMiteredCorners=false
       bSeparateFillAlpha=false
       outlineMaterial=null
       outlineSize=0
       outlineColor=<rgba 0 0 0 1>
      }
     }
     highlightColor {colorUseRule=UseColor_Specified, specifiedColor=<rgba 0 0 0 0>}
     highlightShape {
      bIsDynamicallyLoaded=false
      drawAs=Image
      imageType=NoImage
      mirroring=NoMirror
      resourceName=None
      resourceObject=null
      tiling=NoTile
      imageSize=<32 32>
      margin=<ltrb 0 0 0 0>
      outlineSettings {
       bUseBrushTransparency=false
       roundingType=HalfHeightRadius
       width=0
       color {colorUseRule=UseColor_Specified, specifiedColor=<rgba 0 0 0 0>}
       cornerRadii=<0 0 0 0>
      }
      tintColor {colorUseRule=UseColor_Specified, specifiedColor=<rgba 1 1 1 1>}
      uVRegion {bIsValid=false, max=<0 0>, min=<0 0>}
     }
     selectedBackgroundColor {colorUseRule=UseColor_Foreground, specifiedColor=<rgba 1 0 1 1>}
     shadowColorAndOpacity=<rgba 0 0 0 1>
     shadowOffset=<0 0>
     strikeBrush {
      bIsDynamicallyLoaded=false
      drawAs=Image
      imageType=NoImage
      mirroring=NoMirror
      resourceName=None
      resourceObject=null
      tiling=NoTile
      imageSize=<32 32>
      margin=<ltrb 0 0 0 0>
      outlineSettings {
       bUseBrushTransparency=false
       roundingType=HalfHeightRadius
       width=0
       color {colorUseRule=UseColor_Specified, specifiedColor=<rgba 0 0 0 0>}
       cornerRadii=<0 0 0 0>
      }
      tintColor {colorUseRule=UseColor_Specified, specifiedColor=<rgba 1 1 1 1>}
      uVRegion {bIsValid=false, max=<0 0>, min=<0 0>}
     }
     underlineBrush {
      bIsDynamicallyLoaded=false
      drawAs=Image
      imageType=NoImage
      mirroring=NoMirror
      resourceName=None
      resourceObject=null
      tiling=NoTile
      imageSize=<32 32>
      margin=<ltrb 0 0 0 0>
      outlineSettings {
       bUseBrushTransparency=false
       roundingType=HalfHeightRadius
       width=0
       color {colorUseRule=UseColor_Specified, specifiedColor=<rgba 0 0 0 0>}
       cornerRadii=<0 0 0 0>
      }
      tintColor {colorUseRule=UseColor_Specified, specifiedColor=<rgba 1 1 1 1>}
      uVRegion {bIsValid=false, max=<0 0>, min=<0 0>}
     }
    }
   }
   slot {class=VerticalBoxSlot}
  }
  Text_Season {
   widgetClassPath=&CommonRichTextBlock
   bIsVariable=true
   parent=VBox_Body
   widget=Text_Season
   properties {
    autoWrapText=true
    bOverrideDefaultStyle=true
    lineHeightPercentage=1.15
    mobileTextBlockScale=1.2
    text="[PH] Item Season"
    textStyleSet=&ItemDescriptionTextStyle
    visibility=Collapsed
    defaultTextStyleOverride {
     overflowPolicy=Clip
     transformPolicy=None
     colorAndOpacity {colorUseRule=UseColor_Specified, specifiedColor=<rgba 1 1 1 0.7>}
     font {
      bForceMonospaced=false
      bMaterialIsStencil=false
      fontMaterial=null
      fontObject=&HeadingNow
      letterSpacing=0
      monospacedWidth=1
      size=15.75
      skewAmount=0
      typefaceFontName=74 Regular
      outlineSettings {
       bApplyOutlineToDropShadows=true
       bMiteredCorners=false
       bSeparateFillAlpha=false
       outlineMaterial=null
       outlineSize=0
       outlineColor=<rgba 0 0 0 1>
      }
     }
     highlightColor {colorUseRule=UseColor_Specified, specifiedColor=<rgba 0 0 0 0>}
     highlightShape {
      bIsDynamicallyLoaded=false
      drawAs=Image
      imageType=NoImage
      mirroring=NoMirror
      resourceName=None
      resourceObject=null
      tiling=NoTile
      imageSize=<32 32>
      margin=<ltrb 0 0 0 0>
      outlineSettings {
       bUseBrushTransparency=false
       roundingType=HalfHeightRadius
       width=0
       color {colorUseRule=UseColor_Specified, specifiedColor=<rgba 0 0 0 0>}
       cornerRadii=<0 0 0 0>
      }
      tintColor {colorUseRule=UseColor_Specified, specifiedColor=<rgba 1 1 1 1>}
      uVRegion {bIsValid=false, max=<0 0>, min=<0 0>}
     }
     selectedBackgroundColor {colorUseRule=UseColor_Foreground, specifiedColor=<rgba 1 0 1 1>}
     shadowColorAndOpacity=<rgba 0 0 0 1>
     shadowOffset=<0 0>
     strikeBrush {
      bIsDynamicallyLoaded=false
      drawAs=Image
      imageType=NoImage
      mirroring=NoMirror
      resourceName=None
      resourceObject=null
      tiling=NoTile
      imageSize=<32 32>
      margin=<ltrb 0 0 0 0>
      outlineSettings {
       bUseBrushTransparency=false
       roundingType=HalfHeightRadius
       width=0
       color {colorUseRule=UseColor_Specified, specifiedColor=<rgba 0 0 0 0>}
       cornerRadii=<0 0 0 0>
      }
      tintColor {colorUseRule=UseColor_Specified, specifiedColor=<rgba 1 1 1 1>}
      uVRegion {bIsValid=false, max=<0 0>, min=<0 0>}
     }
     underlineBrush {
      bIsDynamicallyLoaded=false
      drawAs=Image
      imageType=NoImage
      mirroring=NoMirror
      resourceName=None
      resourceObject=null
      tiling=NoTile
      imageSize=<32 32>
      margin=<ltrb 0 0 0 0>
      outlineSettings {
       bUseBrushTransparency=false
       roundingType=HalfHeightRadius
       width=0
       color {colorUseRule=UseColor_Specified, specifiedColor=<rgba 0 0 0 0>}
       cornerRadii=<0 0 0 0>
      }
      tintColor {colorUseRule=UseColor_Specified, specifiedColor=<rgba 1 1 1 1>}
      uVRegion {bIsValid=false, max=<0 0>, min=<0 0>}
     }
    }
   }
   slot {class=VerticalBoxSlot}
  }
  Text_GameplayTags {
   widgetClassPath=&CommonRichTextBlock
   bIsVariable=true
   parent=VBox_Body
   widget=Text_GameplayTags
   properties {
    autoWrapText=true
    bOverrideDefaultStyle=true
    lineHeightPercentage=1.15
    mobileTextBlockScale=1.2
    text="[PH] Item Tags"
    textStyleSet=&ItemDescriptionTextStyle
    visibility=Collapsed
    defaultTextStyleOverride {
     overflowPolicy=Clip
     transformPolicy=None
     colorAndOpacity {colorUseRule=UseColor_Specified, specifiedColor=<rgba 1 1 1 0.7>}
     font {
      bForceMonospaced=false
      bMaterialIsStencil=false
      fontMaterial=null
      fontObject=&HeadingNow
      letterSpacing=0
      monospacedWidth=1
      size=15.75
      skewAmount=0
      typefaceFontName=74 Regular
      outlineSettings {
       bApplyOutlineToDropShadows=true
       bMiteredCorners=false
       bSeparateFillAlpha=false
       outlineMaterial=null
       outlineSize=0
       outlineColor=<rgba 0 0 0 1>
      }
     }
     highlightColor {colorUseRule=UseColor_Specified, specifiedColor=<rgba 0 0 0 0>}
     highlightShape {
      bIsDynamicallyLoaded=false
      drawAs=Image
      imageType=NoImage
      mirroring=NoMirror
      resourceName=None
      resourceObject=null
      tiling=NoTile
      imageSize=<32 32>
      margin=<ltrb 0 0 0 0>
      outlineSettings {
       bUseBrushTransparency=false
       roundingType=HalfHeightRadius
       width=0
       color {colorUseRule=UseColor_Specified, specifiedColor=<rgba 0 0 0 0>}
       cornerRadii=<0 0 0 0>
      }
      tintColor {colorUseRule=UseColor_Specified, specifiedColor=<rgba 1 1 1 1>}
      uVRegion {bIsValid=false, max=<0 0>, min=<0 0>}
     }
     selectedBackgroundColor {colorUseRule=UseColor_Foreground, specifiedColor=<rgba 1 0 1 1>}
     shadowColorAndOpacity=<rgba 0 0 0 1>
     shadowOffset=<0 0>
     strikeBrush {
      bIsDynamicallyLoaded=false
      drawAs=Image
      imageType=NoImage
      mirroring=NoMirror
      resourceName=None
      resourceObject=null
      tiling=NoTile
      imageSize=<32 32>
      margin=<ltrb 0 0 0 0>
      outlineSettings {
       bUseBrushTransparency=false
       roundingType=HalfHeightRadius
       width=0
       color {colorUseRule=UseColor_Specified, specifiedColor=<rgba 0 0 0 0>}
       cornerRadii=<0 0 0 0>
      }
      tintColor {colorUseRule=UseColor_Specified, specifiedColor=<rgba 1 1 1 1>}
      uVRegion {bIsValid=false, max=<0 0>, min=<0 0>}
     }
     underlineBrush {
      bIsDynamicallyLoaded=false
      drawAs=Image
      imageType=NoImage
      mirroring=NoMirror
      resourceName=None
      resourceObject=null
      tiling=NoTile
      imageSize=<32 32>
      margin=<ltrb 0 0 0 0>
      outlineSettings {
       bUseBrushTransparency=false
       roundingType=HalfHeightRadius
       width=0
       color {colorUseRule=UseColor_Specified, specifiedColor=<rgba 0 0 0 0>}
       cornerRadii=<0 0 0 0>
      }
      tintColor {colorUseRule=UseColor_Specified, specifiedColor=<rgba 1 1 1 1>}
      uVRegion {bIsValid=false, max=<0 0>, min=<0 0>}
     }
    }
   }
   slot {class=VerticalBoxSlot}
  }
  Grid_Compatibility {
   widgetClassPath=&GridPanel
   bIsVariable=true
   parent=Root
   widget=Grid_Compatibility
   properties {columnFill [1], rowFill [0,1]}
   slot {class=GridSlot, properties {row=3, padding=<ltrb 0 40 0 0>}}
  }
  Text_Compatibility {
   widgetClassPath=&CommonTextBlock
   bIsVariable=true
   parent=Grid_Compatibility
   widget=Text_Compatibility
   properties {
    bAutoCollapseWithEmptyText=true
    bIsScrollingEnabled=false
    mobileFontSizeMultiplier=1.2
    scrollStyle=&TextScrollStyle-NoFade_C
    text=This item is usable in:
    textOverflowPolicy=Ellipsis
    font {
     bForceMonospaced=false
     bMaterialIsStencil=false
     fontMaterial=null
     fontObject=&HeadingNow
     letterSpacing=0
     monospacedWidth=1
     size=14.25
     skewAmount=0
     typefaceFontName=74 Regular
     outlineSettings {
      bApplyOutlineToDropShadows=false
      bMiteredCorners=false
      bSeparateFillAlpha=false
      outlineMaterial=null
      outlineSize=0
      outlineColor=<rgba 0 0 0 1>
     }
    }
    shadowOffset=<0 0>
   }
   slot {class=GridSlot, properties {verticalAlignment=VAlign_Top, padding=<ltrb 0 0 0 8>}}
  }
  Overlay_0 {widgetClassPath=&Overlay, parent=Grid_Compatibility, widget=Overlay_0, slot {class=GridSlot, properties {row=1}}}
  ModeTagsBox {
   widgetClassPath=&WrapBox
   bIsVariable=true
   parent=Overlay_0
   widget=ModeTagsBox
   properties {innerSlotPadding=<0 8>}
   slot {class=OverlaySlot, properties {horizontalAlignment=HAlign_Fill, verticalAlignment=VAlign_Fill}}
  }
  Slot_Bottom {
   widgetClassPath=&NamedSlot
   bIsVariable=true
   parent=Root
   widget=Slot_Bottom
   slot {class=GridSlot, properties {row=4, padding=<ltrb 0 16 0 0>}}
  }
 ]
}
