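# WBP_UIKit_ItemList_CardBase is a base widget blueprint in the FortUILibrary items system for item cards shown in lists, derived from WBP_UIKit_ItemCard_Base.
# It has no widget tree or variables of its own.
# Its event graph declares two custom events with no logic attached: SetListItem, which takes an item object, and UpdatePresentation.
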WidgetBlueprint /FortUILibrary/Systems/Items/WBP_UIKit_ItemList_CardBase
graphs [EventGraph {nodes [
   SetListItem!() -> (OutputDelegate, then, InListItem:Object Reference) {is=title K2Node_CustomEvent}
   UpdatePresentation() {is=title K2Node_CustomEvent}
  ]}]
variables []
widgets {
 info {
  inheritedWidgetCount=0
  parentClass=&/WBP_UIKit_ItemCard_Base.WBP_UIKit_ItemCard_Base_C
  rootWidgetClass=null
  widgetCount=0
 }
 widgets []
}
